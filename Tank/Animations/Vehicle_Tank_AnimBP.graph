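# Animation Blueprint for the tank vehicle, posing hatches, cannon, passenger turret and treads.
# Hatch poses layer with additive tread-power bars, a CannonFire state machine, turret equip poses, montage slot and overheat additive before cannon aiming.
# That cached pose optionally gains hatch-gun aim, blended by TurretBlendValue.
# A custom roller/tread node then animates running gear from tread velocities before inertialization smooths state transitions.
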AnimBlueprint /Tank/Animations/Vehicle_Tank_AnimBP
target_skeleton=/Tank/Meshes/SK_Tank_Vehicle_Skeleton
graphs [
 AnimGraph {nodes [
   n1 = Use cached pose 'IncludeTurret'()
   n2 = Use cached pose 'ExcludeTurret'()
   n3 = Get TurretBlendValue() -> (TurretBlendValue:float)
   n4 = Transform Rollers/Treads!(
    ComponentPose:Component Space Pose Link Structure=Local To Component(LocalPose=Two Way Blend(A=@n1, B=@n2, Alpha=@n3) {blendNode {alphaBoolBlend {blendOption=Linear}}})
    Alpha:float=1
    RollerAnimStates:Map of Names to Roller Anim State Structures=Get RollerAnimStates() -> (RollerAnimStates:Map of Names to Roller Anim State Structures)
    LeftTreadVelocity:float=Get LeftTreadVelocity() -> (LeftTreadVelocity:float)
    RightTreadVelocity:float=Get RightTreadVelocity() -> (RightTreadVelocity:float)
    RotationBoneName:String="rot_{0}"
   ) -> (Pose:Component Space Pose Link Structure) {is=title FortAnimGraphNode_TransformRollerTreads, alphaBoolBlend {blendOption=Linear}}
   Output Pose(Result=Inertialization!(Source:Pose Link Structure=Component To Local(ComponentPose=@n4)) -> (Pose:Pose Link Structure) {is=title AnimGraphNode_Inertialization}) {layerGroup=None}
   n5 = "Transform (Modify) Bone - Bone: hatch"!(
    ComponentPose=Local To Component(LocalPose=Use cached pose 'ExcludeTurret'())
    Alpha=Get TurretYawAlpha() -> (TurretYawAlpha:float)
    Rotation_Roll:float=0
    Rotation_Pitch:float=0
    Rotation_Yaw:float=Get TurretYaw() -> (TurretYaw:float)
   ) -> (Pose) {
    rotationMode=BMM_Replace
    rotationSpace=BCS_BoneSpace
    translationSpace=BCS_WorldSpace
    alphaBoolBlend {blendOption=Linear, blendOutTime=0.3}
    boneToModify {boneName=hatch}
   }
   n6 = "Transform (Modify) Bone - Bone: gun_base"!(ComponentPose=@n5, Alpha=1, Rotation_Roll:float=0, Rotation_Pitch:float=Get TurretPitch() -> (TurretPitch:float), Rotation_Yaw:float=0) -> (Pose) {
    rotationMode=BMM_Replace
    rotationSpace=BCS_BoneSpace
    translationSpace=BCS_WorldSpace
    alphaBoolBlend {blendOption=Linear, blendOutTime=0.3}
    boneToModify {boneName=gun_base}
   }
   Save cached pose 'IncludeTurret'(Pose=Component To Local(ComponentPose=@n6)) {cacheName=IncludeTurret}
   n7 = "Sequence Player (additive) 'Tank_Overheat_Pose_Vehicle'"() {sequence=/Tank/Animations/Vehicle/Tank_Overheat_Pose_Vehicle}
   n8 = Blend Poses by bool(
    BlendPose_0=Sequence Player 'Tank_HatchShieldOffPose_Vehicle'() {sequence=/Tank/Animations/Vehicle/Tank_HatchShieldOffPose_Vehicle}
    BlendPose_1=Blend Poses by bool(
     BlendPose_0=Sequence Player 'Tank_HatchDriverOccupiedPose_Vehicle'() {sequence=/Tank/Animations/Vehicle/Tank_HatchDriverOccupiedPose_Vehicle}
     BlendPose_1=Sequence Player 'Tank_HatchNoDriverPose_Vehicle'() {sequence=/Tank/Animations/Vehicle/Tank_HatchNoDriverPose_Vehicle}
     BlendTime_0=0
     BlendTime_1=0
     bActiveValue=Get bHasDriver() -> (bHasDriver:bool)
    ) {blendType=Linear}
    BlendTime_0=0
    BlendTime_1=0
    bActiveValue=Get bDriverCoverOnCooldown() -> (bDriverCoverOnCooldown:bool)
   ) {blendType=Linear}
   n9 = Blendspace Player 'Tank_Bars_BS_Vehicle'(X=Get LeftTreadPower() -> (LeftTreadPower:float), Y=Get RightTreadPower() -> (RightTreadPower:float)) {blendSpace=/Tank/Animations/Vehicle/Tank_Bars_BS_Vehicle}
   AnimGraphNode_StateMachine_0 = CannonFire_1() {is=title AnimGraphNode_StateMachine}
   n10 = Layered blend per bone(
    BasePose=Apply Additive(Base=Apply Additive(Base=@n8, Additive=@n9) {alphaBoolBlend {blendOption=Linear}}, Additive=@AnimGraphNode_StateMachine_0) {alphaBoolBlend {blendOption=Linear}}
    BlendPoses_0=Blend Poses by bool(
     BlendPose_0=Sequence Player 'TankVehicle_Turret_EquippedPose'() {sequence=/Tank/Animations/Vehicle/TankVehicle_Turret_EquippedPose}
     BlendPose_1=Sequence Player 'TankVehicle_Turret_NonEquippedPose_Tank_Veh'() {sequence=/Tank/Animations/Vehicle/TankVehicle_Turret_NonEquippedPose_Tank_Veh}
     BlendTime_0=0
     BlendTime_1=0
     bActiveValue=Get bHasPassenger() -> (bHasPassenger:bool)
    ) {blendType=Linear}
   ) {layerSetup [{branchFilters [{boneName=gun_arm_bott}, {boneName=hatch_hinge_01_l}, {boneName=hatch_hinge_01_r}]}]}
   n11 = Slot 'DefaultSlot'(Source=@n10)
   n12 = Get bEngineOverheated() -> (bEngineOverheated:bool)
   n13 = "Transform (Modify) Bone - Bone: turret"!(
    ComponentPose=Local To Component(LocalPose=Apply Additive!(Base=@n11, Additive=@n7, bAlphaBoolEnabled:bool=@n12) -> (Pose) {alphaInputType=Bool, alphaBoolBlend {blendInTime=2, blendOption=Linear, blendOutTime=2}})
    Alpha=1
    Rotation_Roll:float=0
    Rotation_Pitch:float=0
    Rotation_Yaw:float=Get CannonYaw() -> (CannonYaw:float)
   ) -> (Pose) {
    rotationMode=BMM_Replace
    rotationSpace=BCS_BoneSpace
    translationSpace=BCS_WorldSpace
    alphaBoolBlend {blendOption=Linear, blendOutTime=0.3}
    boneToModify {boneName=turret}
   }
   n14 = "Transform (Modify) Bone - Bone: cannon_main_01"!(ComponentPose=@n13, Alpha=1, Rotation_Roll:float=0, Rotation_Pitch:float=Get CannonPitch() -> (CannonPitch:float), Rotation_Yaw:float=0) -> (Pose) {
    rotationMode=BMM_Replace
    rotationSpace=BCS_BoneSpace
    translationSpace=BCS_WorldSpace
    alphaBoolBlend {blendOption=Linear, blendOutTime=0.3}
    boneToModify {boneName=cannon_main_01}
   }
   Save cached pose 'ExcludeTurret'(Pose=Component To Local(ComponentPose=@n14)) {cacheName=ExcludeTurret}
  ]}
 CannonFire_1 {
  entry=Default
  path=AnimGraph.AnimGraphNode_StateMachine_0.CannonFire_1
  ids=id
  nodes [
   AnimStateNode_0 = Fired() {is=title AnimStateNode}
   AnimStateNode_1 = OnCooldown() {is=title AnimStateNode}
   AnimStateNode_2 = Ready() {is=title AnimStateNode}
   AnimStateNode_3 = Default() {is=title AnimStateNode}
   AnimStateTransitionNode_2 = AnimStateTransitionNode() {bAutomaticRuleBasedOnSequencePlayerInState=true, logicType=TLT_Inertialization}
   AnimStateTransitionNode_3 = AnimStateTransitionNode() {crossfadeDuration=0.1, logicType=TLT_Inertialization}
   AnimStateTransitionNode_4 = AnimStateTransitionNode() {crossfadeDuration=0.033, logicType=TLT_Inertialization}
   AnimStateTransitionNode_5 = AnimStateTransitionNode() {bAutomaticRuleBasedOnSequencePlayerInState=true, logicType=TLT_Inertialization}
  ]
  links [
   {
    dst=Fired
    rule=AnimGraph.AnimGraphNode_StateMachine_0.CannonFire_1.AnimStateTransitionNode_4.Transition
    src=Default
    via=AnimStateTransitionNode_4
   }
   {
    dst=OnCooldown
    rule=AnimGraph.AnimGraphNode_StateMachine_0.CannonFire_1.AnimStateTransitionNode_2.Transition
    src=Fired
    via=AnimStateTransitionNode_2
   }
   {
    dst=Ready
    rule=AnimGraph.AnimGraphNode_StateMachine_0.CannonFire_1.AnimStateTransitionNode_3.Transition
    src=OnCooldown
    via=AnimStateTransitionNode_3
   }
   {
    dst=Default
    rule=AnimGraph.AnimGraphNode_StateMachine_0.CannonFire_1.AnimStateTransitionNode_5.Transition
    src=Ready
    via=AnimStateTransitionNode_5
   }
  ]
 }
 Fired {
  path=AnimGraph.AnimGraphNode_StateMachine_0.CannonFire_1.AnimStateNode_0.Fired
  nodes [Output Animation Pose(Result="Sequence Player (additive) 'Tank_CannonFire_Vehicle'"() {bLoopAnimation=false, sequence=/Tank/Animations/Vehicle/Tank_CannonFire_Vehicle})]
 }
 OnCooldown {
  path=AnimGraph.AnimGraphNode_StateMachine_0.CannonFire_1.AnimStateNode_1.OnCooldown
  nodes [Output Animation Pose(Result="Sequence Player (additive) 'Tank_CannonFire_OnCD_Vehicle'"() {sequence=/Tank/Animations/Vehicle/Tank_CannonFire_OnCD_Vehicle})]
 }
 Ready {
  path=AnimGraph.AnimGraphNode_StateMachine_0.CannonFire_1.AnimStateNode_2.Ready
  nodes [Output Animation Pose(Result="Sequence Player (additive) 'Tank_CannonFire_Ready_Vehicle'"() {bLoopAnimation=false, sequence=/Tank/Animations/Vehicle/Tank_CannonFire_Ready_Vehicle})]
 }
 Default {
  path=AnimGraph.AnimGraphNode_StateMachine_0.CannonFire_1.AnimStateNode_3.Default
  nodes [Output Animation Pose(Result=Additive Identity Pose())]
 }
 Transition {path=AnimGraph.AnimGraphNode_StateMachine_0.CannonFire_1.AnimStateTransitionNode_2.Transition, nodes [Result()]}
 Transition {
  path=AnimGraph.AnimGraphNode_StateMachine_0.CannonFire_1.AnimStateTransitionNode_3.Transition
  nodes [Result(bCanEnterTransition=Get bIsReady() -> (bIsReady:bool))]
 }
 Transition {
  path=AnimGraph.AnimGraphNode_StateMachine_0.CannonFire_1.AnimStateTransitionNode_4.Transition
  nodes [Result(bCanEnterTransition=Get bIsFired() -> (bIsFired:bool))]
 }
 Transition {path=AnimGraph.AnimGraphNode_StateMachine_0.CannonFire_1.AnimStateTransitionNode_5.Transition, nodes [Result()]}
 EventGraph {links [], nodes []}
]
variables []
