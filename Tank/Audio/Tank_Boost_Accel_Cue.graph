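# Sound cue for the tank's boost acceleration, layering a one-shot accel sound with a looping boost engine sound.
# A mixer combines a random pick of two accel waves with the loop at reduced volume.
# The loop fades in over three seconds through an enveloper, with pitch modulated continuously by the Forward parameter.
# Output overrides to the vehicle engine sound class and uses tank movement attenuation.
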SoundCue /Tank/Audio/Tank_Boost_Accel_Cue
root=SoundClass_1
nodes [
 {
  id=Enveloper_0
  children [WavePlayer_6]
  properties {pitchCurve {editorCurveData {keys [{time=0}]}}, volumeCurve {editorCurveData {keys [<at 0 0.01>,<at 0.3428 0.2272>,<at 3 1>]}}}
 }
 {id=Mixer_1, children [Random_1, ModulatorContinuous_0], properties {inputVolume [1,0.5]}}
 {id=ModulatorContinuous_0, children [Enveloper_0], properties {pitchModulationParams {parameterName=Forward}}}
 {id=Random_1, children [WavePlayer_4,WavePlayer_5], properties {weights [1,1]}}
 {id=SoundClass_1, children [Mixer_1], properties {soundClassOverride=&Fort_Vehicles_Engine}}
 {id=WavePlayer_4, properties {soundWaveAssetPtr=&/Veh_Tank_Engine_Boost_Accel_01}}
 {id=WavePlayer_5, properties {soundWaveAssetPtr=&/Veh_Tank_Engine_Boost_Accel_02}}
 {id=WavePlayer_6, properties {bLooping=true, soundWaveAssetPtr=&/Veh_Tank_Engine_Boost_Loop_01}}
]
properties {
 attenuationSettings=/Tank/Audio/Settings/Tank_Movement_Att
 maxDistance=28024
 soundClassObject=&Fort_Vehicles_Exterior
 volumeMultiplier=0.708
}
