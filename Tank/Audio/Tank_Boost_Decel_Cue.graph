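# Sound cue for the tank engine's boost deceleration sound.
# A Random node picks one of two boost decel waves at equal weight, then routes it through a SoundClass node that overrides the class to Fort_Vehicles_Engine.
# The cue itself is assigned the Fort_Vehicles_Exterior class, uses the shared tank movement attenuation, and plays at reduced volume.
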SoundCue /Tank/Audio/Tank_Boost_Decel_Cue
root=SoundClass_0
nodes [
 {id=Random_0, children [WavePlayer_0,WavePlayer_1], properties {weights [1,1]}}
 {id=SoundClass_0, children [Random_0], properties {soundClassOverride=&Fort_Vehicles_Engine}}
 {id=WavePlayer_0, properties {soundWaveAssetPtr=&/Veh_Tank_Engine_Boost_Decel_02}}
 {id=WavePlayer_1, properties {soundWaveAssetPtr=&/Veh_Tank_Engine_Boost_Decel_01}}
]
properties {
 attenuationSettings=/Tank/Audio/Settings/Tank_Movement_Att
 maxDistance=28024
 soundClassObject=&Fort_Vehicles_Exterior
 volumeMultiplier=0.531
}
