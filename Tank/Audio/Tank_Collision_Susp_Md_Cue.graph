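# Sound cue for medium tank suspension bounce impacts, with a dedicated collision concurrency setting.
# A quality-level node sends the first two levels to a random pool of eight waves and the last two to four.
# Each pool feeds IsLocalPassenger branches, giving local passengers 2D attenuation and everyone else, or an unset parameter, close attenuation.
# Each mixed pair passes through a modulator that randomly lowers pitch for variation.
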SoundCue /Tank/Audio/Tank_Collision_Susp_Md_Cue
root=QualityLevel_2
nodes [
 {id=Attenuation_0, children [Branch_2], properties {attenuationSettings=/Tank/Audio/Settings/Tank_Movement_Local_2D_Att}}
 {id=Attenuation_1, children [Branch_3], properties {attenuationSettings=/Tank/Audio/Settings/Tank_Movement_Close_Att}}
 {id=Attenuation_2, children [Branch_0], properties {attenuationSettings=&Tank_Movement_Local_2D_Att}}
 {id=Attenuation_3, children [Branch_1], properties {attenuationSettings=&Tank_Movement_Close_Att}}
 {id=Branch_0, children [Random_0,null,null], properties {boolParameterName=IsLocalPassenger}}
 {id=Branch_1, children [null,Random_0,Random_0], properties {boolParameterName=IsLocalPassenger}}
 {id=Branch_2, children [Random_1,null,null], properties {boolParameterName=IsLocalPassenger}}
 {id=Branch_3, children [null,Random_1,Random_1], properties {boolParameterName=IsLocalPassenger}}
 {id=Mixer_0, children [Attenuation_0, Attenuation_1], properties {inputVolume [1,1]}}
 {id=Mixer_1, children [Attenuation_2, Attenuation_3], properties {inputVolume [1,1]}}
 {id=Modulator_0, children [Mixer_0], properties {pitchMax=1, pitchMin=0.65}}
 {id=Modulator_2, children [Mixer_1], properties {pitchMax=1, pitchMin=0.65}}
 {id=QualityLevel_2, children [Modulator_2,Modulator_2,Modulator_0,Modulator_0]}
 {
  id=Random_0
  children [WavePlayer_0,WavePlayer_1,WavePlayer_2,WavePlayer_3,WavePlayer_4,WavePlayer_5,WavePlayer_6,WavePlayer_7]
  properties {weights [1,1,1,1,1,1,1,1]}
 }
 {id=Random_1, children [WavePlayer_8, WavePlayer_9, WavePlayer_10, WavePlayer_11], properties {weights [1,1,1,1]}}
 {id=WavePlayer_0, properties {soundWaveAssetPtr=&/Veh_Tank_Susp_Bounce_Med_01}}
 {id=WavePlayer_1, properties {soundWaveAssetPtr=&/Veh_Tank_Susp_Bounce_Med_02}}
 {id=WavePlayer_10, properties {soundWaveAssetPtr=&/Veh_Tank_Susp_Bounce_Med_03}}
 {id=WavePlayer_11, properties {soundWaveAssetPtr=&/Veh_Tank_Susp_Bounce_Med_04}}
 {id=WavePlayer_2, properties {soundWaveAssetPtr=&Veh_Tank_Susp_Bounce_Med_03}}
 {id=WavePlayer_3, properties {soundWaveAssetPtr=&Veh_Tank_Susp_Bounce_Med_04}}
 {id=WavePlayer_4, properties {soundWaveAssetPtr=&/Veh_Tank_Susp_Bounce_Med_05}}
 {id=WavePlayer_5, properties {soundWaveAssetPtr=&/Veh_Tank_Susp_Bounce_Med_06}}
 {id=WavePlayer_6, properties {soundWaveAssetPtr=&/Veh_Tank_Susp_Bounce_Med_07}}
 {id=WavePlayer_7, properties {soundWaveAssetPtr=&/Veh_Tank_Susp_Bounce_Med_08}}
 {id=WavePlayer_8, properties {soundWaveAssetPtr=&Veh_Tank_Susp_Bounce_Med_01}}
 {id=WavePlayer_9, properties {soundWaveAssetPtr=&Veh_Tank_Susp_Bounce_Med_02}}
]
properties {
 attenuationSettings=/VehicleAudio/Settings/Vehicle_Movement_Close_Detail_Att
 maxDistance=12000
 priority=0.8
 soundClassObject=&Fort_Vehicles_Exterior
 volumeMultiplier=0.4248
 concurrencySet [/Tank/Audio/Settings/Tank_Collision_Susp_Conc]
}
