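# Sound cue for a critically damaged tank, pairing a one-shot damage impact with a looping fire sound.
# A mixer layers a random pick of two impacts in the large explosives sound class with the quieter fire loop in the exterior vehicle class.
# An envelope fades the fire in over about a second, while the modulated mix ducks between 20 and 40 seconds under an outer envelope, using close tank-movement attenuation.
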SoundCue /Tank/Audio/Tank_Damaged_Engine_Loop_Cue
root=Enveloper_0
nodes [
 {
  id=Enveloper_0
  children [Modulator_0]
  properties {pitchCurve {editorCurveData {keys [{time=0}]}}, volumeCurve {editorCurveData {keys [<at 0 1>,<at 20 1>,<at 40 0.3>]}}}
 }
 {
  id=Enveloper_1
  children [WavePlayer_1]
  properties {pitchCurve {editorCurveData {keys [{time=0}]}}, volumeCurve {editorCurveData {keys [<at 0 0>,<at 0.8586 0.2529>,<at 1.25 1>]}}}
 }
 {id=Mixer_2, children [SoundClass_0,SoundClass_1], properties {inputVolume [1,0.6]}}
 {id=Modulator_0, children [Mixer_2]}
 {id=Random_0, children [WavePlayer_2,WavePlayer_3], properties {bShouldExcludeFromBranchCulling=true, weights [1,1]}}
 {id=SoundClass_0, children [Random_0], properties {soundClassOverride=/Game/Sounds/FortSoundClassesAndModes/Fort_Explosives_Large_PSM}}
 {id=SoundClass_1, children [Enveloper_1], properties {soundClassOverride=&Fort_Vehicles_Exterior}}
 {id=WavePlayer_1, properties {bLooping=true, soundWaveAssetPtr=&/Veh_Tank_Damage_Fire_Loop_01}}
 {id=WavePlayer_2, properties {soundWaveAssetPtr=&/Veh_Tank_Impact_CriticalDamage_01}}
 {id=WavePlayer_3, properties {soundWaveAssetPtr=&/Veh_Tank_Impact_CriticalDamage_02}}
]
properties {attenuationSettings=/Tank/Audio/Settings/Tank_Movement_Close_Att, maxDistance=6024}
