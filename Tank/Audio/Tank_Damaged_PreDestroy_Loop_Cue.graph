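# Sound cue for a tank's damaged, pre-destroy state, mixing a critical-damage impact with looping engine fire and engine-disabled warning layers.
# The impact randomly picks one of two waves under an explosives sound class and wider movement attenuation.
# The fire loop's pitch and volume follow a Fade parameter.
# The warning loop plays at doubled pitch with overridden attenuation adding distance low-pass and occlusion; both loops use the vehicle exterior class.
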SoundCue /Tank/Audio/Tank_Damaged_PreDestroy_Loop_Cue
root=Mixer_0
nodes [
 {
  id=Attenuation_0
  children [SoundClass_2]
  properties {
   attenuationSettings=/Tank/Audio/Settings/Tank_Movement_Close_Att
   bOverrideAttenuation=true
   attenuationOverrides {
    bAttenuateWithLPF=true
    bEnableOcclusion=true
    falloffDistance=2500
    lPFFrequencyAtMax=1000
    lPFRadiusMax=2000
    lPFRadiusMin=500
    nonSpatializedRadiusStart=400
    occlusionLowPassFilterFrequency=1000
    occlusionTraceChannel=ECC_GameTraceChannel11
    occlusionVolumeAttenuation=0.7
    attenuationShapeExtents {x=500}
   }
  }
 }
 {id=Attenuation_1, children [Modulator_3], properties {attenuationSettings=&Tank_Movement_Close_Att}}
 {id=Attenuation_2, children [Modulator_2], properties {attenuationSettings=/Tank/Audio/Settings/Tank_Movement_Att}}
 {id=Mixer_0, children [Attenuation_2, Attenuation_1, Attenuation_0], properties {inputVolume [0.75,0.75,1]}}
 {
  id=ModulatorContinuous_0
  children [SoundClass_1]
  properties {pitchModulationParams {minOutput=0.8, parameterName=Fade}, volumeModulationParams {maxInput=0.1, minOutput=0.5, parameterName=Fade}}
 }
 {id=Modulator_1, children [WavePlayer_1], properties {pitchMax=2, pitchMin=2}}
 {id=Modulator_2, children [SoundClass_0]}
 {id=Modulator_3, children [ModulatorContinuous_0]}
 {id=Random_0, children [WavePlayer_2,WavePlayer_0], properties {weights [1,1]}}
 {id=SoundClass_0, children [Random_0], properties {soundClassOverride=/Game/Sounds/FortSoundClassesAndModes/Fort_Explosives_Large_PSM}}
 {id=SoundClass_1, children [WavePlayer_4], properties {soundClassOverride=&Fort_Vehicles_Exterior}}
 {id=SoundClass_2, children [Modulator_1], properties {soundClassOverride=&Fort_Vehicles_Exterior}}
 {id=WavePlayer_0, properties {soundWaveAssetPtr=&/Veh_Tank_Impact_CriticalDamage_02}}
 {id=WavePlayer_1, properties {bLooping=true, soundWaveAssetPtr=&/Veh_Tank_Warning_Engine_Disabled_01}}
 {id=WavePlayer_2, properties {soundWaveAssetPtr=&/Veh_Tank_Impact_CriticalDamage_01}}
 {id=WavePlayer_4, properties {bLooping=true, soundWaveAssetPtr=/VehicleAudio/Assets/Damage/Veh_Valet_Damage_Engine_Fire_Loop_01}}
]
properties {maxDistance=28024, volumeMultiplier=0.5}
