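# Sound cue for the damaged tank, layering a critical-damage impact over a looping engine smoke sound, with close-range tank attenuation.
# The smoke loop fades in under the vehicle-exterior sound class, mixed with random impacts under the large-explosives class.
# A quality switch gives the first two levels a four-sample impact pool and the last two a two-sample pool.
# The mix passes through a modulator, then an envelope that gradually lowers volume after a long hold.
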SoundCue /Tank/Audio/Tank_Damaged_Smoke_Loop_Cue
root=Enveloper_0
nodes [
 {
  id=Enveloper_0
  children [Modulator_0]
  properties {pitchCurve {editorCurveData {keys [{time=0}]}}, volumeCurve {editorCurveData {keys [<at 0 1>,<at 20 1>,<at 40 0.3>]}}}
 }
 {
  id=Enveloper_1
  children [WavePlayer_0]
  properties {pitchCurve {editorCurveData {keys [{time=0}]}}, volumeCurve {editorCurveData {keys [<at 0 0>,<at 0.8586 0.2529>,<at 1.25 1>]}}}
 }
 {id=Mixer_2, children [SoundClass_1,SoundClass_0], properties {inputVolume [0.75,0.75]}}
 {id=Modulator_0, children [Mixer_2]}
 {id=QualityLevel_0, children [Random_0,Random_0,Random_1,Random_1]}
 {id=Random_0, children [WavePlayer_2,WavePlayer_3,WavePlayer_4,WavePlayer_5], properties {weights [1,1,1,1]}}
 {id=Random_1, children [WavePlayer_1,WavePlayer_6], properties {weights [1,1]}}
 {id=SoundClass_0, children [QualityLevel_0], properties {soundClassOverride=/Game/Sounds/FortSoundClassesAndModes/Fort_Explosives_Large_PSM}}
 {id=SoundClass_1, children [Enveloper_1], properties {soundClassOverride=&Fort_Vehicles_Exterior}}
 {id=WavePlayer_0, properties {bLooping=true, soundWaveAssetPtr=/VehicleAudio/Assets/Damage/Veh_Valet_Damage_Engine_Smoke_Loop_01}}
 {id=WavePlayer_1, properties {soundWaveAssetPtr=&/Veh_Tank_Impact_CriticalDamage_01}}
 {id=WavePlayer_2, properties {soundWaveAssetPtr=&Veh_Tank_Impact_CriticalDamage_01}}
 {id=WavePlayer_3, properties {soundWaveAssetPtr=&/Veh_Tank_Impact_CriticalDamage_02}}
 {id=WavePlayer_4, properties {soundWaveAssetPtr=&/Veh_Tank_Impact_CriticalDamage_03}}
 {id=WavePlayer_5, properties {soundWaveAssetPtr=&/Veh_Tank_Impact_CriticalDamage_04}}
 {id=WavePlayer_6, properties {soundWaveAssetPtr=&Veh_Tank_Impact_CriticalDamage_02}}
]
properties {attenuationSettings=/Tank/Audio/Settings/Tank_Movement_Close_Att, maxDistance=6024, volumeMultiplier=0.5}
