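# Sound cue for the tank's destruction, built from reused Valet vehicle explosion waves in close, medium and distant layers.
# A top mixer blends the three layers, each with its own distance attenuation, sound class override and random choice of two variants.
# The close layer also mixes in a looping engine-fire sound whose pitch and volume follow a continuous "Fade" parameter, all passing through a modulator.
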SoundCue /Tank/Audio/Tank_Destroyed_Cue
root=Mixer_4
nodes [
 {
  id=Attenuation_0
  children [Random_2]
  properties {
   attenuationSettings=/Game/Sounds/Attenuation/Fort_Explosion_Distant_Athena_Att
   attenuationOverrides {falloffDistance=3000, attenuationShapeExtents {x=1000}}
  }
 }
 {
  id=Attenuation_4
  children [Mixer_0]
  properties {attenuationSettings=/Game/Sounds/Attenuation/Fort_Explosion_Close_Athena_Att, attenuationOverrides {falloffDistance=1600}}
 }
 {
  id=Attenuation_5
  children [Random_1]
  properties {
   attenuationSettings=/Game/Sounds/Attenuation/Fort_Explosion_Medium_Athena_Att
   attenuationOverrides {falloffDistance=3000, attenuationShapeExtents {x=1000}}
  }
 }
 {id=Mixer_0, children [ModulatorContinuous_0, Random_0], properties {inputVolume [1,1]}}
 {id=Mixer_4, children [SoundClass_3,SoundClass_2,SoundClass_0], properties {inputVolume [0.7,2,2]}}
 {
  id=ModulatorContinuous_0
  children [WavePlayer_6]
  properties {pitchModulationParams {minOutput=0.8, parameterName=Fade}, volumeModulationParams {parameterName=Fade}}
 }
 {id=Modulator_0, children [Attenuation_4]}
 {id=Random_0, children [WavePlayer_0,WavePlayer_1], properties {bShouldExcludeFromBranchCulling=true, weights [1,1]}}
 {id=Random_1, children [WavePlayer_2,WavePlayer_3], properties {bShouldExcludeFromBranchCulling=true, weights [1,1]}}
 {id=Random_2, children [WavePlayer_4,WavePlayer_5], properties {weights [1,1]}}
 {
  id=SoundClass_0
  children [Attenuation_0]
  properties {soundClassOverride=&Fort_WeaponRanged_Fire_Distant}
 }
 {
  id=SoundClass_2
  children [Attenuation_5]
  properties {soundClassOverride=&Fort_WeaponRanged_Fire_Distant}
 }
 {id=SoundClass_3, children [Modulator_0], properties {soundClassOverride=/Game/Sounds/FortSoundClassesAndModes/Fort_Explosives_Large_PSM}}
 {id=WavePlayer_0, properties {soundWaveAssetPtr=/VehicleAudio/Assets/Damage/Veh_Valet_Damage_Explode_Close_01}}
 {id=WavePlayer_1, properties {soundWaveAssetPtr=/VehicleAudio/Assets/Damage/Veh_Valet_Damage_Explode_Close_02}}
 {id=WavePlayer_2, properties {soundWaveAssetPtr=/VehicleAudio/Assets/Damage/Veh_Valet_Damage_Explode_Med_01}}
 {id=WavePlayer_3, properties {soundWaveAssetPtr=/VehicleAudio/Assets/Damage/Veh_Valet_Damage_Explode_Med_02}}
 {id=WavePlayer_4, properties {soundWaveAssetPtr=/VehicleAudio/Assets/Damage/Veh_Valet_Damage_Explode_Distant_01}}
 {id=WavePlayer_5, properties {soundWaveAssetPtr=/VehicleAudio/Assets/Damage/Veh_Valet_Damage_Explode_Distant_02}}
 {id=WavePlayer_6, properties {bLooping=true, soundWaveAssetPtr=/VehicleAudio/Assets/Damage/Veh_Valet_Damage_Engine_Fire_Loop_01}}
]
properties {maxDistance=27500, volumeMultiplier=1}
