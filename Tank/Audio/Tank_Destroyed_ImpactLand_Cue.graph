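# Sound cue for a destroyed tank's landing impact, built from heavy Valet vehicle collision waves.
# A Modulator feeds a Quality Level switch: the first two slots distance-crossfade randomized near and distant sets of five waves.
# The last two slots use a reduced random set of three near waves with their own close-detail attenuation.
# A concurrency override limits playback to one instance per owner.
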SoundCue /Tank/Audio/Tank_Destroyed_ImpactLand_Cue
root=Modulator_1
nodes [
 {id=Attenuation_0, children [Random_1], properties {attenuationSettings=/VehicleAudio/Settings/Vehicle_Movement_Close_Detail_Att}}
 {
  id=DistanceCrossFade_1
  children [Random_0,Random_2]
  properties {crossFadeInput [
    {fadeInDistanceEnd=0.1, fadeInDistanceStart=0, fadeOutDistanceEnd=5000, fadeOutDistanceStart=3500, volume=1}
    {fadeInDistanceEnd=4500, fadeInDistanceStart=4000, fadeOutDistanceEnd=14000, fadeOutDistanceStart=11000, volume=1}
   ]}
 }
 {id=Modulator_1, children [QualityLevel_0]}
 {id=QualityLevel_0, children [DistanceCrossFade_1, DistanceCrossFade_1, Attenuation_0, Attenuation_0]}
 {id=Random_0, children [WavePlayer_3,WavePlayer_4,WavePlayer_5,WavePlayer_6,WavePlayer_7], properties {weights [1,1,1,1,1]}}
 {id=Random_1, children [WavePlayer_0,WavePlayer_1,WavePlayer_2], properties {bShouldExcludeFromBranchCulling=true, weights [1,1,1]}}
 {id=Random_2, children [WavePlayer_19, WavePlayer_20, WavePlayer_21, WavePlayer_22, WavePlayer_23], properties {weights [1,1,1,1,1]}}
 {id=WavePlayer_0, properties {soundWaveAssetPtr=/VehicleAudio/Assets/Collision/Veh_Valet_Impt_Hv_01}}
 {id=WavePlayer_1, properties {soundWaveAssetPtr=/VehicleAudio/Assets/Collision/Veh_Valet_Impt_Hv_02}}
 {id=WavePlayer_19, properties {soundWaveAssetPtr=/VehicleAudio/Assets/Collision/Veh_Valet_Impt_Hv_Dist_01}}
 {id=WavePlayer_2, properties {soundWaveAssetPtr=/VehicleAudio/Assets/Collision/Veh_Valet_Impt_Hv_03}}
 {id=WavePlayer_20, properties {soundWaveAssetPtr=/VehicleAudio/Assets/Collision/Veh_Valet_Impt_Hv_Dist_02}}
 {id=WavePlayer_21, properties {soundWaveAssetPtr=/VehicleAudio/Assets/Collision/Veh_Valet_Impt_Hv_Dist_03}}
 {id=WavePlayer_22, properties {soundWaveAssetPtr=/VehicleAudio/Assets/Collision/Veh_Valet_Impt_Hv_Dist_04}}
 {id=WavePlayer_23, properties {soundWaveAssetPtr=/VehicleAudio/Assets/Collision/Veh_Valet_Impt_Hv_Dist_05}}
 {id=WavePlayer_3, properties {soundWaveAssetPtr=&Veh_Valet_Impt_Hv_01}}
 {id=WavePlayer_4, properties {soundWaveAssetPtr=&Veh_Valet_Impt_Hv_02}}
 {id=WavePlayer_5, properties {soundWaveAssetPtr=&Veh_Valet_Impt_Hv_03}}
 {id=WavePlayer_6, properties {soundWaveAssetPtr=/VehicleAudio/Assets/Collision/Veh_Valet_Impt_Hv_04}}
 {id=WavePlayer_7, properties {soundWaveAssetPtr=/VehicleAudio/Assets/Collision/Veh_Valet_Impt_Hv_05}}
]
properties {
 attenuationSettings=/Tank/Audio/Settings/Tank_Movement_Close_Att
 bOverrideConcurrency=true
 maxDistance=14000
 soundClassObject=&Fort_Vehicles_Exterior
 volumeMultiplier=1
 concurrencyOverrides {bLimitToOwner=true, maxCount=1, resolutionRule=StopFarthestThenPreventNew, retriggerTime=1}
}
