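# Sound cue for a tank becoming disabled, mixing a critical-damage impact with a looping engine-disabled warning.
# A random node picks one of two impact waves to play at once, while the looping warning starts after a fixed two-second delay at lower mixer volume.
# Uses tank close-movement attenuation, the weapon foley sound class, and concurrency that blocks overlapping instances.
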SoundCue /Tank/Audio/Tank_Disabled_Cue
root=Mixer_0
nodes [
 {id=Delay_0, children [WavePlayer_1], properties {delayMax=2, delayMin=2}}
 {id=Mixer_0, children [Random_0,Delay_0], properties {inputVolume [1,0.35]}}
 {id=Random_0, children [WavePlayer_2,WavePlayer_3], properties {weights [1,1]}}
 {id=WavePlayer_1, properties {bLooping=true, soundWaveAssetPtr=&/Veh_Tank_Warning_Engine_Disabled_01}}
 {id=WavePlayer_2, properties {soundWaveAssetPtr=&/Veh_Tank_Impact_CriticalDamage_03}}
 {id=WavePlayer_3, properties {soundWaveAssetPtr=&/Veh_Tank_Impact_CriticalDamage_04}}
]
properties {
 attenuationSettings=/Tank/Audio/Settings/Tank_Movement_Close_Att
 bOverrideConcurrency=true
 maxDistance=6024
 soundClassObject=&Fort_WeaponFoley
 volumeMultiplier=0.501
 concurrencyOverrides {maxCount=1, resolutionRule=PreventNew, retriggerTime=1}
}
