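# Sound cue for activating the tank's driver cover, playing a driver hatch opening sound.
# A single wave player feeds the output directly, with the driver cover attenuation asset and the weapon foley sound class.
# A concurrency override allows one instance per owner and blocks new plays while it is active, so retriggers do not stack.
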SoundCue /Tank/Audio/Tank_DriverCover_Activate_Cue
root=WavePlayer_0
nodes [{id=WavePlayer_0, properties {soundWaveAssetPtr=&/Veh_Tank_Hatch_Driver_Open_01}}]
properties {
 attenuationSettings=/Tank/Audio/Settings/Tank_DriverCover_Att
 bOverrideConcurrency=true
 maxDistance=2512
 soundClassObject=&Fort_WeaponFoley
 volumeMultiplier=1.002
 concurrencyOverrides {bLimitToOwner=true, maxCount=1, resolutionRule=PreventNew}
}
