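# Sound cue for the tank's driver cover deactivation, playing a driver hatch closing sound.
# A single wave player feeds the Veh_Tank_Hatch_Driver_Close_01 wave straight to the output, with no randomization or modulation.
# The cue uses the driver cover attenuation asset, a limited audible range, a slight volume boost and the Fort_WeaponFoley sound class.
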SoundCue /Tank/Audio/Tank_DriverCover_Deactivate_Cue
root=WavePlayer_1
nodes [{id=WavePlayer_1, properties {soundWaveAssetPtr=&/Veh_Tank_Hatch_Driver_Close_01}}]
properties {
 attenuationSettings=/Tank/Audio/Settings/Tank_DriverCover_Att
 maxDistance=2512
 soundClassObject=&Fort_WeaponFoley
 volumeMultiplier=1.253
}
