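# Sound cue for entering the tank's driver seat, judging by its name and its single seat-enter wave.
# A Wave Player feeds a root Modulator left at default settings, which slightly randomizes pitch and volume on each play.
# The cue plays at half volume through the vehicle interior sound class, using the tank's close movement attenuation.
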SoundCue /Tank/Audio/Tank_Driver_Enter_Cue
root=Modulator_0
nodes [{id=Modulator_0, children [WavePlayer_1]}, {id=WavePlayer_1, properties {soundWaveAssetPtr=&/Veh_Tank_Seat_Driver_Enter_01}}]
properties {
 attenuationSettings=/Tank/Audio/Settings/Tank_Movement_Close_Att
 maxDistance=6024
 soundClassObject=&Fort_Vehicles_Interior
 volumeMultiplier=0.5
}
