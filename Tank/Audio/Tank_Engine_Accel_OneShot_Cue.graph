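# Sound cue for the tank's one-shot engine acceleration, layering an engine accel sound with an exhaust burst.
# Each Mixer blends two Random picks, with the engine layer quieter than the exhaust.
# A root Quality Level node sends its first two slots to the full variation set and its last two to a reduced set.
# It uses tank movement attenuation, the vehicle engine sound class and a reduced overall volume.
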SoundCue /Tank/Audio/Tank_Engine_Accel_OneShot_Cue
root=QualityLevel_0
nodes [
 {id=Mixer_0, children [Random_0,Random_1], properties {inputVolume [0.7,1]}}
 {id=Mixer_1, children [Random_2,Random_3], properties {inputVolume [0.7,1]}}
 {id=QualityLevel_0, children [Mixer_0,Mixer_0,Mixer_1,Mixer_1]}
 {id=Random_0, children [WavePlayer_0,WavePlayer_1,WavePlayer_2,WavePlayer_3], properties {weights [1,1,1,1]}}
 {id=Random_1, children [WavePlayer_4,WavePlayer_5,WavePlayer_6,WavePlayer_7,WavePlayer_8], properties {weights [1,1,1,1,1]}}
 {id=Random_2, children [WavePlayer_9, WavePlayer_10, WavePlayer_18], properties {weights [1,1,1]}}
 {id=Random_3, children [WavePlayer_16, WavePlayer_17], properties {weights [1,1]}}
 {id=WavePlayer_0, properties {soundWaveAssetPtr=&/Veh_Tank_Engine_Accel_01}}
 {id=WavePlayer_1, properties {soundWaveAssetPtr=&/Veh_Tank_Engine_Accel_02}}
 {id=WavePlayer_10, properties {soundWaveAssetPtr=&Veh_Tank_Engine_Accel_02}}
 {id=WavePlayer_16, properties {soundWaveAssetPtr=&/Veh_Tank_Accel_Exhaust_Burst_01}}
 {id=WavePlayer_17, properties {soundWaveAssetPtr=&/Veh_Tank_Accel_Exhaust_Burst_02}}
 {id=WavePlayer_18, properties {soundWaveAssetPtr=&/Veh_Tank_Engine_Accel_03}}
 {id=WavePlayer_2, properties {soundWaveAssetPtr=&Veh_Tank_Engine_Accel_03}}
 {id=WavePlayer_3, properties {soundWaveAssetPtr=&/Veh_Tank_Engine_Accel_04}}
 {id=WavePlayer_4, properties {soundWaveAssetPtr=&/Veh_Tank_Accel_Exhaust_Burst_03}}
 {id=WavePlayer_5, properties {soundWaveAssetPtr=&/Veh_Tank_Accel_Exhaust_Burst_04}}
 {id=WavePlayer_6, properties {soundWaveAssetPtr=&/Veh_Tank_Accel_Exhaust_Burst_05}}
 {id=WavePlayer_7, properties {soundWaveAssetPtr=&Veh_Tank_Accel_Exhaust_Burst_01}}
 {id=WavePlayer_8, properties {soundWaveAssetPtr=&Veh_Tank_Accel_Exhaust_Burst_02}}
 {id=WavePlayer_9, properties {soundWaveAssetPtr=&Veh_Tank_Engine_Accel_01}}
]
properties {
 attenuationSettings=/Tank/Audio/Settings/Tank_Movement_Att
 maxDistance=28024
 priority=0.9
 soundClassObject=&Fort_Vehicles_Engine
 volumeMultiplier=0.496
}
