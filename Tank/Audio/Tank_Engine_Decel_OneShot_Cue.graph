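# Sound cue for a one-shot tank engine deceleration sound in the vehicle engine sound class.
# A quality-level root routes its first two levels to a random pick of seven decel waves and its last two to a pick of three.
# Each equally weighted random set passes through a half-volume mixer.
# Movement attenuation spatializes it, with high priority and a concurrency override allowing one instance at a time.
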SoundCue /Tank/Audio/Tank_Engine_Decel_OneShot_Cue
root=QualityLevel_0
nodes [
 {id=Mixer_0, children [Random_2], properties {inputVolume [0.5]}}
 {id=Mixer_1, children [Random_0], properties {inputVolume [0.5]}}
 {id=QualityLevel_0, children [Mixer_0,Mixer_0,Mixer_1,Mixer_1]}
 {id=Random_0, children [WavePlayer_1,WavePlayer_2,WavePlayer_3], properties {weights [1,1,1]}}
 {
  id=Random_2
  children [WavePlayer_4, WavePlayer_5, WavePlayer_6, WavePlayer_7, WavePlayer_8, WavePlayer_9, WavePlayer_10]
  properties {weights [1,1,1,1,1,1,1]}
 }
 {id=WavePlayer_1, properties {soundWaveAssetPtr=&/Veh_Tank_Engine_Decel_01}}
 {id=WavePlayer_10, properties {soundWaveAssetPtr=&/Veh_Tank_Engine_Decel_06}}
 {id=WavePlayer_2, properties {soundWaveAssetPtr=&/Veh_Tank_Engine_Decel_02}}
 {id=WavePlayer_3, properties {soundWaveAssetPtr=&/Veh_Tank_Engine_Decel_03}}
 {id=WavePlayer_4, properties {soundWaveAssetPtr=&/Veh_Tank_Engine_Decel_07}}
 {id=WavePlayer_5, properties {soundWaveAssetPtr=&Veh_Tank_Engine_Decel_01}}
 {id=WavePlayer_6, properties {soundWaveAssetPtr=&Veh_Tank_Engine_Decel_02}}
 {id=WavePlayer_7, properties {soundWaveAssetPtr=&Veh_Tank_Engine_Decel_03}}
 {id=WavePlayer_8, properties {soundWaveAssetPtr=&/Veh_Tank_Engine_Decel_04}}
 {id=WavePlayer_9, properties {soundWaveAssetPtr=&/Veh_Tank_Engine_Decel_05}}
]
properties {
 attenuationSettings=/Tank/Audio/Settings/Tank_Movement_Att
 maxDistance=28024
 priority=0.9
 soundClassObject=&Fort_Vehicles_Engine
 volumeMultiplier=0.496
 concurrencyOverrides {maxCount=1}
}
