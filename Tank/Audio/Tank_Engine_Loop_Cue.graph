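# Looping Sound Cue for the tank's engine, treads, cannon and turret movement audio, driven by vehicle state parameters.
# Continuous modulators scale idle, forward, reverse and turn loops, with Boost fading all but turn.
# Surface crossfades dirt and stone rolling, IsInWater swaps rolling for water loops, and airborne wind plays only for local passengers.
# Envelopers fade layers in, and IsLocalTurret switches turret sound between local 2D and close attenuation.
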SoundCue /Tank/Audio/Tank_Engine_Loop_Cue
root=Mixer_2
nodes [
 {id=Attenuation_0, children [SoundClass_1], properties {attenuationSettings=/Tank/Audio/Settings/Tank_Movement_Local_2D_Att}}
 {id=Attenuation_1, children [SoundClass_2], properties {attenuationSettings=/Tank/Audio/Settings/Tank_Movement_Close_Att}}
 {id=Attenuation_2, children [SoundClass_1], properties {attenuationSettings=&Tank_Movement_Close_Att}}
 {id=Attenuation_3, children [SoundClass_0], properties {attenuationSettings=&Tank_Movement_Close_Att}}
 {id=Branch_0, children [Attenuation_0, Attenuation_2, Attenuation_2], properties {boolParameterName=IsLocalTurret}}
 {id=Branch_1, children [ModulatorContinuous_26, null, null], properties {boolParameterName=IsLocalPassenger}}
 {
  id=Enveloper_0
  children [Mixer_0]
  properties {pitchCurve {editorCurveData {keys [{time=0}]}}, volumeCurve {editorCurveData {keys [<at 0 0>,<at 1 1>]}}}
 }
 {
  id=Enveloper_1
  children [ModulatorContinuous_8]
  properties {pitchCurve {editorCurveData {keys [{time=0}]}}, volumeCurve {editorCurveData {keys [<at 0 0>,<at 2 1>,<at 10 0.7>]}}}
 }
 {
  id=Enveloper_4
  children [WavePlayer_3]
  properties {pitchCurve {editorCurveData {keys [{time=0}]}}, volumeCurve {editorCurveData {keys [<at 0.01 0>,<at 4 1>]}}}
 }
 {
  id=Enveloper_5
  children [WavePlayer_1]
  properties {pitchCurve {editorCurveData {keys [{time=0}]}}, volumeCurve {editorCurveData {keys [<at 0.01 0>,<at 1.0096 0.1685>,<at 4 1>]}}}
 }
 {
  id=Enveloper_6
  children [Mixer_1]
  properties {volumeMax=1.3, volumeMin=1.3, pitchCurve {editorCurveData {keys [{time=0}]}}, volumeCurve {editorCurveData {keys [<at 0 0>,<at 1 1>]}}}
 }
 {id=Mixer_0, children [ModulatorContinuous_3, ModulatorContinuous_1], properties {inputVolume [1,1]}}
 {id=Mixer_1, children [ModulatorContinuous_6, ModulatorContinuous_4], properties {inputVolume [2,1]}}
 {
  id=Mixer_2
  children [Enveloper_1, ModulatorContinuous_7, ModulatorContinuous_9, ModulatorContinuous_5, Attenuation_3, Branch_0, Attenuation_1]
  properties {inputVolume [1,1.5,1.5,1,1.5,0.4,1]}
 }
 {id=Mixer_3, children [ModulatorContinuous_22, ModulatorContinuous_17], properties {inputVolume [0.7,1]}}
 {id=Mixer_6, children [ModulatorContinuous_16, ModulatorContinuous_19, Branch_1], properties {inputVolume [1,2,2]}}
 {id=Mixer_7, children [ModulatorContinuous_35, ModulatorContinuous_12, ModulatorContinuous_33], properties {inputVolume [0.7,0.5,0.5]}}
 {
  id=ModulatorContinuous_0
  children [Enveloper_5]
  properties {pitchModulationParams {default=0, minOutput=1, parameterName=Forward}, volumeModulationParams {default=0, parameterName=Forward}}
 }
 {
  id=ModulatorContinuous_1
  children [WavePlayer_7]
  properties {
   pitchModulationParams {default=0, minOutput=1, parameterName=CannonHorizontalMovement}
   volumeModulationParams {default=0, parameterName=CannonHorizontalMovement}
  }
 }
 {
  id=ModulatorContinuous_10
  children [ModulatorContinuous_13]
  properties {pitchModulationParams {default=0, minOutput=1}, volumeModulationParams {default=0, parameterName=HasDriver}}
 }
 {
  id=ModulatorContinuous_11
  children [Mixer_3]
  properties {
   pitchModulationParams {default=0, maxOutput=1.3, minOutput=1, parameterName=Boost}
   volumeModulationParams {default=0, maxOutput=0.5, minOutput=1, parameterName=Boost}
  }
 }
 {
  id=ModulatorContinuous_12
  children [WavePlayer_14]
  properties {pitchModulationParams {default=0, maxOutput=1.3, minOutput=1, parameterName=Speed}, volumeModulationParams {default=0, parameterName=Speed}}
 }
 {
  id=ModulatorContinuous_13
  children [WavePlayer_2]
  properties {pitchModulationParams {maxOutput=0.95, minOutput=1.1}, volumeModulationParams {default=0, parameterName=Idle}}
 }
 {
  id=ModulatorContinuous_14
  children [Enveloper_4]
  properties {pitchModulationParams {default=0, minOutput=1, parameterName=Reverse}, volumeModulationParams {default=0, parameterName=Reverse}}
 }
 {
  id=ModulatorContinuous_16
  children [ModulatorContinuous_11]
  properties {pitchModulationParams {maxOutput=1.2, minOutput=0.8}, volumeModulationParams {default=0, maxOutput=0, minOutput=1, parameterName=IsInWater}}
 }
 {
  id=ModulatorContinuous_17
  children [ParamCrossFade_4]
  properties {
   pitchModulationParams {default=0, maxOutput=1.1, minInput=0.1, minOutput=1, parameterName=Turning}
   volumeModulationParams {default=0, maxInput=0.5, parameterName=Turning}
  }
 }
 {
  id=ModulatorContinuous_18
  children [WavePlayer_16]
  properties {pitchModulationParams {default=0, parameterName=Speed}, volumeModulationParams {default=0, parameterName=Speed}}
 }
 {
  id=ModulatorContinuous_19
  children [Mixer_7]
  properties {pitchModulationParams {maxOutput=1.2, minOutput=0.8}, volumeModulationParams {default=0, parameterName=IsInWater}}
 }
 {
  id=ModulatorContinuous_2
  children [ParamCrossFade_4]
  properties {pitchModulationParams {default=0, maxOutput=1.3, minOutput=1, parameterName=Speed}, volumeModulationParams {default=0, parameterName=Speed}}
 }
 {
  id=ModulatorContinuous_22
  children [ModulatorContinuous_2]
  properties {
   pitchModulationParams {default=0, maxOutput=1.3, minOutput=1}
   volumeModulationParams {default=0, maxOutput=2.5, minOutput=1, parameterName=TurnPower}
  }
 }
 {id=ModulatorContinuous_26, children [ModulatorContinuous_18], properties {volumeModulationParams {default=0, parameterName=IsInAir}}}
 {
  id=ModulatorContinuous_3
  children [WavePlayer_8]
  properties {
   pitchModulationParams {default=0, maxOutput=1.6, minOutput=1.4, parameterName=CannonVerticalMovement}
   volumeModulationParams {default=0, parameterName=CannonVerticalMovement}
  }
 }
 {
  id=ModulatorContinuous_33
  children [WavePlayer_14]
  properties {pitchModulationParams {default=0, minOutput=0.8, parameterName=Turning}, volumeModulationParams {default=0, parameterName=Turning}}
 }
 {
  id=ModulatorContinuous_34
  children [WavePlayer_13]
  properties {pitchModulationParams {default=0}, volumeModulationParams {default=0, parameterName=Idle}}
 }
 {
  id=ModulatorContinuous_35
  children [ModulatorContinuous_34]
  properties {pitchModulationParams {default=0, parameterName=HasDriver}, volumeModulationParams {default=0, parameterName=HasDriver}}
 }
 {
  id=ModulatorContinuous_36
  children [ModulatorContinuous_10]
  properties {pitchModulationParams {default=0, minOutput=1}, volumeModulationParams {default=0, maxOutput=0, minOutput=1, parameterName=OutOfFuel}}
 }
 {
  id=ModulatorContinuous_4
  children [WavePlayer_10]
  properties {
   pitchModulationParams {default=0, maxOutput=1.1, minOutput=0.9, parameterName=TurretHorizontalMovement}
   volumeModulationParams {default=0, parameterName=TurretHorizontalMovement}
  }
 }
 {
  id=ModulatorContinuous_5
  children [WavePlayer_4]
  properties {pitchModulationParams {default=0, minOutput=1, parameterName=Turning}, volumeModulationParams {default=0, parameterName=Turning}}
 }
 {
  id=ModulatorContinuous_6
  children [WavePlayer_5]
  properties {
   pitchModulationParams {default=0, maxOutput=1.1, minOutput=0.9, parameterName=TurretVerticalMovement}
   volumeModulationParams {default=0, parameterName=TurretVerticalMovement}
  }
 }
 {
  id=ModulatorContinuous_7
  children [ModulatorContinuous_0]
  properties {
   pitchModulationParams {default=0, maxOutput=1.3, minOutput=1, parameterName=Boost}
   volumeModulationParams {default=0, maxOutput=0, minOutput=1, parameterName=Boost}
  }
 }
 {
  id=ModulatorContinuous_8
  children [ModulatorContinuous_36]
  properties {
   pitchModulationParams {default=0, maxOutput=1.3, minOutput=1, parameterName=Boost}
   volumeModulationParams {default=0, maxOutput=0, minOutput=1, parameterName=Boost}
  }
 }
 {
  id=ModulatorContinuous_9
  children [ModulatorContinuous_14]
  properties {
   pitchModulationParams {default=0, maxOutput=1.3, minOutput=1, parameterName=Boost}
   volumeModulationParams {default=0, maxOutput=0, minOutput=1, parameterName=Boost}
  }
 }
 {
  id=ParamCrossFade_4
  children [WavePlayer_6, WavePlayer_6, WavePlayer_12]
  properties {
   paramName=Surface
   crossFadeInput [
    {fadeInDistanceEnd=0.3333, fadeInDistanceStart=0, fadeOutDistanceEnd=0.6667, fadeOutDistanceStart=0.3333, volume=1}
    {fadeInDistanceEnd=0.6667, fadeInDistanceStart=0.3333, fadeOutDistanceEnd=1, fadeOutDistanceStart=0.6667, volume=1}
    {fadeInDistanceEnd=1, fadeInDistanceStart=0.6667, fadeOutDistanceEnd=100, fadeOutDistanceStart=1, volume=1}
   ]
  }
 }
 {id=SoundClass_0, children [Enveloper_0], properties {soundClassOverride=&Fort_Vehicles_Exterior}}
 {id=SoundClass_1, children [Enveloper_6], properties {soundClassOverride=&Fort_Vehicles_Exterior}}
 {id=SoundClass_2, children [Mixer_6], properties {soundClassOverride=&Fort_Vehicles_Exterior}}
 {id=WavePlayer_1, properties {bLooping=true, soundWaveAssetPtr=&/Veh_Tank_Engine_Forward_Loop_01}}
 {id=WavePlayer_10, properties {bLooping=true, soundWaveAssetPtr=&/Veh_Tank_Cannon_Turn_Hor_Loop_01}}
 {id=WavePlayer_12, properties {bLooping=true, soundWaveAssetPtr=&/Veh_Tank_Roll_Stone_Loop_01}}
 {id=WavePlayer_13, properties {bLooping=true, soundWaveAssetPtr=/Game/Athena/Sounds/Vehicles/Meatball/SoundWaves/Veh_Boat_Idle_Water_Loop_01}}
 {
  id=WavePlayer_14
  properties {bLooping=true, soundWaveAssetPtr=/Game/Athena/Sounds/Vehicles/Meatball/SoundWaves/Veh_Meatball_Wake_Choppy_Loop_01}
 }
 {id=WavePlayer_16, properties {bLooping=true, soundWaveAssetPtr=/Game/Athena/Sounds/Wind_Falling/AMB_Wind_Skydive_Vertical_Loop_01}}
 {id=WavePlayer_2, properties {bLooping=true, soundWaveAssetPtr=&/Veh_Tank_Engine_Idle_Loop_01}}
 {id=WavePlayer_3, properties {bLooping=true, soundWaveAssetPtr=&/Veh_Tank_Engine_Reverse_Loop_01}}
 {id=WavePlayer_4, properties {bLooping=true, soundWaveAssetPtr=&/Veh_Tank_Engine_Turn_Loop_01}}
 {id=WavePlayer_5, properties {bLooping=true, soundWaveAssetPtr=&/Veh_Tank_Cannon_Turn_Vert_Loop_01}}
 {id=WavePlayer_6, properties {bLooping=true, soundWaveAssetPtr=&/Veh_Tank_Roll_Dirt_Loop_01}}
 {id=WavePlayer_7, properties {bLooping=true, soundWaveAssetPtr=&Veh_Tank_Cannon_Turn_Hor_Loop_01}}
 {id=WavePlayer_8, properties {bLooping=true, soundWaveAssetPtr=&Veh_Tank_Cannon_Turn_Vert_Loop_01}}
]
properties {
 attenuationSettings=/Tank/Audio/Settings/Tank_Movement_Att
 maxDistance=28024
 soundClassObject=&Fort_Vehicles_Engine
 volumeMultiplier=1
 attenuationOverrides {falloffDistance=20000}
}
