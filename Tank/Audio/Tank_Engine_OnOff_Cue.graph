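# Sound cue for the tank's engine on/off one-shots, routed to the vehicle interior sound class with tank movement attenuation.
# An outer OutOfFuel switch passes to the EngineOn switch by default and when fuel remains, and plays nothing when out of fuel.
# The EngineOn switch picks the ignition sound for on, or the engine-off warning for off and unset.
# Concurrency limits the cue to one instance.
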SoundCue /Tank/Audio/Tank_Engine_OnOff_Cue
root=Switch_2
nodes [
 {id=Switch_0, children [WavePlayer_0,WavePlayer_0,WavePlayer_2], properties {intParameterName=EngineOn}}
 {id=Switch_2, children [Switch_0,Switch_0,null], properties {intParameterName=OutOfFuel}}
 {id=WavePlayer_0, properties {soundWaveAssetPtr=&/Veh_Tank_Warning_Engine_Off_01}}
 {id=WavePlayer_2, properties {soundWaveAssetPtr=&/Veh_Tank_Engine_Ignition_02}}
]
properties {
 attenuationSettings=/Tank/Audio/Settings/Tank_Movement_Att
 maxDistance=28024
 soundClassObject=&Fort_Vehicles_Interior
 volumeMultiplier=0.4
 concurrencyOverrides {maxCount=1}
}
