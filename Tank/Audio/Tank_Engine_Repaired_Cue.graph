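# Sound cue for the tank's engine-repaired audio, built from two "warning engine enabled" wave variations.
# A mixer layers both waves at equal, reduced input volume, leaving its other two inputs empty.
# The cue plays at lowered volume in the exterior vehicle sound class with close-range tank movement attenuation.
# A concurrency override limits it to one instance at a time.
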SoundCue /Tank/Audio/Tank_Engine_Repaired_Cue
root=Mixer_0
nodes [
 {id=Mixer_0, children [WavePlayer_5,WavePlayer_4,null,null], properties {inputVolume [0.7,0.7,1.5,1]}}
 {id=WavePlayer_4, properties {soundWaveAssetPtr=&/Veh_Tank_Warning_Engine_Enabled_02}}
 {id=WavePlayer_5, properties {soundWaveAssetPtr=&/Veh_Tank_Warning_Engine_Enabled_01}}
]
properties {
 attenuationSettings=/Tank/Audio/Settings/Tank_Movement_Close_Att
 bOverrideConcurrency=true
 maxDistance=6024
 soundClassObject=&Fort_Vehicles_Exterior
 volumeMultiplier=0.5
 concurrencyOverrides {maxCount=1}
}
