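# Tank engine one-shot Sound Cue for turning, layering an immediate gear-shift sound with a pitched-down turn rev delayed a quarter second.
# Each layer picks randomly from equally weighted variations through a modulator before mixing.
# A quality node switches between two such mixes, the last two levels using smaller variation pools.
# It uses tank movement attenuation and allows one voice per owner.
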SoundCue /Tank/Audio/Tank_Engine_Turn_OneShot_Cue
root=QualityLevel_0
nodes [
 {id=Delay_0, children [Modulator_0], properties {delayMax=0.25, delayMin=0.25}}
 {id=Delay_1, children [Modulator_2], properties {delayMax=0.25, delayMin=0.25}}
 {id=Mixer_0, children [Delay_0,Modulator_1], properties {inputVolume [1,1]}}
 {id=Mixer_1, children [Delay_1,Modulator_3], properties {inputVolume [1,1]}}
 {id=Modulator_0, children [Random_1], properties {pitchMax=0.95, pitchMin=0.85}}
 {id=Modulator_1, children [Random_0]}
 {id=Modulator_2, children [Random_2], properties {pitchMax=0.95, pitchMin=0.85}}
 {id=Modulator_3, children [Random_3]}
 {id=QualityLevel_0, children [Mixer_0,Mixer_0,Mixer_1,Mixer_1]}
 {id=Random_0, children [WavePlayer_0,WavePlayer_1,WavePlayer_2,WavePlayer_3,WavePlayer_7,WavePlayer_8], properties {weights [1,1,1,1,1,1]}}
 {id=Random_1, children [WavePlayer_6, WavePlayer_9, WavePlayer_10, WavePlayer_11, WavePlayer_12], properties {weights [1,1,1,1,1]}}
 {id=Random_2, children [WavePlayer_4, WavePlayer_5, WavePlayer_13], properties {weights [1,1,1]}}
 {id=Random_3, children [WavePlayer_16, WavePlayer_17, WavePlayer_18], properties {weights [1,1,1]}}
 {id=WavePlayer_0, properties {soundWaveAssetPtr=&/Veh_Tank_Engine_ShiftGear_01}}
 {id=WavePlayer_1, properties {soundWaveAssetPtr=&/Veh_Tank_Engine_ShiftGear_02}}
 {id=WavePlayer_10, properties {soundWaveAssetPtr=&/Veh_Tank_Engine_Turn_Rev_03}}
 {id=WavePlayer_11, properties {soundWaveAssetPtr=&/Veh_Tank_Engine_Turn_Rev_04}}
 {id=WavePlayer_12, properties {soundWaveAssetPtr=&/Veh_Tank_Engine_Turn_Rev_05}}
 {id=WavePlayer_13, properties {soundWaveAssetPtr=&Veh_Tank_Engine_Turn_Rev_03}}
 {id=WavePlayer_16, properties {soundWaveAssetPtr=&Veh_Tank_Engine_ShiftGear_01}}
 {id=WavePlayer_17, properties {soundWaveAssetPtr=&Veh_Tank_Engine_ShiftGear_02}}
 {id=WavePlayer_18, properties {soundWaveAssetPtr=&/Veh_Tank_Engine_ShiftGear_03}}
 {id=WavePlayer_2, properties {soundWaveAssetPtr=&Veh_Tank_Engine_ShiftGear_03}}
 {id=WavePlayer_3, properties {soundWaveAssetPtr=&/Veh_Tank_Engine_ShiftGear_04}}
 {id=WavePlayer_4, properties {soundWaveAssetPtr=&/Veh_Tank_Engine_Turn_Rev_01}}
 {id=WavePlayer_5, properties {soundWaveAssetPtr=&/Veh_Tank_Engine_Turn_Rev_02}}
 {id=WavePlayer_6, properties {soundWaveAssetPtr=&Veh_Tank_Engine_Turn_Rev_01}}
 {id=WavePlayer_7, properties {soundWaveAssetPtr=&/Veh_Tank_Engine_ShiftGear_05}}
 {id=WavePlayer_8, properties {soundWaveAssetPtr=&/Veh_Tank_Engine_ShiftGear_06}}
 {id=WavePlayer_9, properties {soundWaveAssetPtr=&Veh_Tank_Engine_Turn_Rev_02}}
]
properties {
 attenuationSettings=/Tank/Audio/Settings/Tank_Movement_Att
 maxDistance=28024
 soundClassObject=&Fort_Vehicles_Engine
 volumeMultiplier=0.496
 concurrencyOverrides {bLimitToOwner=true, maxCount=1, voiceStealReleaseTime=0.5}
}
