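# Sound cue for the tank's engine gear shift, routed to the vehicle interior sound class with close movement attenuation.
# A quality-level switch sends its first two slots to a six-wave random pool and its last two to a separate four-wave pool.
# Each pool passes through a Modulator for pitch and volume variation.
# A concurrency override limits the cue to one playing instance.
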SoundCue /Tank/Audio/Tank_GearShift_Cue
root=QualityLevel_0
nodes [
 {id=Modulator_0, children [Random_0]}
 {id=Modulator_1, children [Random_1]}
 {id=QualityLevel_0, children [Modulator_0,Modulator_0,Modulator_1,Modulator_1]}
 {id=Random_0, children [WavePlayer_0,WavePlayer_1,WavePlayer_2,WavePlayer_3,WavePlayer_7,WavePlayer_8], properties {weights [1,1,1,1,1,1]}}
 {id=Random_1, children [WavePlayer_4,WavePlayer_5,WavePlayer_6,WavePlayer_9], properties {weights [1,1,1,1]}}
 {id=WavePlayer_0, properties {soundWaveAssetPtr=&/Veh_Tank_Engine_ShiftGear_01}}
 {id=WavePlayer_1, properties {soundWaveAssetPtr=&/Veh_Tank_Engine_ShiftGear_02}}
 {id=WavePlayer_2, properties {soundWaveAssetPtr=&/Veh_Tank_Engine_ShiftGear_03}}
 {id=WavePlayer_3, properties {soundWaveAssetPtr=&/Veh_Tank_Engine_ShiftGear_04}}
 {id=WavePlayer_4, properties {soundWaveAssetPtr=&Veh_Tank_Engine_ShiftGear_01}}
 {id=WavePlayer_5, properties {soundWaveAssetPtr=&Veh_Tank_Engine_ShiftGear_02}}
 {id=WavePlayer_6, properties {soundWaveAssetPtr=&Veh_Tank_Engine_ShiftGear_03}}
 {id=WavePlayer_7, properties {soundWaveAssetPtr=&/Veh_Tank_Engine_ShiftGear_05}}
 {id=WavePlayer_8, properties {soundWaveAssetPtr=&/Veh_Tank_Engine_ShiftGear_06}}
 {id=WavePlayer_9, properties {soundWaveAssetPtr=&Veh_Tank_Engine_ShiftGear_04}}
]
properties {
 attenuationSettings=/Tank/Audio/Settings/Tank_Movement_Close_Att
 bOverrideConcurrency=true
 maxDistance=6024
 soundClassObject=&Fort_Vehicles_Interior
 volumeMultiplier=0.496
 concurrencyOverrides {maxCount=1}
}
