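# Sound cue for a player exiting the tank's gunner seat.
# A single wave player feeds the root directly with the gunner seat exit sample, with no randomization or modulation.
# It plays at reduced volume through the vehicle interior sound class, using the tank's close-movement attenuation settings.
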SoundCue /Tank/Audio/Tank_Gunner_Exit_Cue
root=WavePlayer_2
nodes [{id=WavePlayer_2, properties {soundWaveAssetPtr=&/Veh_Tank_Seat_Gunner_Exit_01}}]
properties {
 attenuationSettings=/Tank/Audio/Settings/Tank_Movement_Close_Att
 maxDistance=6024
 soundClassObject=&Fort_Vehicles_Interior
 volumeMultiplier=0.6
}
