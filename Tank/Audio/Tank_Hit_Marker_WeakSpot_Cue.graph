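# Sound cue for the hit marker heard when a tank weak spot is hit, routed to the 2D instigator hit-notification sound class.
# A quality-level root sends its first two slots to a random pick of seven impact waves and its last two to a pick of three.
# Both branches modulate pitch downward and vary volume for variety.
# Concurrency settings cap playback at two instances, stopping the oldest.
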SoundCue /Tank/Audio/Tank_Hit_Marker_WeakSpot_Cue
root=QualityLevel_1
nodes [
 {id=Modulator_0, children [Random_1], properties {pitchMax=1, pitchMin=0.7, volumeMax=1, volumeMin=0.6}}
 {id=Modulator_1, children [Random_0], properties {pitchMax=1, pitchMin=0.7, volumeMax=1, volumeMin=0.6}}
 {id=QualityLevel_1, children [Modulator_0,Modulator_0,Modulator_1,Modulator_1]}
 {id=Random_0, children [WavePlayer_4, WavePlayer_12, WavePlayer_13], properties {weights [1,1,1]}}
 {
  id=Random_1
  children [WavePlayer_0,WavePlayer_1,WavePlayer_2,WavePlayer_3,WavePlayer_5,WavePlayer_6,WavePlayer_7]
  properties {weights [1,1,1,1,1,1,1]}
 }
 {id=WavePlayer_0, properties {soundWaveAssetPtr=&/Veh_Tank_Impact_WeakSpot_01}}
 {id=WavePlayer_1, properties {soundWaveAssetPtr=&/Veh_Tank_Impact_WeakSpot_07}}
 {id=WavePlayer_12, properties {soundWaveAssetPtr=&/Veh_Tank_Impact_WeakSpot_03}}
 {id=WavePlayer_13, properties {soundWaveAssetPtr=&/Veh_Tank_Impact_WeakSpot_02}}
 {id=WavePlayer_2, properties {soundWaveAssetPtr=&/Veh_Tank_Impact_WeakSpot_06}}
 {id=WavePlayer_3, properties {soundWaveAssetPtr=&/Veh_Tank_Impact_WeakSpot_05}}
 {id=WavePlayer_4, properties {soundWaveAssetPtr=&Veh_Tank_Impact_WeakSpot_01}}
 {id=WavePlayer_5, properties {soundWaveAssetPtr=&/Veh_Tank_Impact_WeakSpot_04}}
 {id=WavePlayer_6, properties {soundWaveAssetPtr=&Veh_Tank_Impact_WeakSpot_03}}
 {id=WavePlayer_7, properties {soundWaveAssetPtr=&Veh_Tank_Impact_WeakSpot_02}}
]
properties {
 maxDistance=2097152
 soundClassObject=&Fort_2DHitNotifications_Instigate_PSM
 volumeMultiplier=0.4
 concurrencyOverrides {maxCount=2, resolutionRule=StopOldest}
 concurrencySet [&HitNotifications_TwoMaximum_StopOldest_VolumeScale]
}
