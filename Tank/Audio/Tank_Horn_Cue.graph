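# Sound cue for the tank's horn.
# It plays a single horn wave player directly as the root, with no randomization or mixing nodes.
# It routes through the exterior vehicle sound class, using a distant attenuation asset for long-range audibility and a dedicated horn music concurrency set to limit overlapping plays.
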SoundCue /Tank/Audio/Tank_Horn_Cue
root=WavePlayer_1
nodes [{id=WavePlayer_1, properties {soundWaveAssetPtr=&/Veh_Tank_Horn_01}}]
properties {
 attenuationSettings=/Tank/Audio/Settings/Tank_Horn_Distant_Att
 maxDistance=27500
 soundClassObject=&Fort_Vehicles_Exterior
 volumeMultiplier=1
 concurrencySet [/Tank/Audio/Settings/Tank_Horn_Music_Conc]
}
