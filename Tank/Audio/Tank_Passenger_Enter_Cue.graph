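# Sound cue for a passenger entering a tank seat, classed as an interior vehicle sound.
# It randomly picks one of two equally weighted enter recordings and passes it through a Modulator node.
# A close-range tank movement attenuation asset handles spatialization, and the cue plays at reduced volume.
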SoundCue /Tank/Audio/Tank_Passenger_Enter_Cue
root=Modulator_1
nodes [
 {id=Modulator_1, children [Random_0]}
 {id=Random_0, children [WavePlayer_1,WavePlayer_2], properties {weights [1,1]}}
 {id=WavePlayer_1, properties {soundWaveAssetPtr=&/Veh_Tank_Seat_Passenger_Enter_02}}
 {id=WavePlayer_2, properties {soundWaveAssetPtr=&/Veh_Tank_Seat_Passenger_Enter_01}}
]
properties {
 attenuationSettings=/Tank/Audio/Settings/Tank_Movement_Close_Att
 maxDistance=6024
 soundClassObject=&Fort_Vehicles_Interior
 volumeMultiplier=0.5
}
