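# Sound cue for the tank's track-break damage sound, routed to the exterior vehicle sound class with the tank movement attenuation.
# A quality-level switch feeds the first two levels from an equally weighted random pick of four break waves and the last two from a pick of two.
# A root Modulator at default settings varies the pitch and volume of the chosen wave.
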SoundCue /Tank/Audio/Tank_Tracks_Break_Cue
root=Modulator_0
nodes [
 {id=Modulator_0, children [QualityLevel_0]}
 {id=QualityLevel_0, children [Random_0,Random_0,Random_1,Random_1]}
 {id=Random_0, children [WavePlayer_3,WavePlayer_4,WavePlayer_5,WavePlayer_6], properties {weights [1,1,1,1]}}
 {id=Random_1, children [WavePlayer_0,WavePlayer_1], properties {weights [1,1]}}
 {id=WavePlayer_0, properties {soundWaveAssetPtr=&/Veh_Tank_Damage_Track_Break_01}}
 {id=WavePlayer_1, properties {soundWaveAssetPtr=&/Veh_Tank_Damage_Track_Break_02}}
 {id=WavePlayer_3, properties {soundWaveAssetPtr=&Veh_Tank_Damage_Track_Break_01}}
 {id=WavePlayer_4, properties {soundWaveAssetPtr=&Veh_Tank_Damage_Track_Break_02}}
 {id=WavePlayer_5, properties {soundWaveAssetPtr=&/Veh_Tank_Damage_Track_Break_03}}
 {id=WavePlayer_6, properties {soundWaveAssetPtr=&/Veh_Tank_Damage_Track_Break_04}}
]
properties {
 attenuationSettings=/Tank/Audio/Settings/Tank_Movement_Att
 maxDistance=28024
 soundClassObject=&Fort_Vehicles_Exterior
 volumeMultiplier=1
}
