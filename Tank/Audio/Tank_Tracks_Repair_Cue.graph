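# Sound cue for the tank track repair sound, played on the exterior vehicle sound class.
# A single wave player feeds one repair sample straight to the output at reduced volume.
# The tank's close-movement attenuation handles spatial falloff, and a concurrency override limits it to one playing instance at a time.
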SoundCue /Tank/Audio/Tank_Tracks_Repair_Cue
root=WavePlayer_4
nodes [{id=WavePlayer_4, properties {soundWaveAssetPtr=&/Veh_Tank_Track_Repair_01}}]
properties {
 attenuationSettings=/Tank/Audio/Settings/Tank_Movement_Close_Att
 bOverrideConcurrency=true
 maxDistance=6024
 soundClassObject=&Fort_Vehicles_Exterior
 volumeMultiplier=0.6
 concurrencyOverrides {maxCount=1}
}
