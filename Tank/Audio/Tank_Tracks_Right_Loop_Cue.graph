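# Looping sound cue for the tank's right tracks, spatialized with tank movement attenuation in the vehicle exterior sound class.
# An IsEnabled branch plays forward and reverse track loops when true or unset, and broken-track loops when false.
# Each pair is mixed, with Forward and Reverse parameters scaling each layer's volume while pitch output stays flat.
# The broken pair uses the L wave for Forward and the R wave for Reverse.
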SoundCue /Tank/Audio/Tank_Tracks_Right_Loop_Cue
root=Branch_2
nodes [
 {id=Branch_2, children [Mixer_2,Mixer_0,Mixer_2], properties {boolParameterName=IsEnabled}}
 {id=Mixer_0, children [ModulatorContinuous_3, ModulatorContinuous_1], properties {inputVolume [1.25,1.25]}}
 {id=Mixer_2, children [ModulatorContinuous_0, ModulatorContinuous_14], properties {inputVolume [1,1.2]}}
 {
  id=ModulatorContinuous_0
  children [WavePlayer_5]
  properties {
   pitchModulationParams {default=0, minOutput=1, parameterName=Forward}
   volumeModulationParams {default=0, maxOutput=1.2, parameterName=Forward}
  }
 }
 {
  id=ModulatorContinuous_1
  children [WavePlayer_2]
  properties {pitchModulationParams {default=0, minOutput=1, parameterName=Reverse}, volumeModulationParams {default=0, parameterName=Reverse}}
 }
 {
  id=ModulatorContinuous_14
  children [WavePlayer_7]
  properties {pitchModulationParams {default=0, minOutput=1, parameterName=Reverse}, volumeModulationParams {default=0, parameterName=Reverse}}
 }
 {
  id=ModulatorContinuous_3
  children [WavePlayer_1]
  properties {pitchModulationParams {default=0, minOutput=1, parameterName=Forward}, volumeModulationParams {default=0, parameterName=Forward}}
 }
 {id=WavePlayer_1, properties {bLooping=true, soundWaveAssetPtr=&/Veh_Tank_Tracks_Broken_Loop_L_01}}
 {id=WavePlayer_2, properties {bLooping=true, soundWaveAssetPtr=&/Veh_Tank_Tracks_Broken_Loop_R_01}}
 {id=WavePlayer_5, properties {bLooping=true, soundWaveAssetPtr=&/Veh_Tank_Tracks_Forward_Loop_R_01}}
 {id=WavePlayer_7, properties {bLooping=true, soundWaveAssetPtr=&/Veh_Tank_Tracks_Reverse_Loop_R_01}}
]
properties {
 attenuationSettings=/Tank/Audio/Settings/Tank_Movement_Att
 maxDistance=28024
 soundClassObject=&Fort_Vehicles_Exterior
 volumeMultiplier=0.4
 attenuationOverrides {falloffDistance=20000}
}
