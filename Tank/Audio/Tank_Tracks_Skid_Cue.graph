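# Sound cue for tank track skid sounds, playing random dirt or stone variations on the exterior vehicle sound class with close movement attenuation.
# An integer Skid parameter switches between half-volume dirt, full-volume stone and a silent slot, with slight random pitch reduction on both surfaces.
# A quality-level node uses five variations per surface in its first two slots and three in the last two.
# Concurrency allows one instance, stopping the oldest.
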SoundCue /Tank/Audio/Tank_Tracks_Skid_Cue
root=QualityLevel_0
nodes [
 {id=Modulator_0, children [Random_0], properties {pitchMax=1, pitchMin=0.9, volumeMax=1, volumeMin=1}}
 {id=Modulator_1, children [Random_1], properties {pitchMax=1, pitchMin=0.9, volumeMax=0.5, volumeMin=0.5}}
 {id=Modulator_2, children [Random_3], properties {pitchMax=1, pitchMin=0.9, volumeMax=1, volumeMin=1}}
 {id=Modulator_3, children [Random_2], properties {pitchMax=1, pitchMin=0.9, volumeMax=0.5, volumeMin=0.5}}
 {id=QualityLevel_0, children [Switch_0,Switch_0,Switch_1,Switch_1]}
 {id=Random_0, children [WavePlayer_0,WavePlayer_1,WavePlayer_2,WavePlayer_3,WavePlayer_4], properties {weights [1,1,1,1,1]}}
 {id=Random_1, children [WavePlayer_19, WavePlayer_20, WavePlayer_21, WavePlayer_22, WavePlayer_23], properties {weights [1,1,1,1,1]}}
 {id=Random_2, children [WavePlayer_5,WavePlayer_6,WavePlayer_7], properties {weights [1,1,1]}}
 {id=Random_3, children [WavePlayer_10, WavePlayer_11, WavePlayer_12], properties {weights [1,1,1]}}
 {id=Switch_0, children [Modulator_1,null,Modulator_1,Modulator_1,Modulator_0], properties {intParameterName=Skid}}
 {id=Switch_1, children [Modulator_3,null,Modulator_3,Modulator_3,Modulator_2], properties {intParameterName=Skid}}
 {id=WavePlayer_0, properties {soundWaveAssetPtr=&/Veh_Tank_Skid_Stone_01}}
 {id=WavePlayer_1, properties {soundWaveAssetPtr=&/Veh_Tank_Skid_Stone_02}}
 {id=WavePlayer_10, properties {soundWaveAssetPtr=&Veh_Tank_Skid_Stone_01}}
 {id=WavePlayer_11, properties {soundWaveAssetPtr=&Veh_Tank_Skid_Stone_02}}
 {id=WavePlayer_12, properties {soundWaveAssetPtr=&/Veh_Tank_Skid_Stone_03}}
 {id=WavePlayer_19, properties {soundWaveAssetPtr=&/Veh_Tank_Skid_Dirt_05}}
 {id=WavePlayer_2, properties {soundWaveAssetPtr=&Veh_Tank_Skid_Stone_03}}
 {id=WavePlayer_20, properties {soundWaveAssetPtr=&/Veh_Tank_Skid_Dirt_01}}
 {id=WavePlayer_21, properties {soundWaveAssetPtr=&/Veh_Tank_Skid_Dirt_02}}
 {id=WavePlayer_22, properties {soundWaveAssetPtr=&/Veh_Tank_Skid_Dirt_03}}
 {id=WavePlayer_23, properties {soundWaveAssetPtr=&/Veh_Tank_Skid_Dirt_04}}
 {id=WavePlayer_3, properties {soundWaveAssetPtr=&/Veh_Tank_Skid_Stone_04}}
 {id=WavePlayer_4, properties {soundWaveAssetPtr=&/Veh_Tank_Skid_Stone_05}}
 {id=WavePlayer_5, properties {soundWaveAssetPtr=&Veh_Tank_Skid_Dirt_05}}
 {id=WavePlayer_6, properties {soundWaveAssetPtr=&Veh_Tank_Skid_Dirt_01}}
 {id=WavePlayer_7, properties {soundWaveAssetPtr=&Veh_Tank_Skid_Dirt_02}}
]
properties {
 attenuationSettings=/Tank/Audio/Settings/Tank_Movement_Close_Att
 bOverrideConcurrency=true
 maxDistance=6024
 soundClassObject=&Fort_Vehicles_Exterior
 volumeMultiplier=0.354
 attenuationOverrides {falloffDistance=20000}
 concurrencyOverrides {maxCount=1, resolutionRule=StopOldest}
}
