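# Niagara system for staged tank damage effects, driven by DamageScale against three thresholds: smoke first, embers next, then engine fire and critical smoke.
# Sprite emitters split particles between two skeletal mesh bone or socket points, with smoke color indexed by damage.
# Vehicle_Speed modulates fire, smoke and critical ember rates.
# GPU embers use collision and pulsed spawning, a lighter LOWEND copy omits collision, and scalability modifiers enlarge distant embers and world smoke.
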NiagaraSystem /Tank/FX/NS_Tank_Damage
dependencies {
 usedDataInterfaces [
  &NiagaraDataInterfaceArrayFloat3
  &NiagaraDataInterfaceCurve
  &NiagaraDataInterfaceColorCurve
  &NiagaraDataInterfaceVectorCurve
 ]
 usedDynamicInputs []
 usedModules [&SystemState]
}
emitters {
 Embers {
  simTarget=GPUComputeSim
  emitterUpdateScript [
   &EmitterState(
    Enable Distance Culling=true
    MaxDistance=20000
    Module.bUseMaxDistance=true
    Module.bUseMinDistance=false
    Scalability Mode=::NewEnumerator1
   )
   &SpawnRate(
    SpawnRate=&Multiply_Float(
     A=&MakeCustomFloatFromBool(
      Bool=&SetBoolByFloatComparison(A=$User.DamageScale, B=$User.Stage2, Comparison Type=::NewEnumerator2)
      True Float=&MakeCustomFloatFromBool(
       Bool=&SetBoolByFloatComparison(A=$User.DamageScale, B=$User.Stage3Critical, Comparison Type=::NewEnumerator2)
       True Float=&Lerp_Float(A=300, B=500, Alpha=$User.Vehicle_Speed)
       False Float=400
      )
      False Float=0
     )
     B=&FloatFromCurve(
      FloatCurve=DataInterface() {propertyValues {Curve {keys [
          {interpMode=RCIM_Cubic, tangentMode=RCTM_User, time=0, value=1}
          {arriveTangent=-1.5312, interpMode=RCIM_Cubic, leaveTangent=-1.5312, tangentMode=RCTM_User, time=0.2159, value=0}
          {arriveTangent=8.1348, interpMode=RCIM_Cubic, leaveTangent=8.1348, tangentMode=RCTM_User, time=0.2491, value=0.6186}
          {arriveTangent=-1.4415, interpMode=RCIM_Cubic, leaveTangent=-1.4415, tangentMode=RCTM_User, time=0.3217, value=0.8606}
          <at 0.3597 0>
          {arriveTangent=0.132, interpMode=RCIM_Cubic, leaveTangent=0.132, tangentMode=RCTM_User, time=0.5004, value=0.2712}
          {arriveTangent=-0.2914, interpMode=RCIM_Cubic, leaveTangent=-0.2914, tangentMode=RCTM_User, time=0.6515, value=0.0385}
          {arriveTangent=-0.1105, interpMode=RCIM_Cubic, leaveTangent=-0.1105, tangentMode=RCTM_User, time=0.6954, value=0.7517}
          {interpMode=RCIM_Cubic, tangentMode=RCTM_User, time=1, value=0}
         ]}}}
      CurveIndex=$Emitter.NormalizedLoopAge
     )
    )
    Use Spawn Probability=false
   ) {module=SpawnRate001}
  ]
  particleSpawnScript [
   &FN_SetSpawnGroup(Number of Spawn Groups=2)
   /Niagara/Modules/Spawn/Initialization/InitializeParticle(
    Color=&UniformRangedLinearColor(Minimum=<rgba 5 1.494 0 1>, Maximum=<rgba 1 0.169 0 1>)
    Lifetime=&UniformRangedFloat(Minimum=1, Maximum=2)
    Module.Write Color=true
    Module.Write Lifetime=true
    Module.Write Mass=true
    Module.Write Material Random=true
    Module.Write Position=false
    Module.Write Scale=false
    Module.Write SpriteRotation=false
    Module.Write SpriteSize=true
    Position=&SimulationPosition()
    Sprite Size=&UniformRangedVector2D(Minimum=<0.5 1>, Maximum=<2 3>)
   ) {module=InitializeParticle001}
   &SkeletalMeshLocation(
    Bone / Socket Index=$Particles.SpawnGroup
    Bone Sampling Mode=::NewEnumerator4
    Inherit Velocity=false
    MaskSpawnGroup=false
    Sampled Position Offset=&SelectVectorFromArray(Direct Array Index=$Particles.SpawnGroup, Vector Selection Array=DataInterface() {propertyValues {FloatData [<130 -130 0>, <130 130 0>]}})
    Sampled Position Offset Coordinate Space:ENiagaraMeshTransforms=::NewEnumerator3
    Skeletal Mesh=$User.SkeletalMesh
   ) {module=SkeletalMeshLocation001}
   &ShapeLocation(MaskSpawnGroup=false, Non Uniform Scale=<1 1 0.5>, Offset Mode=::NewEnumerator0, Override Seed=false, Sphere Radius=30, UseOwnerScale=true)
   &AddVelocity(
    Additional Quaternion Rotation=false
    Cone Axis=<1 0 0.5>
    Module.UseConeFalloff=false
    Module.UseVelDistribution=false
    Override Seed=false
    Rotation Mode=::NewEnumerator3
    Rotation Quaternion:Quat4f=$Output.SkeletalMeshLocation001.SampledRotation
    Velocity Mode=::NewEnumerator2
    Velocity Speed=&Multiply_Float(
     A=&RandomRangeFloat(Minimum=150, Maximum=500)
     B=&MakeCustomFloatFromBool(
      Bool=&SetBoolByFloatComparison(A=$User.DamageScale, B=$User.Stage2, Comparison Type=::NewEnumerator2)
      True Float=0.6
      False Float=1
     )
    )
   )
   &CameraOffset(Camera Offset Amount=&UniformRangedFloat(Minimum=10, Maximum=25))
  ]
  particleUpdateScript [
   &ParticleState()
   &AccelerationForce(Acceleration=<0 0 -980>)
   &Collision(
    Advanced Aging Rate=1
    Age Colliding Particles=false
    Friction=0.1
    Kill Offscreen Particles=true
    Module.Control Rest State Via Surface Normal=false
    Module.EnableMaxCollisionCount=false
    Module.Randomize Collision Normal Vector=false
    Restitution=0.5
   )
   /Niagara/Modules/Update/Forces/CurlNoiseForce(Noise Frequency=0.005, Noise Strength=250, Pan Noise=true, Pan Noise Field=<0 0 1>)
   &CurlNoiseForce(
    Noise Frequency=0.01
    Noise Strength=&FloatFromCurve(FloatCurve=DataInterface() {propertyValues {Curve {keys [<at 0 1>,<at 0.1954 0.0398>]}}}, Scale Curve=2500)
    Pan Noise=true
    Pan Noise Field=<0 0 1>
   ) {module=CurlNoiseForce001}
   &SolveForcesAndVelocity(Clamp Velocity=false, Limit Acceleration=false)
   &ScaleSpriteSize(Scale Factor=&Vector2DFromFloat(Value=&FloatFromCurve(
      FloatCurve=DataInterface() {propertyValues {Curve {keys [
          {arriveTangent=56.1861, interpMode=RCIM_Cubic, leaveTangent=56.1859, tangentMode=RCTM_User, time=0, value=0}
          {arriveTangent=0.8166, interpMode=RCIM_Cubic, leaveTangent=0.8166, time=0.03, value=1}
          {arriveTangent=-1.7925, interpMode=RCIM_Cubic, leaveTangent=-1.7925, tangentMode=RCTM_User, time=0.7111, value=0.5807}
          <at 1 0 RCIM_Cubic>
         ]}}}
      Scale Curve=/Game/Effects/Niagara/DynamicInputs/Scalability/FN_ScalabilityFloatModifier(
       Value at DistanceFraction 0:NiagaraFloat=1.25
       Value at DistanceFraction 1:NiagaraFloat=3
       DistanceRamp:NiagaraDataInterfaceCurve=DataInterface() {propertyValues {Curve {keys [<at 0.1 0>,<at 0.2 0.5>,<at 0.6 1>]}}}
      )
     )))
   &ScaleSpriteSizeBySpeed(Max Scale Factor=<1 10>)
   &ScaleColor(
    Scale Alpha=&FloatFromCurve(FloatCurve=DataInterface() {propertyValues {Curve {keys [
         {interpMode=RCIM_Cubic, tangentMode=RCTM_User, time=0.0048, value=0.438}
         {arriveTangent=0.3021, interpMode=RCIM_Cubic, leaveTangent=0.3021, tangentMode=RCTM_User, time=0.0983, value=1.5996}
         {arriveTangent=-1.0514, interpMode=RCIM_Cubic, leaveTangent=-1.0514, tangentMode=RCTM_User, time=0.3368, value=0.5383}
         {arriveTangent=-0.8117, interpMode=RCIM_Cubic, leaveTangent=-0.8117, tangentMode=RCTM_User, time=0.5811, value=1.092}
         <at 1 0>
        ]}}})
    Scale RGB=&VectorFromCurve(VectorCurve=DataInterface() {propertyValues {
       XCurve {keys [<at 0 16 RCIM_Cubic>,<at 0.8 1 RCIM_Cubic>]}
       YCurve {keys [<at 0 16 RCIM_Cubic>,<at 0.8 0.1249 RCIM_Cubic>]}
       ZCurve {keys [<at 0 16 RCIM_Cubic>,<at 0.8 0.2547 RCIM_Cubic>]}
      }})
    ScaleA=true
    ScaleRGB=true
   )
  ]
  rendererClasses [&NiagaraSpriteRendererProperties]
  renderers [{rendererClass=&NiagaraSpriteRendererProperties}]
 }
 Embers_LOWEND {
  emitterUpdateScript [
   &EmitterState(
    Enable Distance Culling=true
    MaxDistance=20000
    Module.bUseMaxDistance=true
    Module.bUseMinDistance=false
    Scalability Mode=::NewEnumerator1
   )
   &SpawnRate(
    SpawnRate=&Multiply_Float(
     A=&MakeCustomFloatFromBool(
      Bool=&SetBoolByFloatComparison(A=$User.DamageScale, B=$User.Stage2, Comparison Type=::NewEnumerator2)
      True Float=&MakeCustomFloatFromBool(
       Bool=&SetBoolByFloatComparison(A=$User.DamageScale, B=$User.Stage3Critical, Comparison Type=::NewEnumerator2)
       True Float=&Lerp_Float(A=150, B=200, Alpha=$User.Vehicle_Speed)
       False Float=200
      )
      False Float=0
     )
     B=&FloatFromCurve(
      FloatCurve=DataInterface() {propertyValues {Curve {keys [
          {interpMode=RCIM_Cubic, tangentMode=RCTM_User, time=0, value=1}
          {arriveTangent=-1.5312, interpMode=RCIM_Cubic, leaveTangent=-1.5312, tangentMode=RCTM_User, time=0.2159, value=0}
          {arriveTangent=8.1348, interpMode=RCIM_Cubic, leaveTangent=8.1348, tangentMode=RCTM_User, time=0.2491, value=0.6186}
          {arriveTangent=-1.4415, interpMode=RCIM_Cubic, leaveTangent=-1.4415, tangentMode=RCTM_User, time=0.3217, value=0.8606}
          <at 0.3597 0>
          {arriveTangent=0.132, interpMode=RCIM_Cubic, leaveTangent=0.132, tangentMode=RCTM_User, time=0.5004, value=0.2712}
          {arriveTangent=-0.2914, interpMode=RCIM_Cubic, leaveTangent=-0.2914, tangentMode=RCTM_User, time=0.6515, value=0.0385}
          {arriveTangent=-0.1105, interpMode=RCIM_Cubic, leaveTangent=-0.1105, tangentMode=RCTM_User, time=0.6954, value=0.7517}
          {interpMode=RCIM_Cubic, tangentMode=RCTM_User, time=1, value=0}
         ]}}}
      CurveIndex=$Emitter.NormalizedLoopAge
     )
    )
    Use Spawn Probability=false
   ) {module=SpawnRate001}
  ]
  particleSpawnScript [
   &FN_SetSpawnGroup(Number of Spawn Groups=2)
   &InitializeParticle(
    Color=&UniformRangedLinearColor(Minimum=<rgba 5 1.494 0 1>, Maximum=<rgba 1 0.169 0 1>)
    Lifetime=&UniformRangedFloat(Minimum=1, Maximum=2)
    Module.Write Color=true
    Module.Write Lifetime=true
    Module.Write Mass=true
    Module.Write Material Random=true
    Module.Write Position=false
    Module.Write Scale=false
    Module.Write SpriteRotation=false
    Module.Write SpriteSize=true
    Position=&SimulationPosition()
    Sprite Size=&UniformRangedVector2D(Minimum=<0.5 1>, Maximum=<2 3>)
   ) {module=InitializeParticle001}
   &SkeletalMeshLocation(
    Bone / Socket Index=$Particles.SpawnGroup
    Bone Sampling Mode=::NewEnumerator4
    Inherit Velocity=false
    MaskSpawnGroup=false
    Sampled Position Offset=&SelectVectorFromArray(Direct Array Index=$Particles.SpawnGroup, Vector Selection Array=DataInterface() {propertyValues {FloatData [<130 -130 0>, <130 130 0>]}})
    Sampled Position Offset Coordinate Space:ENiagaraMeshTransforms=::NewEnumerator3
    Skeletal Mesh=$User.SkeletalMesh
   ) {module=SkeletalMeshLocation001}
   &ShapeLocation(MaskSpawnGroup=false, Non Uniform Scale=<1 1 0.5>, Offset Mode=::NewEnumerator0, Override Seed=false, Sphere Radius=30, UseOwnerScale=true)
   &AddVelocity(
    Additional Quaternion Rotation=false
    Cone Axis=<1 0 0.5>
    Module.UseConeFalloff=false
    Module.UseVelDistribution=false
    Override Seed=false
    Rotation Mode=::NewEnumerator3
    Rotation Quaternion:Quat4f=$Output.SkeletalMeshLocation001.SampledRotation
    Velocity Mode=::NewEnumerator2
    Velocity Speed=&Multiply_Float(
     A=&RandomRangeFloat(Minimum=150, Maximum=500)
     B=&MakeCustomFloatFromBool(
      Bool=&SetBoolByFloatComparison(A=$User.DamageScale, B=$User.Stage2, Comparison Type=::NewEnumerator2)
      True Float=0.6
      False Float=1
     )
    )
   )
   &CameraOffset(Camera Offset Amount=&UniformRangedFloat(Minimum=10, Maximum=25))
  ]
  particleUpdateScript [
   &ParticleState()
   &AccelerationForce(Acceleration=<0 0 -980>)
   &CurlNoiseForce(
    Noise Frequency=0.01
    Noise Strength=&FloatFromCurve(FloatCurve=DataInterface() {propertyValues {Curve {keys [<at 0 1>,<at 0.1954 0.0398>]}}}, Scale Curve=2500)
    Pan Noise=true
    Pan Noise Field=<0 0 1>
   ) {module=CurlNoiseForce001}
   &SolveForcesAndVelocity(Clamp Velocity=false, Limit Acceleration=false)
   &ScaleSpriteSize(Scale Factor=&Vector2DFromFloat(Value=&FloatFromCurve(
      FloatCurve=DataInterface() {propertyValues {Curve {keys [
          {arriveTangent=56.1861, interpMode=RCIM_Cubic, leaveTangent=56.1859, tangentMode=RCTM_User, time=0, value=0}
          {arriveTangent=0.8166, interpMode=RCIM_Cubic, leaveTangent=0.8166, time=0.03, value=1}
          {arriveTangent=-1.7925, interpMode=RCIM_Cubic, leaveTangent=-1.7925, tangentMode=RCTM_User, time=0.7111, value=0.5807}
          <at 1 0 RCIM_Cubic>
         ]}}}
      Scale Curve=&FN_ScalabilityFloatModifier(
       Value at DistanceFraction 0:NiagaraFloat=1.25
       Value at DistanceFraction 1:NiagaraFloat=3
       DistanceRamp:NiagaraDataInterfaceCurve=DataInterface() {propertyValues {Curve {keys [<at 0.1 0>,<at 0.2 0.5>,<at 0.6 1>]}}}
      )
     )))
   &ScaleSpriteSizeBySpeed(Max Scale Factor=<1 10>)
   &ScaleColor(
    Scale Alpha=&FloatFromCurve(FloatCurve=DataInterface() {propertyValues {Curve {keys [
         {interpMode=RCIM_Cubic, tangentMode=RCTM_User, time=0.0048, value=0.438}
         {arriveTangent=0.3021, interpMode=RCIM_Cubic, leaveTangent=0.3021, tangentMode=RCTM_User, time=0.0983, value=1.5996}
         {arriveTangent=-1.0514, interpMode=RCIM_Cubic, leaveTangent=-1.0514, tangentMode=RCTM_User, time=0.3368, value=0.5383}
         {arriveTangent=-0.8117, interpMode=RCIM_Cubic, leaveTangent=-0.8117, tangentMode=RCTM_User, time=0.5811, value=1.092}
         <at 1 0>
        ]}}})
    Scale RGB=&VectorFromCurve(VectorCurve=DataInterface() {propertyValues {
       XCurve {keys [<at 0 16 RCIM_Cubic>,<at 0.8 1 RCIM_Cubic>]}
       YCurve {keys [<at 0 16 RCIM_Cubic>,<at 0.8 0.1249 RCIM_Cubic>]}
       ZCurve {keys [<at 0 16 RCIM_Cubic>,<at 0.8 0.2547 RCIM_Cubic>]}
      }})
    ScaleA=true
    ScaleRGB=true
   )
  ]
  rendererClasses [&NiagaraSpriteRendererProperties]
  renderers [{rendererClass=&NiagaraSpriteRendererProperties}]
 }
 Local_Fire_Engine {
  emitterUpdateScript [
   &EmitterState(Life Cycle Mode=::NewEnumerator1, Loop Duration=1, UseLoopDelay=false)
   &SpawnRate(
    SpawnRate=&MakeCustomFloatFromBool(
     Bool=&SetBoolByFloatComparison(A=$User.DamageScale, B=$User.Stage3Critical, Comparison Type=::NewEnumerator2)
     True Float=40
     False Float=0
    )
    Use Spawn Probability=false
   )
  ]
  particleSpawnScript [
   &FN_SetSpawnGroup(Number of Spawn Groups=2)
   &InitializeParticle(
    Color=<rgba 2 0.9857 0.2812 1>
    Lifetime=&Lerp_Float(A=&UniformRangedFloat(Minimum=0.4, Maximum=0.65), B=0.325, Alpha=$User.Vehicle_Speed)
    Module.Write Color=true
    Module.Write Lifetime=true
    Module.Write Mass=true
    Module.Write Material Random=true
    Module.Write Position=false
    Module.Write Scale=false
    Module.Write SpriteRotation=true
    Module.Write SpriteSize=true
    Position=&SimulationPosition()
    Sprite Rotation=&NormalizedAngleToDegrees(NormalizedAngle=&UniformRangedFloat(Minimum=0, Maximum=1))
    Sprite Size=&Vector2DFromFloat(Value=&UniformRangedFloat(Minimum=100, Maximum=150))
   )
   &SkeletalMeshLocation(
    Bone / Socket Index=$Particles.SpawnGroup
    Bone Sampling Mode=::NewEnumerator4
    Inherit Velocity=false
    MaskSpawnGroup=false
    Sampled Position Offset=&SelectVectorFromArray(Direct Array Index=$Particles.SpawnGroup, Vector Selection Array=DataInterface() {propertyValues {FloatData [<120 -120 0>, <120 120 0>]}})
    Sampled Position Offset Coordinate Space:ENiagaraMeshTransforms=::NewEnumerator3
    Skeletal Mesh=$User.SkeletalMesh
   ) {module=SkeletalMeshLocation001}
   &ShapeLocation(MaskSpawnGroup=false, Non Uniform Scale=<1 1 0.5>, Offset Mode=::NewEnumerator0, Override Seed=false, Sphere Radius=30, UseOwnerScale=true)
   &AddVelocity(
    Additional Quaternion Rotation=false
    Module.UseConeFalloff=false
    Module.UseVelDistribution=false
    Override Seed=false
    Rotation Mode=::NewEnumerator3
    Rotation Quaternion:Quat4f=$Output.SkeletalMeshLocation001.SampledRotation
    Velocity Mode=::NewEnumerator2
    Velocity Speed=&RandomRangeFloat(Minimum=150, Maximum=400)
   ) {module=AddVelocity002}
   &DynamicMaterialParameters(
    Index 0 Param 2=&UniformRangedFloat(Minimum=0.333, Maximum=1)
    Index 0 Param 4=&UniformRangedFloat(Minimum=1.333, Maximum=2)
    Module.Param0WriteEnabled=false
    Module.Param1WriteEnabled=true
    Module.Param2WriteEnabled=false
    Module.Param3WriteEnabled=true
   )
  ]
  particleUpdateScript [
   &ParticleState()
   &Drag(Drag=2, Use Linear Drag=true, Use Rotational Drag=false)
   &AccelerationForce(Acceleration=<0 0 100>)
   &DynamicMaterialParameters(
    Index 0 Param 1=&FloatFromCurve(FloatCurve=DataInterface() {propertyValues {Curve {keys [
         {arriveTangent=4.229, interpMode=RCIM_Cubic, leaveTangent=4.229, tangentMode=RCTM_User, time=0, value=0.5415}
         {arriveTangent=0.917, interpMode=RCIM_Cubic, leaveTangent=0.917, time=0.1, value=0.9}
         <at 0.5 1 RCIM_Cubic>
        ]}}})
    Index 0 Param 3=&FloatFromCurve(FloatCurve=DataInterface() {propertyValues {Curve {keys [{arriveTangent=5.3935, interpMode=RCIM_Cubic, leaveTangent=3.8884, tangentMode=RCTM_User, time=0.5, value=1}, <at 1 2 RCIM_Cubic>]}}})
    Module.Param0WriteEnabled=true
    Module.Param1WriteEnabled=false
    Module.Param2WriteEnabled=true
    Module.Param3WriteEnabled=false
   ) {module=DynamicMaterialParameters001}
   &SolveForcesAndVelocity(Clamp Velocity=false, Limit Acceleration=false)
   &ScaleColor(
    Scale Alpha=&FloatFromCurve(FloatCurve=DataInterface() {propertyValues {Curve {keys [
         {arriveTangent=6.233, interpMode=RCIM_Cubic, leaveTangent=6.233, tangentMode=RCTM_User, time=0, value=0}
         {arriveTangent=3.1104, interpMode=RCIM_Cubic, leaveTangent=3.1104, time=0.2, value=1.2445}
         {arriveTangent=-1.5556, interpMode=RCIM_Cubic, leaveTangent=-1.5556, time=0.6833, value=2.1254}
         {interpMode=RCIM_Cubic, leaveTangent=-3.4385, time=1, value=0}
        ]}}})
    Scale RGB=&MakeVectorFromLinearColorRGB(Color=&ColorFromCurve(ColorCurve=DataInterface() {propertyValues {
        AlphaCurve {keys [<at -0.0023 1 RCIM_Cubic>,<at 1 1 RCIM_Cubic>]}
        BlueCurve {keys [<at -0.002 0 RCIM_Cubic>,<at 0.998 0 RCIM_Cubic>]}
        GreenCurve {keys [<at -0.002 1.3836 RCIM_Cubic>,<at 0.998 0.0572 RCIM_Cubic>]}
        RedCurve {keys [<at -0.002 4 RCIM_Cubic>,<at 0.998 1 RCIM_Cubic>]}
       }}))
    ScaleA=true
    ScaleRGB=true
   )
   &ScaleSpriteSize(Scale Factor=&Vector2DFromFloat(Value=&FloatFromCurve(
      FloatCurve=DataInterface() {propertyValues {Curve {keys [
          {arriveTangent=1.067, interpMode=RCIM_Cubic, leaveTangent=1.067, tangentMode=RCTM_User, time=0, value=0.5357}
          {arriveTangent=0.0093, interpMode=RCIM_Cubic, leaveTangent=0.0093, tangentMode=RCTM_User, time=1, value=1}
         ]}}}
      Scale Curve=&Lerp_Float(A=0.6, B=1, Alpha=$User.Vehicle_Speed)
     )))
   /SetVariables(Particles.SubImageIndex=&FloatFromCurve(FloatCurve=DataInterface() {propertyValues {Curve {keys [{arriveTangent=87.6094, time=0, value=0}, <at 1.009 50>]}}})) {isSetParameters=true}
  ]
  rendererClasses [&NiagaraSpriteRendererProperties]
  renderers [{rendererClass=&NiagaraSpriteRendererProperties}]
 }
 Local_Smoke_Forward001 {
  emitterUpdateScript [
   &EmitterState()
   &SpawnRate(
    SpawnRate=&MakeCustomFloatFromBool(
     Bool=&SetBoolByFloatComparison(A=$User.DamageScale, B=$User.Stage1, Comparison Type=::NewEnumerator2)
     True Float=&MakeCustomFloatFromBool(
      Bool=&SetBoolByFloatComparison(A=$User.DamageScale, B=$User.Stage2, Comparison Type=::NewEnumerator2)
      True Float=20
      False Float=15
     )
     False Float=0
    )
    Use Spawn Probability=false
   )
  ]
  particleSpawnScript [
   &FN_SetSpawnGroup(Number of Spawn Groups=2)
   &InitializeParticle(
    Lifetime=&UniformRangedFloat(Minimum=0.6, Maximum=1)
    Module.Write Color=true
    Module.Write Lifetime=true
    Module.Write Mass=true
    Module.Write Material Random=true
    Module.Write Position=false
    Module.Write Scale=false
    Module.Write SpriteRotation=true
    Module.Write SpriteSize=true
    Position=&SimulationPosition()
    Sprite Rotation=&NormalizedAngleToDegrees(NormalizedAngle=&UniformRangedFloat(Minimum=0, Maximum=1))
    Sprite Size=&Vector2DFromFloat(Value=&UniformRangedFloat(Minimum=75, Maximum=90))
   )
   &InheritVelocity(Module.Limit Speed=true, Velocity Limit=1500, Velocity Scale=<0.04 0.04 -0.02>)
   &SkeletalMeshLocation(
    Bone / Socket Index=$Particles.SpawnGroup
    Bone Sampling Mode=::NewEnumerator4
    Inherit Velocity=false
    MaskSpawnGroup=false
    Sampled Position Offset=&SelectVectorFromArray(Direct Array Index=$Particles.SpawnGroup, Vector Selection Array=DataInterface() {propertyValues {FloatData [<130 -130 0>, <130 130 0>]}})
    Sampled Position Offset Coordinate Space:ENiagaraMeshTransforms=::NewEnumerator3
    Skeletal Mesh=$User.SkeletalMesh
   ) {module=SkeletalMeshLocation001}
   &ShapeLocation(MaskSpawnGroup=false, Non Uniform Scale=<1 1 0.5>, Offset Mode=::NewEnumerator0, Override Seed=false, Sphere Radius=30, UseOwnerScale=true)
   &AddVelocity(
    Additional Quaternion Rotation=false
    Module.UseConeFalloff=false
    Module.UseVelDistribution=false
    Override Seed=false
    Rotation Mode=::NewEnumerator3
    Rotation Quaternion:Quat4f=$Output.SkeletalMeshLocation001.SampledRotation
    Velocity Mode=::NewEnumerator2
    Velocity Speed=&RandomRangeFloat(Minimum=100, Maximum=300)
   ) {module=AddVelocity003}
   &CameraOffset(Camera Offset Amount=25)
   &Color(Color=&MakeLinearColorFromVectorAndFloat(
     "Vector (RGB)"=&MakeVectorFromLinearColorRGB(Color=&ColorFromCurve(
       ColorCurve=DataInterface() {propertyValues {
         AlphaCurve {keys [<at -0.0023 1 RCIM_Constant>,<at 0.0995 1 RCIM_Constant>,<at 1 1 RCIM_Constant>]}
         BlueCurve {keys [<at 0 0.1 RCIM_Constant>,<at 0.45 0.05 RCIM_Constant>,<at 1 0.0464 RCIM_Constant>]}
         GreenCurve {keys [<at 0 0.1 RCIM_Constant>,<at 0.45 0.05 RCIM_Constant>,<at 1 0.2858 RCIM_Constant>]}
         RedCurve {keys [<at 0 0.1 RCIM_Constant>,<at 0.45 0.05 RCIM_Constant>,<at 1 2 RCIM_Constant>]}
        }}
       CurveIndex=$User.DamageScale
      ))
     "Float (Alpha)"=1
    ))
   &BoxLocation(Box Size=<100 100 1>, MaskSpawnGroup=false, Module.Override Seed=false, Rotate Box=false, Surface Only=false)
   &AddVelocity(Velocity=&Lerp_Vector(A=<0 0 0>, B=<-605 0 0>, Alpha=$User.Vehicle_Speed)) {module=AddVelocity002}
  ]
  particleUpdateScript [
   &ParticleState()
   &Drag(Drag=3, Use Linear Drag=true, Use Rotational Drag=false)
   &AccelerationForce(Acceleration=<0 0 100>)
   &SolveForcesAndVelocity(Clamp Velocity=false, Limit Acceleration=false)
   &ScaleColor(
    Scale Alpha=&FloatFromCurve(
     FloatCurve=DataInterface() {propertyValues {Curve {keys [
         {interpMode=RCIM_Cubic, tangentMode=RCTM_User, time=0, value=0}
         {arriveTangent=0.201, interpMode=RCIM_Cubic, leaveTangent=0.201, tangentMode=RCTM_User, time=0.1, value=0.575}
         {arriveTangent=-1.4222, interpMode=RCIM_Cubic, leaveTangent=-1.4222, tangentMode=RCTM_User, time=0.5526, value=0.434}
         {interpMode=RCIM_Cubic, tangentMode=RCTM_User, time=1, value=0}
        ]}}}
     Scale Curve=0.5
    )
    Scale RGB=&VectorFromFloat(Value=&FloatFromCurve(FloatCurve=DataInterface() {propertyValues {Curve {keys [{arriveTangent=-3.2082, interpMode=RCIM_Cubic, leaveTangent=-1.8602, tangentMode=RCTM_User, time=0, value=1.25}, <at 1 0.15 RCIM_Cubic>]}}}))
    ScaleA=true
    ScaleRGB=true
   )
   &DynamicMaterialParameters(
    Index 0 Param 1=&FloatFromCurve(FloatCurve=DataInterface() {propertyValues {Curve {keys [<at 0 1>,<at 1 0>]}}})
    Index 0 Param 2=&FloatFromCurve(FloatCurve=DataInterface() {propertyValues {Curve {keys [
         {arriveTangent=3.5758, interpMode=RCIM_Cubic, leaveTangent=3.5758, tangentMode=RCTM_User, time=-0.0037, value=-0.957}
         {arriveTangent=1.3602, interpMode=RCIM_Cubic, leaveTangent=1.3602, tangentMode=RCTM_User, time=1.0011, value=1}
        ]}}})
    Index 0 Param 4=0.5
    Module.Param0WriteEnabled=true
    Module.Param1WriteEnabled=true
    Module.Param2WriteEnabled=true
    Module.Param3WriteEnabled=true
   )
   &ScaleSpriteSize(Scale Factor=&Vector2DFromFloat(Value=&FloatFromCurve(
      FloatCurve=DataInterface() {propertyValues {Curve {keys [{arriveTangent=4.3948, interpMode=RCIM_Cubic, leaveTangent=4.3948, tangentMode=RCTM_User, time=0, value=0.5}, <at 1 3 RCIM_Cubic>]}}}
      Scale Curve=0.5
     )))
   &SpriteRotationRate(Rotation Rate=&UniformRangedFloat(Minimum=-1, Maximum=1))
  ]
  rendererClasses [&NiagaraSpriteRendererProperties]
  renderers [{rendererClass=&NiagaraSpriteRendererProperties}]
 }
 World_Smoke_Critical001 {
  emitterUpdateScript [
   &EmitterState()
   &SpawnRate(
    SpawnRate=&MakeCustomFloatFromBool(
     Bool=&SetBoolByFloatComparison(A=$User.DamageScale, B=$User.Stage3Critical, Comparison Type=::NewEnumerator2)
     True Float=35
     False Float=0
    )
    Use Spawn Probability=false
   ) {module=SpawnRate001}
  ]
  particleSpawnScript [
   &FN_SetSpawnGroup(Number of Spawn Groups=2)
   &InitializeParticle(
    Color=<rgba 1 1 1 0.5>
    Lifetime=&UniformRangedFloat(Minimum=0.6, Maximum=1)
    Module.Write Color=true
    Module.Write Lifetime=true
    Module.Write Mass=true
    Module.Write Material Random=true
    Module.Write Position=false
    Module.Write Scale=false
    Module.Write SpriteRotation=true
    Module.Write SpriteSize=true
    Position=&SimulationPosition()
    Sprite Rotation=&NormalizedAngleToDegrees(NormalizedAngle=&UniformRangedFloat(Minimum=0, Maximum=1))
    Sprite Size=&Vector2DFromFloat(Value=&UniformRangedFloat(Minimum=30, Maximum=50))
   )
   &SkeletalMeshLocation(
    Bone / Socket Index=$Particles.SpawnGroup
    Bone Sampling Mode=::NewEnumerator4
    Inherit Velocity=false
    MaskSpawnGroup=false
    Sampled Position Offset=&SelectVectorFromArray(Direct Array Index=$Particles.SpawnGroup, Vector Selection Array=DataInterface() {propertyValues {FloatData [<120 -120 0>, <120 120 0>]}})
    Sampled Position Offset Coordinate Space:ENiagaraMeshTransforms=::NewEnumerator3
    Skeletal Mesh=$User.SkeletalMesh
   ) {module=SkeletalMeshLocation001}
   &ShapeLocation(MaskSpawnGroup=false, Non Uniform Scale=<1 1 0.5>, Offset Mode=::NewEnumerator0, Override Seed=false, Sphere Radius=30, UseOwnerScale=true)
   &AddVelocity(
    Additional Quaternion Rotation=false
    Module.UseConeFalloff=false
    Module.UseVelDistribution=false
    Override Seed=false
    Rotation Mode=::NewEnumerator3
    Rotation Quaternion:Quat4f=$Output.SkeletalMeshLocation001.SampledRotation
    Velocity Mode=::NewEnumerator2
    Velocity Speed=&RandomRangeFloat(Minimum=100, Maximum=300)
   ) {module=AddVelocity002}
   &CameraOffset(Camera Offset Amount=50)
   &Color(Color=&MakeLinearColorFromVectorAndFloat(
     "Vector (RGB)"=&MakeVectorFromLinearColorRGB(Color=&ColorFromCurve(
       ColorCurve=DataInterface() {propertyValues {
         AlphaCurve {keys [<at -0.0023 1 RCIM_Constant>,<at 0.0995 1 RCIM_Constant>,<at 1 1 RCIM_Constant>]}
         BlueCurve {keys [<at 0 0.1 RCIM_Constant>,<at 0.45 0.05 RCIM_Constant>,<at 1 0.0464 RCIM_Constant>]}
         GreenCurve {keys [<at 0 0.1 RCIM_Constant>,<at 0.45 0.05 RCIM_Constant>,<at 1 0.2858 RCIM_Constant>]}
         RedCurve {keys [<at 0 0.1 RCIM_Constant>,<at 0.45 0.05 RCIM_Constant>,<at 1 2 RCIM_Constant>]}
        }}
       CurveIndex=$User.DamageScale
      ))
     "Float (Alpha)"=0.75
    ))
  ]
  particleUpdateScript [
   &ParticleState()
   &Drag(Use Linear Drag=true, Use Rotational Drag=false)
   &AccelerationForce(Acceleration=&RandomRangeVector(Minimum=<0 0 250>, Maximum=<0 0 500>))
   &SolveForcesAndVelocity(Clamp Velocity=false, Limit Acceleration=false)
   &ScaleColor(
    Scale Alpha=&FloatFromCurve(FloatCurve=DataInterface() {propertyValues {Curve {keys [
         {interpMode=RCIM_Cubic, tangentMode=RCTM_User, time=0, value=0}
         {interpMode=RCIM_Cubic, tangentMode=RCTM_User, time=0.1007, value=0.6723}
         {arriveTangent=-2.0134, interpMode=RCIM_Cubic, leaveTangent=-2.0134, tangentMode=RCTM_User, time=0.5526, value=0.5348}
         {interpMode=RCIM_Cubic, tangentMode=RCTM_User, time=1, value=0}
        ]}}})
    Scale RGB=&VectorFromFloat(Value=&FloatFromCurve(FloatCurve=DataInterface() {propertyValues {Curve {keys [
          {arriveTangent=-3.2082, interpMode=RCIM_Cubic, leaveTangent=-1.8602, tangentMode=RCTM_User, time=0, value=1.25}
          <at 0.3977 0.25>
          <at 0.5013 0.2904 RCIM_Cubic>
         ]}}}))
    ScaleA=true
    ScaleRGB=true
   )
   &DynamicMaterialParameters(
    Index 0 Param 1=&FloatFromCurve(FloatCurve=DataInterface() {propertyValues {Curve {keys [<at 0 0.15>,<at 1 -1>]}}})
    Index 0 Param 2=&FloatFromCurve(FloatCurve=DataInterface() {propertyValues {Curve {keys [{arriveTangent=2.2631, interpMode=RCIM_Cubic, time=0, value=0}, {interpMode=RCIM_Cubic, leaveTangent=1.4308, time=1, value=0.125}]}}})
    Index 0 Param 4=0.5
    Module.Param0WriteEnabled=true
    Module.Param1WriteEnabled=false
    Module.Param2WriteEnabled=true
    Module.Param3WriteEnabled=true
   )
   &ScaleSpriteSize(Scale Factor=&Vector2DFromFloat(Value=&FloatFromCurve(
      FloatCurve=DataInterface() {propertyValues {Curve {keys [{arriveTangent=4.3948, interpMode=RCIM_Cubic, leaveTangent=4.3948, tangentMode=RCTM_User, time=0, value=0.5}, <at 1 3 RCIM_Cubic>]}}}
      Scale Curve=1.3
     )))
   &SpriteRotationRate(Rotation Rate=&UniformRangedFloat(Minimum=-1, Maximum=1))
  ]
  rendererClasses [&NiagaraSpriteRendererProperties]
  renderers [
   {rendererClass=&NiagaraSpriteRendererProperties}
   {rendererClass=&NiagaraSpriteRendererProperties, rendererIndex=1}
  ]
 }
 World_Smoke_Forward001 {
  emitterUpdateScript [
   &EmitterState(
    Enable Distance Culling=true
    MaxDistance=20000
    Module.bUseMaxDistance=true
    Module.bUseMinDistance=false
    Scalability Mode=::NewEnumerator1
   )
   &SpawnRate(
    SpawnRate=&MakeCustomFloatFromBool(
     Bool=&SetBoolByFloatComparison(A=$User.DamageScale, B=$User.Stage1, Comparison Type=::NewEnumerator2)
     True Float=&MakeCustomFloatFromBool(
      Bool=&SetBoolByFloatComparison(A=$User.DamageScale, B=$User.Stage2, Comparison Type=::NewEnumerator2)
      True Float=30
      False Float=20
     )
     False Float=0
    )
    Use Spawn Probability=false
   )
  ]
  particleSpawnScript [
   &FN_SetSpawnGroup(Number of Spawn Groups=2)
   &InitializeParticle(
    Color=<rgba 1 1 1 0.25>
    Lifetime=&UniformRangedFloat(Minimum=0.8, Maximum=1)
    Module.Write Color=true
    Module.Write Lifetime=true
    Module.Write Mass=true
    Module.Write Material Random=true
    Module.Write Position=false
    Module.Write Scale=false
    Module.Write SpriteRotation=true
    Module.Write SpriteSize=true
    Position=&SimulationPosition()
    Sprite Rotation=&NormalizedAngleToDegrees(NormalizedAngle=&UniformRangedFloat(Minimum=0, Maximum=1))
    Sprite Size=&Vector2DFromFloat(Value=&Multiply_Float(
      A=&UniformRangedFloat(Minimum=30, Maximum=50)
      B=&Lerp_Float(A=1, B=1.5, Alpha=$User.Vehicle_Speed)
     ))
   )
   &SkeletalMeshLocation(
    Bone / Socket Index=$Particles.SpawnGroup
    Bone Sampling Mode=::NewEnumerator4
    Inherit Velocity=false
    MaskSpawnGroup=false
    Sampled Position Offset=&SelectVectorFromArray(Direct Array Index=$Particles.SpawnGroup, Vector Selection Array=DataInterface() {propertyValues {FloatData [<120 -120 0>, <120 120 0>]}})
    Sampled Position Offset Coordinate Space:ENiagaraMeshTransforms=::NewEnumerator3
    Skeletal Mesh=$User.SkeletalMesh
   ) {module=SkeletalMeshLocation001}
   &ShapeLocation(MaskSpawnGroup=false, Non Uniform Scale=<1 1 0.5>, Offset Mode=::NewEnumerator0, Override Seed=false, Sphere Radius=30, UseOwnerScale=true)
   &AddVelocity(
    Additional Quaternion Rotation=false
    Module.UseConeFalloff=false
    Module.UseVelDistribution=false
    Override Seed=false
    Rotation Mode=::NewEnumerator3
    Rotation Quaternion:Quat4f=$Output.SkeletalMeshLocation001.SampledRotation
    Velocity Mode=::NewEnumerator2
    Velocity Speed=&RandomRangeFloat(Minimum=100, Maximum=300)
   ) {module=AddVelocity002}
   &CameraOffset(Camera Offset Amount=50)
   &Color(Color=&MakeLinearColorFromVectorAndFloat(
     "Vector (RGB)"=&MakeVectorFromLinearColorRGB(Color=&ColorFromCurve(
       ColorCurve=DataInterface() {propertyValues {
         AlphaCurve {keys [<at -0.0023 1 RCIM_Constant>,<at 0.0995 1 RCIM_Constant>,<at 1 1 RCIM_Constant>]}
         BlueCurve {keys [<at 0 0.1 RCIM_Constant>,<at 0.45 0.05 RCIM_Constant>,<at 1 0.0464 RCIM_Constant>]}
         GreenCurve {keys [<at 0 0.1 RCIM_Constant>,<at 0.45 0.05 RCIM_Constant>,<at 1 0.2858 RCIM_Constant>]}
         RedCurve {keys [<at 0 0.1 RCIM_Constant>,<at 0.45 0.05 RCIM_Constant>,<at 1 2 RCIM_Constant>]}
        }}
       CurveIndex=$User.DamageScale
      ))
     "Float (Alpha)"=1
    ))
   &BoxLocation(Box Size=<100 100 1>, MaskSpawnGroup=false, Module.Override Seed=false, Rotate Box=false, Surface Only=false)
  ]
  particleUpdateScript [
   &ParticleState()
   &AccelerationForce(Acceleration=&Multiply_VectorByFloat(
     Vector=<0 0 500>
     Float=&FN_ScalabilityFloatModifier(
      Value at DistanceFraction 0:NiagaraFloat=1
      Value at DistanceFraction 1:NiagaraFloat=3
      DistanceRamp:NiagaraDataInterfaceCurve=DataInterface() {propertyValues {Curve {keys [<at 0.1 0>,<at 0.2 0.5>,<at 0.7 1>]}}}
     )
    ))
   &CurlNoiseForce(
    Noise Frequency=0.5
    Noise Strength=&Lerp_Float(A=0, B=500, Alpha=&Power(a=$Particles.NormalizedAge, b=0.25))
    Pan Noise=true
    Pan Noise Field=<0 0 100>
   )
   &Drag(Drag=2, Use Linear Drag=true, Use Rotational Drag=false)
   &SolveForcesAndVelocity(Clamp Velocity=false, Limit Acceleration=false)
   &ScaleColor(
    Scale Alpha=&FloatFromCurve(
     FloatCurve=DataInterface() {propertyValues {Curve {keys [
         {interpMode=RCIM_Cubic, tangentMode=RCTM_User, time=0, value=0}
         {interpMode=RCIM_Cubic, tangentMode=RCTM_User, time=0.0958, value=0.9565}
         {arriveTangent=-2.0134, interpMode=RCIM_Cubic, leaveTangent=-2.0134, tangentMode=RCTM_User, time=0.594, value=0.7182}
         {interpMode=RCIM_Cubic, tangentMode=RCTM_User, time=1, value=0}
        ]}}}
     Scale Curve=&FN_ScalabilityFloatModifier(
      Value at DistanceFraction 0:NiagaraFloat=0.5
      Value at DistanceFraction 1:NiagaraFloat=1.25
      DistanceRamp:NiagaraDataInterfaceCurve=DataInterface() {propertyValues {Curve {keys [<at 0.1 0>,<at 0.2 0.5>,<at 0.7 1>]}}}
     )
    )
    Scale RGB=&VectorFromFloat(Value=&FloatFromCurve(FloatCurve=DataInterface() {propertyValues {Curve {keys [{arriveTangent=-3.2082, interpMode=RCIM_Cubic, leaveTangent=-1.8602, tangentMode=RCTM_User, time=0, value=1.25}, <at 1 0.15 RCIM_Cubic>]}}}))
    ScaleA=true
    ScaleRGB=true
   )
   &DynamicMaterialParameters(
    Index 0 Param 1=&FloatFromCurve(FloatCurve=DataInterface() {propertyValues {Curve {keys [
         {arriveTangent=1.3596, interpMode=RCIM_Cubic, leaveTangent=1.3596, tangentMode=RCTM_User, time=0, value=0}
         {arriveTangent=0.1028, interpMode=RCIM_Cubic, leaveTangent=0.1028, tangentMode=RCTM_User, time=1, value=0.75}
        ]}}})
    Index 0 Param 2=&FloatFromCurve(FloatCurve=DataInterface() {propertyValues {Curve {keys [
         {arriveTangent=0.6717, interpMode=RCIM_Cubic, leaveTangent=0.6717, tangentMode=RCTM_User, time=-0.0003, value=-0.9768}
         {arriveTangent=3.037, interpMode=RCIM_Cubic, leaveTangent=3.037, tangentMode=RCTM_User, time=1.0011, value=1}
        ]}}})
    Index 0 Param 4=0.5
    Index 2 Param 1=&FloatFromCurve(FloatCurve=DataInterface() {propertyValues {Curve {keys [
         {arriveTangent=2.5015, interpMode=RCIM_Cubic, leaveTangent=2.5015, tangentMode=RCTM_User, time=0, value=0}
         {arriveTangent=1.1306, interpMode=RCIM_Cubic, leaveTangent=1.1306, tangentMode=RCTM_User, time=0.25, value=0.5}
         <at 1 1>
        ]}}})
    Index 2 Param 2=&FloatFromCurve(FloatCurve=DataInterface() {propertyValues {Curve {keys [
         {arriveTangent=2.5015, interpMode=RCIM_Cubic, leaveTangent=2.5015, tangentMode=RCTM_User, time=0, value=0}
         {arriveTangent=1.1306, interpMode=RCIM_Cubic, leaveTangent=1.1306, tangentMode=RCTM_User, time=0.25, value=0.5}
         <at 1 1>
        ]}}})
    Index 2 Param 3=&FloatFromCurve(FloatCurve=DataInterface() {propertyValues {Curve {keys [
         {arriveTangent=0.6717, interpMode=RCIM_Cubic, leaveTangent=0.6717, tangentMode=RCTM_User, time=-0.0028, value=-0.0101}
         {arriveTangent=3.037, interpMode=RCIM_Cubic, leaveTangent=3.037, tangentMode=RCTM_User, time=1.0011, value=2}
        ]}}})
    Index 2 Param 4=&FloatFromCurve(FloatCurve=DataInterface() {propertyValues {Curve {keys [
         {arriveTangent=-0.4045, interpMode=RCIM_Cubic, leaveTangent=-0.4045, tangentMode=RCTM_User, time=-0.0192, value=1.0634}
         {arriveTangent=-0.318, interpMode=RCIM_Cubic, leaveTangent=-0.318, tangentMode=RCTM_User, time=0.9751, value=0.415}
        ]}}})
    Module.Param0WriteEnabled=true
    Module.Param0WriteEnabled2=true
    Module.Param1WriteEnabled=true
    Module.Param1WriteEnabled2=true
    Module.Param2WriteEnabled=true
    Module.Param2WriteEnabled2=true
    Module.Param3WriteEnabled=true
    Module.Param3WriteEnabled2=true
    Write Parameter Index 2=true
   )
   &ScaleSpriteSize(Scale Factor=&Vector2DFromFloat(Value=&FloatFromCurve(
      FloatCurve=DataInterface() {propertyValues {Curve {keys [{arriveTangent=4.3948, interpMode=RCIM_Cubic, leaveTangent=4.3948, tangentMode=RCTM_User, time=0, value=1}, <at 1 3 RCIM_Cubic>]}}}
      Scale Curve=1.1
     )))
   &ScaleSpriteSize(
    Scale Sprite Size Mode=::NewEnumerator0
    Uniform Scale Factor=&FN_ScalabilityFloatModifier(
     Value at DistanceFraction 0:NiagaraFloat=1
     Value at DistanceFraction 1:NiagaraFloat=1.5
     DistanceRamp:NiagaraDataInterfaceCurve=DataInterface() {propertyValues {Curve {keys [<at 0.1 0>,<at 0.2 0.5>,<at 0.7 1>]}}}
    )
   ) {module=ScaleSpriteSize001}
   &SpriteRotationRate(Rotation Rate=&UniformRangedFloat(Minimum=-1, Maximum=1))
  ]
  rendererClasses [&NiagaraSpriteRendererProperties]
  renderers [
   {rendererClass=&NiagaraSpriteRendererProperties}
   {rendererClass=&NiagaraSpriteRendererProperties, rendererIndex=1}
  ]
 }
}
summary {
 systemName=NS_Tank_Damage
 emitters [
  {emitterName=Local_Fire_Engine, rendererClasses [&NiagaraSpriteRendererProperties]}
  {emitterName=Embers, simTarget=GPUComputeSim, rendererClasses [&NiagaraSpriteRendererProperties]}
  {emitterName=World_Smoke_Forward001, rendererClasses [&NiagaraSpriteRendererProperties]}
  {emitterName=Local_Smoke_Forward001, rendererClasses [&NiagaraSpriteRendererProperties]}
  {emitterName=World_Smoke_Critical001, rendererClasses [&NiagaraSpriteRendererProperties]}
  {emitterName=Embers_LOWEND, rendererClasses [&NiagaraSpriteRendererProperties]}
 ]
 userVariables [
  User.DamageScale {type=NiagaraFloat, defaultValue=1, description=""}
  User.SimRT {
   type=Object
   description=""
   defaultValue {struct=NiagaraExt_VariableValue_Object, value {object=null, objectClass=&Object}}
  }
  User.SkeletalMesh {
   type=NiagaraDataInterfaceSkeletalMesh
   defaultValue {struct=NiagaraExt_VariableValue_DataInterface, value {dataInterface=&self:NiagaraDataInterfaceSkeletalMesh_0}}
  }
  User.SkeletalMeshTemp {
   type=NiagaraDataInterfaceSkeletalMesh
   defaultValue {struct=NiagaraExt_VariableValue_DataInterface, value {dataInterface=&self:NiagaraDataInterfaceSkeletalMesh_1}}
  }
  User.Stage1 {type=NiagaraFloat, defaultValue=0.33, description=""}
  User.Stage2 {type=NiagaraFloat, defaultValue=0.66, description=""}
  User.Stage3Critical {type=NiagaraFloat, defaultValue=1, description=""}
  User.Vehicle_Speed {type=NiagaraFloat, defaultValue=0.5, description=""}
 ]
}
