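# Niagara system for tank driving effects, with tread dust, dirt clumps and streaks, plus exhaust airflow, smoke, flame and light.
# Spawn groups index mesh sockets and per-socket offset arrays to place and orient particles.
# User parameters for speed, slide, boost, driver presence and a dust color channel gate or scale spawning, size and velocity.
# Exhaust flame and flickering light spawn on boost or peel-out, with deep water zeroing the flame rate.
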NiagaraSystem /Tank/FX/NS_Tank_Tread_Interaction
dependencies {
 usedDataInterfaces [
  &NiagaraDataInterfaceColorCurve
  &NiagaraDataInterfaceArrayFloat3
  &NiagaraDataInterfaceMeshRendererInfo
  &NiagaraDataInterfaceVector2DCurve
  &NiagaraDataInterfaceVectorCurve
 ]
 usedDynamicInputs []
 usedModules [&SystemState]
}
emitters {
 Airflow {
  emitterUpdateScript [
   &EmitterState()
   &SpawnRate(
    Spawn Probability=&MakeCustomFloatFromBool(
     Bool=$User.VehicleHasDriver
     True Float=&MakeCustomFloatFromBool(
      Bool=&SetBoolByFloatComparison(
       A=&MakeFloatFromLinearColor(LinearColor=$User.DustTireBlurColor, Channel:ENiagaraLinearColor_Channels=::NewEnumerator0)
       B=1
       Comparison Type=::NewEnumerator4
      )
      True Float=1
      False Float=0
     )
     False Float=0
    )
    SpawnRate=16
    Use Spawn Probability=true
   )
  ]
  particleSpawnScript [
   &FN_SetSpawnGroup()
   &InitializeParticle(
    Color=&MultiplyLinearColor(
     Linear Color=<rgba 0.2128 0.2241 0.2656 1.5>
     Scale Factor=&MakeFloat4FromVectorAndFloat(
      "Vector (RGB)"=&VectorFromFloat(Value=&Lerp_Float(A=1, B=0.7, Alpha=&MakeCustomFloatFromBool(Bool=$User.BoostFX, True Float=1, False Float=0)))
      "Float (Alpha)"=1
     )
    )
    Lifetime Max=0.75
    Lifetime Min=0.55
    Lifetime Mode=::NewEnumerator1
    Mesh Scale Mode=::NewEnumerator3
    Mesh Uniform Scale=&Multiply_Float(
     A=&RandomRangeFloat(Minimum=0.7, Maximum=0.85)
     B=&Lerp_Float(
      A=8
      B=&Lerp_Float(A=6, B=9, Alpha=&MakeCustomFloatFromBool(Bool=$User.BoostFX, True Float=1, False Float=0))
      Alpha=&FloatFromCurve(
       FloatCurve=DataInterface() {propertyValues {Curve {keys [{interpMode=RCIM_Cubic, tangentMode=RCTM_Break, time=0, value=0}, {interpMode=RCIM_Cubic, tangentMode=RCTM_Break, time=0.5, value=1}]}}}
       CurveIndex=$User.Vehicle_Speed
      )
     )
    )
    UsePositionOffset=false
    Write Material Random=true
   )
   &SkeletalMeshLocation(
    Bone / Socket Index=$Particles.SpawnGroup
    Interpolate Mesh Sampling=false
    MaskSpawnGroup=false
    Mesh Sampling Type=::NewEnumerator5
    Orientation Sampling=::NewEnumerator1
    Sampled Position Offset=&SelectVectorFromArray(
     Direct Array Index=$Particles.SpawnGroup
     Vector Selection Array=DataInterface() {propertyValues {FloatData [<-15 5 -10>, <-10 5 20>, <0 20 -40>, <5 20 40>]}}
    )
    Skeletal Mesh=$User.Vehicle_Selection_Exhaust
    Socket Sampling Mode=::NewEnumerator5
   )
   &AddVelocity(Velocity=&Multiply_VectorByFloat(
     Vector=&SelectVectorFromArray(Direct Array Index=$Particles.SpawnGroup, Vector Selection Array=DataInterface() {propertyValues {FloatData [<0 1 0>, <0 -1 0>]}})
     Float=&RandomRangeFloat(Minimum=25, Maximum=75)
    ))
   &InitialMeshOrientation(
    Mesh Axis To Orient=<0 1 0>
    Mesh Orientation Mode=::NewEnumerator2
    Orientation Vector=&SelectVectorFromArray(
     Direct Array Index=$Particles.SpawnGroup
     Vector Selection Array=DataInterface() {propertyValues {FloatData [<0 0 -1>, <0 0 1>, <0 0 -1>, <0 0 1>]}}
    )
    Rotation=&Add_Vector(
     A=&RandomRangeVector(Minimum=<-0.01 -0.01 0>, Maximum=<0.01 0.01 0>)
     B=&MakeVector(
      X=0
      Y=&Lerp_Float(A=0, B=0.05, Alpha=&MakeCustomFloatFromBool(Bool=$User.BoostFX, True Float=1, False Float=0))
      Z=0
     )
    )
    Use Rotation Vector=true
   )
  ]
  particleUpdateScript [
   &ParticleState()
   &ScaleColor(
    Scale Alpha=&FloatFromCurve(
     FloatCurve=DataInterface() {propertyValues {Curve {keys [
         {interpMode=RCIM_Cubic, tangentMode=RCTM_Break, time=0, value=0}
         {interpMode=RCIM_Cubic, tangentMode=RCTM_Break, time=0.1, value=0.8}
         {interpMode=RCIM_Cubic, tangentMode=RCTM_Break, time=1, value=0}
        ]}}}
     Scale Curve=&RandomRangeFloat(Minimum=10, Maximum=20)
    )
    Scale RGB=&MakeVectorFromLinearColorRGB(Color=&ColorFromCurve(ColorCurve=DataInterface() {propertyValues {
        AlphaCurve {keys []}
        BlueCurve {keys [<at 0.0727 0.5365>, {interpMode=RCIM_Cubic, leaveTangent=0.0185, time=1, value=0.1146}]}
        GreenCurve {keys [<at 0.0727 0.4763>, {interpMode=RCIM_Cubic, leaveTangent=-0.7509, time=1, value=0.0813}]}
        RedCurve {keys [<at 0.0727 0.4613>, {interpMode=RCIM_Cubic, leaveTangent=-1.9276, time=1, value=0.0724}]}
       }}))
    ScaleA=true
    ScaleRGB=true
   )
   &ScaleMeshSize(Scale Factor=&VectorFromCurve(VectorCurve=DataInterface() {propertyValues {XCurve {keys [<at 0 1>,<at 1 1>]}, YCurve {keys [<at 0 1>,<at 1 1>]}, ZCurve {keys [<at 0 0.2>,<at 1 1>]}}}))
   &SolveForcesAndVelocity(Clamp Velocity=false, Limit Acceleration=false)
   &DynamicMaterialParameters(
    Index 0 Param 1=&FloatFromCurve(
     FloatCurve=DataInterface() {propertyValues {Curve {keys [{interpMode=RCIM_Cubic, tangentMode=RCTM_Break, time=0, value=0.2}, {interpMode=RCIM_Cubic, tangentMode=RCTM_Break, time=1, value=1}]}}}
     Scale Curve=1.5
    )
    Index 0 Param 2=&FloatFromCurve(FloatCurve=DataInterface() {propertyValues {Curve {keys [{tangentMode=RCTM_User, time=0, value=0.5}, {tangentMode=RCTM_User, time=1, value=1.5}]}}})
    Index 0 Param 3=1
    Index 0 Param 4=1
    Module.Param0WriteEnabled=true
    Module.Param1WriteEnabled=true
    Module.Param2WriteEnabled=true
    Module.Param3WriteEnabled=true
   )
  ]
  rendererClasses [&NiagaraMeshRendererProperties]
  renderers [{rendererClass=&NiagaraMeshRendererProperties, rendererIndex=0}]
 }
 DirtClumps {
  emitterUpdateScript [
   &EmitterState(Life Cycle Mode=::NewEnumerator1, Loop Duration=1, UseLoopDelay=false)
   &SpawnRate(
    Spawn Probability=&MakeCustomFloatFromBool(
     Bool=&SetBoolByFloatComparison(A=$User.Vehicle_Speed, B=0.1)
     True Float=&MakeCustomFloatFromBool(
      Bool=&SetBoolByFloatComparison(
       A=&MakeFloatFromLinearColor(LinearColor=$User.DustTireBlurColor, Channel:ENiagaraLinearColor_Channels=::NewEnumerator0)
       B=1
       Comparison Type=::NewEnumerator4
      )
      True Float=1
      False Float=0
     )
     False Float=0
    )
    SpawnGroup=3
    SpawnRate=&Lerp_Float(
     A=10
     B=&Lerp_Float(A=16, B=60, Alpha=&MakeCustomFloatFromBool(Bool=$User.BoostFX, True Float=1, False Float=0))
     Alpha=&FloatFromCurve(
      FloatCurve=DataInterface() {propertyValues {Curve {keys [{interpMode=RCIM_Cubic, tangentMode=RCTM_Break, time=0, value=0}, {interpMode=RCIM_Cubic, tangentMode=RCTM_Break, time=1, value=1}]}}}
      CurveIndex=$User.Vehicle_Speed
     )
    )
    Use Spawn Probability=true
   )
   /SetVariables(Emitter.UseDynamicParamter=1) {isSetParameters=true}
  ]
  particleSpawnScript [
   &FN_SetSpawnGroup()
   /SetVariables(Particles.SizeMax=450, Particles.SizeMin=150) {isSetParameters=true, module=SetVariables_#1}
   /SetVariables(Particles.SizeCurve:NiagaraFloat=&Lerp_Float(
     A=$Particles.SizeMin
     B=$Particles.SizeMax
     Alpha=&FloatFromCurve(
      FloatCurve=DataInterface() {propertyValues {Curve {keys [
          {interpMode=RCIM_Cubic, tangentMode=RCTM_Break, time=0, value=0}
          {arriveTangent=2.305, interpMode=RCIM_Cubic, tangentMode=RCTM_Break, time=1, value=1}
         ]}}}
      CurveIndex=&RandomRangeFloat(Minimum=0, Maximum=1)
     )
    )) {isSetParameters=true, module=SetVariables_#2}
   &InitializeParticle(
    Color Maximum=<rgba 0.2188 0.1981 0.18 1>
    Color Minimum=<rgba 0.1 0.0858 0.0818 1>
    Color Mode=::NewEnumerator2
    Lifetime=&Lerp_Float(
     A=&UniformRangedFloat(Minimum=0.5, Maximum=1)
     B=&UniformRangedFloat(Minimum=0.4, Maximum=0.9)
     Alpha=$User.Vehicle_Speed
    )
    Sprite Rotation Mode=::NewEnumerator2
    Sprite Size Mode=::NewEnumerator3
    Uniform Sprite Size=&Lerp_Float(
     A=&Lerp_Float(
      A=&UniformRangedFloat(Minimum=60, Maximum=100)
      B=&UniformRangedFloat(Minimum=120, Maximum=140)
      Alpha=$User.Vehicle_Speed
     )
     B=100
     Alpha=&self:ScratchAbs(New float:NiagaraFloat=$User.Slide_Direction)
    )
    UsePositionOffset=false
    Write Material Random=true
   )
   &SkeletalMeshLocation(
    Bone / Socket Index=&Add_Integer(
     A=$Particles.SpawnGroup
     B=&MakeCustomIntFromBool(
      Bool=&SetBoolByIntComparison(A=$Particles.SpawnGroup, B=1, Comparison Type=::NewEnumerator7)
      True INT=&MakeIntFromFloat(Float=&UniformAOrBFloat(A=1, B=0))
      False INT=0
     )
    )
    Interpolate Mesh Sampling=false
    MaskSpawnGroup=false
    Mesh Sampling Type=::NewEnumerator5
    Orientation Sampling=::NewEnumerator1
    Sampled Position Offset=&SelectVectorFromArray(
     Direct Array Index=$Particles.SpawnGroup
     Vector Selection Array=DataInterface() {propertyValues {FloatData [<-25 -30 -10>, <-25 30 -10>, <10 -10 -10>, <10 10 -10>]}}
    )
    Skeletal Mesh=$User.Vehicle_Selection_02
    Socket Sampling Mode=::NewEnumerator5
   )
   &ShapeLocation(
    Box Size=&SelectVectorFromArray(
     Direct Array Index=$Particles.SpawnGroup
     Vector Selection Array=DataInterface() {propertyValues {FloatData [<200 50 10>, <200 50 10>, <50 50 10>, <50 50 10>]}}
    )
    MaskSpawnGroup=false
    Override Seed=false
    Shape Primitive=::NewEnumerator2
    Surface Only=false
    UseOwnerScale=true
   )
   &AddVelocity(Velocity=&MakeVector(X=0, Y=&Multiply_Float(A=$User.Slide_Direction, B=300), Z=0))
   &AddVelocity(Velocity=&Multiply_VectorByFloat(
     Vector=&SelectVectorFromArray(
      Direct Array Index=$Particles.SpawnGroup
      Vector Selection Array=DataInterface() {propertyValues {FloatData [<20 700 500>, <20 -700 500>, <50 100 400>, <50 -100 400>]}}
     )
     Float=&RandomRangeFloat(Minimum=0.35, Maximum=1)
    )) {module=AddVelocity001}
  ]
  particleUpdateScript [
   &ParticleState()
   &GravityForce(Gravity=&RandomRangeVector(Minimum=<0 0 -500>, Maximum=<0 0 -980>))
   &Drag(
    Drag=&FloatFromCurve(
     FloatCurve=DataInterface() {propertyValues {Curve {keys [
         {interpMode=RCIM_Cubic, leaveTangentWeight=1.071, tangentMode=RCTM_Break, time=0, value=1}
         {interpMode=RCIM_Cubic, tangentMode=RCTM_Break, time=0.75, value=0.1}
        ]}}}
     Scale Curve=5
    )
    Use Linear Drag=true
    Use Rotational Drag=false
   )
   &SolveForcesAndVelocity(Clamp Velocity=false, Limit Acceleration=false)
   &ScaleSpriteSize(
    Uniform Curve Scale=&Lerp_Float(
     A=1
     B=&Lerp_Float(A=1, B=1, Alpha=&MakeCustomFloatFromBool(Bool=$User.BoostFX, True Float=1, False Float=0))
     Alpha=&FloatFromCurve(
      FloatCurve=DataInterface() {propertyValues {Curve {keys [
          {interpMode=RCIM_Cubic, tangentMode=RCTM_Break, time=0, value=0}
          {arriveTangent=2.5761, interpMode=RCIM_Cubic, tangentMode=RCTM_Break, time=1, value=1}
         ]}}}
      CurveIndex=$User.Vehicle_Speed
     )
    )
    Uniform Curve Sprite Scale=DataInterface() {propertyValues {Curve {keys [
        {arriveTangent=4.3762, interpMode=RCIM_Cubic, leaveTangent=4.3762, tangentMode=RCTM_User, time=0, value=0.7}
        {arriveTangentWeight=0.1414, interpMode=RCIM_Cubic, leaveTangentWeight=0.3299, tangentMode=RCTM_User, time=0.2, value=1}
        {arriveTangent=-3.2279, interpMode=RCIM_Cubic, leaveTangent=-3.2279, tangentMode=RCTM_User, time=1, value=0.3}
       ]}}}
   )
   &ScaleColor(
    Scale Alpha=&RandomRangeFloat(Minimum=1.2, Maximum=1.75)
    Scale RGB=&VectorFromFloat(Value=&RandomRangeFloat(Minimum=0.4, Maximum=1))
    ScaleA=true
    ScaleRGB=true
   )
   &SpriteRotationRate(Rotation Rate=&Lerp_Float(
     A=1
     B=1
     Alpha=&FloatFromCurve(
      FloatCurve=DataInterface() {propertyValues {Curve {keys [
          {interpMode=RCIM_Cubic, leaveTangent=-13.5937, tangentMode=RCTM_Break, time=0, value=1}
          {interpMode=RCIM_Cubic, tangentMode=RCTM_Break, time=0.2, value=0}
         ]}}}
      Scale Curve=&RandomRangeFloat(Minimum=-1, Maximum=1)
     )
    ))
   &DynamicMaterialParameters(
    Index 0 Param 1=&FloatFromCurve(FloatCurve=DataInterface() {propertyValues {Curve {keys [
         {interpMode=RCIM_Cubic, leaveTangentWeight=0.8293, tangentMode=RCTM_Break, time=0, value=0.1}
         {arriveTangentWeight=0.259, interpMode=RCIM_Cubic, tangentMode=RCTM_Break, time=0.9, value=1}
        ]}}})
    Module.Param0WriteEnabled=true
    Module.Param1WriteEnabled=true
    Module.Param2WriteEnabled=true
    Module.Param3WriteEnabled=true
   )
  ]
  rendererClasses [&NiagaraSpriteRendererProperties]
  renderers [{rendererClass=&NiagaraSpriteRendererProperties}]
 }
 DirtStreaks {
  emitterUpdateScript [
   &EmitterState(Life Cycle Mode=::NewEnumerator1, Loop Duration=1, UseLoopDelay=false)
   &SpawnRate(
    Spawn Probability=&MakeCustomFloatFromBool(
     Bool=&SetBoolByFloatComparison(A=$User.Vehicle_Speed, B=0.1)
     True Float=&MakeCustomFloatFromBool(
      Bool=&SetBoolByFloatComparison(
       A=&MakeFloatFromLinearColor(LinearColor=$User.DustTireBlurColor, Channel:ENiagaraLinearColor_Channels=::NewEnumerator0)
       B=1
       Comparison Type=::NewEnumerator4
      )
      True Float=1
      False Float=0
     )
     False Float=0
    )
    SpawnRate=&Lerp_Float(
     A=35
     B=&Lerp_Float(A=60, B=100, Alpha=&MakeCustomFloatFromBool(Bool=$User.BoostFX, True Float=1, False Float=0))
     Alpha=&FloatFromCurve(
      FloatCurve=DataInterface() {propertyValues {Curve {keys [{interpMode=RCIM_Cubic, tangentMode=RCTM_Break, time=0, value=0}, {interpMode=RCIM_Cubic, tangentMode=RCTM_Break, time=1, value=1}]}}}
      CurveIndex=$User.Vehicle_Speed
     )
    )
    Use Spawn Probability=true
   )
   /SetVariables(Emitter.DepthFade=150) {isSetParameters=true, module=SetVariables_#3}
  ]
  particleSpawnScript [
   &FN_SetSpawnGroup()
   /SetVariables(Particles.SizeMax=450, Particles.SizeMin=150) {isSetParameters=true, module=SetVariables_#1}
   /SetVariables(Particles.SizeCurve:NiagaraFloat=&Lerp_Float(
     A=$Particles.SizeMin
     B=$Particles.SizeMax
     Alpha=&FloatFromCurve(
      FloatCurve=DataInterface() {propertyValues {Curve {keys [
          {interpMode=RCIM_Cubic, tangentMode=RCTM_Break, time=0, value=0}
          {arriveTangent=2.305, interpMode=RCIM_Cubic, tangentMode=RCTM_Break, time=1, value=1}
         ]}}}
      CurveIndex=&RandomRangeFloat(Minimum=0, Maximum=1)
     )
    )) {isSetParameters=true, module=SetVariables_#2}
   &InitializeParticle(
    Color Maximum=<rgba 0.2604 0.2171 0.1737 1>
    Color Minimum=<rgba 0.0781 0.0651 0.0521 1>
    Color Mode=::NewEnumerator2
    Lifetime=&Lerp_Float(
     A=&UniformRangedFloat(Minimum=0.4, Maximum=0.7)
     B=&UniformRangedFloat(Minimum=0.35, Maximum=0.5)
     Alpha=$User.Vehicle_Speed
    )
    Sprite Rotation Mode=::NewEnumerator2
    Sprite Size=&Multiply_Vector2D_ByFloat(
     Float=&Lerp_Float(
      A=0.4
      B=&Lerp_Float(A=0.7, B=1, Alpha=&MakeCustomFloatFromBool(Bool=$User.BoostFX, True Float=1, False Float=0))
      Alpha=$User.Vehicle_Speed
     )
     Vector2D=<75 100>
    )
    Sprite Size Mode=::NewEnumerator4
    UsePositionOffset=false
    Write Material Random=true
   )
   &SkeletalMeshLocation(
    Bone / Socket Index=&Add_Integer(
     A=$Particles.SpawnGroup
     B=&MakeCustomIntFromBool(
      Bool=&SetBoolByIntComparison(A=$Particles.SpawnGroup, B=1, Comparison Type=::NewEnumerator7)
      True INT=&MakeIntFromFloat(Float=&UniformAOrBFloat(A=0, B=0))
      False INT=0
     )
    )
    Interpolate Mesh Sampling=false
    MaskSpawnGroup=false
    Mesh Sampling Type=::NewEnumerator5
    Orientation Sampling=::NewEnumerator1
    Sampled Position Offset=&SelectVectorFromArray(
     Direct Array Index=$Particles.SpawnGroup
     Vector Selection Array=DataInterface() {propertyValues {FloatData [<-250 -30 -20>, <-250 30 -20>, <-10 -10 -20>, <-10 10 -20>]}}
    )
    Skeletal Mesh=$User.Vehicle_Selection_02
    Socket Sampling Mode=::NewEnumerator5
   )
   &ShapeLocation(
    Box Size=&SelectVectorFromArray(
     Direct Array Index=$Particles.SpawnGroup
     Vector Selection Array=DataInterface() {propertyValues {FloatData [<450 10 10>, <450 10 10>, <50 10 10>, <50 10 10>]}}
    )
    MaskSpawnGroup=false
    Override Seed=false
    Shape Primitive=::NewEnumerator2
    Surface Only=false
    UseOwnerScale=true
   )
   &AddVelocity(Velocity=&MakeVector(X=0, Y=&Multiply_Float(A=$User.Slide_Direction, B=300), Z=0)) {enabled=false}
   &AddVelocity(Velocity=&Multiply_VectorByFloat(
     Vector=&SelectVectorFromArray(
      Direct Array Index=$Particles.SpawnGroup
      Vector Selection Array=DataInterface() {propertyValues {FloatData [<75 100 25>, <75 -100 25>, <150 30 25>, <150 -30 25>]}}
     )
     Float=&RandomRangeFloat(Minimum=0.25, Maximum=1)
    )) {module=AddVelocity001}
   &SubUVAnimation(End Frame=3, SubUV Animation Mode:ENiagara_SubUVLookupMode=::NewEnumerator2)
  ]
  particleUpdateScript [
   &ParticleState()
   &SolveForcesAndVelocity(Clamp Velocity=false, Limit Acceleration=false)
   &ScaleSpriteSize(Scale Factor=&Vector2DFromCurve(
     Vector2Curve=DataInterface() {propertyValues {XCurve {keys [<at 0.0036 1.6897>,<at 0.9967 0.0582>]}, YCurve {keys [<at 0 1>,<at 1.002 3.7227>]}}}
     ScaleCurve=<1 1>
    ))
   &ScaleColor(
    Scale Alpha=&FloatFromCurve(
     FloatCurve=DataInterface() {propertyValues {Curve {keys [
         {interpMode=RCIM_Cubic, leaveTangentWeight=0.0962, tangentMode=RCTM_Break, time=0, value=0}
         {interpMode=RCIM_Cubic, tangentMode=RCTM_Break, time=0.2, value=1}
         {interpMode=RCIM_Cubic, tangentMode=RCTM_Break, time=1, value=0}
        ]}}}
     Scale Curve=&RandomRangeFloat(Minimum=0.35, Maximum=0.65)
    )
    ScaleA=true
    ScaleRGB=false
   )
   &DynamicMaterialParameters(
    Index 0 Param 2=1
    Index 0 Param 3=1
    Index 0 Param 4=&FloatFromCurve(
     FloatCurve=DataInterface() {propertyValues {Curve {keys [{interpMode=RCIM_Cubic, tangentMode=RCTM_Break, time=0.3, value=1}, {interpMode=RCIM_Cubic, tangentMode=RCTM_Break, time=1, value=0.25}]}}}
     Scale Curve=&RandomRangeFloat(Minimum=0.75, Maximum=1)
    )
    Module.Param0WriteEnabled=true
    Module.Param1WriteEnabled=true
    Module.Param2WriteEnabled=true
    Module.Param3WriteEnabled=true
   )
   &ScaleColor(
    Scale Alpha=&FNTimeOfDayScalar(Night Scale=0.3, Morning Scale=0.1, Evening Scale=0.2)
    ScaleA=true
    ScaleRGB=false
   ) {module=ScaleColor001}
  ]
  rendererClasses [&NiagaraSpriteRendererProperties]
  renderers [{rendererClass=&NiagaraSpriteRendererProperties}]
 }
 Dust_Tread {
  emitterUpdateScript [
   &EmitterState(Life Cycle Mode=::NewEnumerator1, Loop Duration=1, UseLoopDelay=false)
   &SpawnRate(
    Spawn Probability=&MakeCustomFloatFromBool(
     Bool=&SetBoolByFloatComparison(A=$User.Vehicle_Speed, B=0.1)
     True Float=&MakeCustomFloatFromBool(
      Bool=&SetBoolByFloatComparison(
       A=&MakeFloatFromLinearColor(LinearColor=$User.DustTireBlurColor, Channel:ENiagaraLinearColor_Channels=::NewEnumerator0)
       B=1
       Comparison Type=::NewEnumerator4
      )
      True Float=1
      False Float=0
     )
     False Float=0
    )
    SpawnGroup=3
    SpawnRate=&Lerp_Float(
     A=8
     B=&Lerp_Float(A=16, B=25, Alpha=&MakeCustomFloatFromBool(Bool=$User.BoostFX, True Float=1, False Float=0))
     Alpha=&FloatFromCurve(
      FloatCurve=DataInterface() {propertyValues {Curve {keys [{interpMode=RCIM_Cubic, tangentMode=RCTM_Break, time=0, value=0}, {interpMode=RCIM_Cubic, tangentMode=RCTM_Break, time=1, value=1}]}}}
      CurveIndex=$User.Vehicle_Speed
     )
    )
    Use Spawn Probability=true
   )
   /SetVariables(Emitter.DebugSpawnRate:NiagaraFloat=$User.Vehicle_Speed) {enabled=false, isSetParameters=true}
   /SetVariables(Emitter.DepthFadeHigh=400, Emitter.DepthFadeLow=200) {isSetParameters=true, module=SetVariables_#4}
  ]
  particleSpawnScript [
   &FN_SetSpawnGroup()
   /SetVariables(Particles.InitSpriteRotRate=&RandomRangeFloat(Minimum=-1, Maximum=1)) {isSetParameters=true, module=SetVariables_#5}
   /SetVariables(Particles.SizeCurve:NiagaraFloat=&FloatFromCurve(
     FloatCurve=DataInterface() {propertyValues {Curve {keys [
         {interpMode=RCIM_Cubic, tangentMode=RCTM_Break, time=0, value=0}
         {arriveTangent=1.7842, interpMode=RCIM_Cubic, tangentMode=RCTM_Break, time=1, value=1}
        ]}}}
     CurveIndex=&RandomRangeFloat(Minimum=0, Maximum=1)
     Scale Curve=2
    )) {enabled=false, isSetParameters=true}
   &InitializeParticle(
    Color Maximum=<rgba 0.666 0.5551 0.4442 1>
    Color Minimum=<rgba 0.333 0.2776 0.2221 0.5>
    Color Mode=::NewEnumerator2
    Lifetime=&Lerp_Float(
     A=&UniformRangedFloat(Minimum=1.4, Maximum=2)
     B=&UniformRangedFloat(Minimum=1.6, Maximum=1.4)
     Alpha=$User.Vehicle_Speed
    )
    Sprite Rotation Mode=::NewEnumerator2
    Sprite Size Mode=::NewEnumerator3
    Uniform Sprite Size=&Lerp_Float(
     A=&Lerp_Float(
      A=&UniformRangedFloat(Minimum=90, Maximum=150)
      B=&UniformRangedFloat(Minimum=350, Maximum=500)
      Alpha=$User.Vehicle_Speed
     )
     B=400
     Alpha=&self:ScratchAbs(New float:NiagaraFloat=$User.Slide_Direction)
    )
    UsePositionOffset=false
    Write Material Random=true
   )
   &SkeletalMeshLocation(
    Bone / Socket Index=$Particles.SpawnGroup
    Interpolate Mesh Sampling=false
    MaskSpawnGroup=false
    Mesh Sampling Type=::NewEnumerator5
    Orientation Sampling=::NewEnumerator1
    Sampled Position Offset=&SelectVectorFromArray(
     Direct Array Index=$Particles.SpawnGroup
     Vector Selection Array=DataInterface() {propertyValues {FloatData [<-160 -25 -10>, <-160 25 -10>, <50 10 -10>, <50 -10 -10>]}}
    )
    Skeletal Mesh=$User.Vehicle_Selection_02
    Socket Sampling Mode=::NewEnumerator5
   )
   &ShapeLocation(
    Box Size=&SelectVectorFromArray(
     Direct Array Index=$Particles.SpawnGroup
     Vector Selection Array=DataInterface() {propertyValues {FloatData [<400 50 10>, <400 50 10>, <100 50 10>, <100 50 10>]}}
    )
    MaskSpawnGroup=false
    Override Seed=false
    Shape Primitive=::NewEnumerator2
    Surface Only=false
    UseOwnerScale=true
   )
   &AddVelocity(Velocity=&MakeVector(X=0, Y=&Multiply_Float(A=$User.Slide_Direction, B=300), Z=0))
   &AddVelocity(
    Velocity=&Multiply_VectorByFloat(
     Vector=&SelectVectorFromArray(
      Direct Array Index=$Particles.SpawnGroup
      Vector Selection Array=DataInterface() {propertyValues {FloatData [<50 1100 600>, <50 -1100 600>, <50 500 500>, <50 -500 500>]}}
     )
     Float=&RandomRangeFloat(Minimum=0.25, Maximum=1)
    )
    Velocity Speed Scale=&Lerp_Float(
     A=0.5
     B=1
     Alpha=&FloatFromCurve(
      FloatCurve=DataInterface() {propertyValues {Curve {keys [
          {interpMode=RCIM_Cubic, tangentMode=RCTM_Break, time=0, value=0}
          {arriveTangentWeight=0.9211, interpMode=RCIM_Cubic, tangentMode=RCTM_Break, time=1, value=1}
         ]}}}
      CurveIndex=$User.Vehicle_Speed
     )
    )
   ) {module=AddVelocity001}
  ]
  particleUpdateScript [
   &ParticleState()
   &GravityForce(Gravity=&RandomRangeVector(Minimum=<0 0 -50>, Maximum=<0 0 -300>))
   &Drag(Drag=5, Use Linear Drag=true, Use Rotational Drag=false)
   &ScaleColor(
    Scale Alpha=&FloatFromCurve(
     FloatCurve=DataInterface() {propertyValues {Curve {keys [
         {arriveTangent=-0.0075, interpMode=RCIM_Cubic, leaveTangent=-0.0075, tangentMode=RCTM_User, time=0, value=0}
         {interpMode=RCIM_Cubic, tangentMode=RCTM_User, time=0.3, value=1}
         {interpMode=RCIM_Cubic, tangentMode=RCTM_Break, time=0.9, value=1}
         {interpMode=RCIM_Cubic, tangentMode=RCTM_Break, time=1, value=0}
        ]}}}
     Scale Curve=&Multiply_Float(
      A=&RandomRangeFloat(Minimum=0.3, Maximum=0.6)
      B=&Lerp_Float(A=0.6, B=1, Alpha=&MakeCustomFloatFromBool(Bool=$User.BoostFX, True Float=1, False Float=0))
     )
    )
    Scale RGB=&VectorFromFloat(Value=1)
    ScaleA=true
    ScaleRGB=false
   )
   &SolveForcesAndVelocity(Clamp Velocity=false, Limit Acceleration=false)
   &ScaleSpriteSize(Scale Factor=&Vector2DFromFloat(Value=&FloatFromCurve(
      FloatCurve=DataInterface() {propertyValues {Curve {keys [
          {arriveTangent=0.3223, interpMode=RCIM_Cubic, leaveTangent=0.3223, tangentMode=RCTM_User, time=0, value=0.25}
          {arriveTangent=0.9419, interpMode=RCIM_Cubic, leaveTangent=0.9419, tangentMode=RCTM_User, time=0.6, value=0.85}
          <at 1 1>
         ]}}}
      Scale Curve=1.5
     )))
   &DynamicMaterialParameters(
    Index 0 Param 1=&Add_Float(
     A=&RandomRangeFloat(Minimum=0, Maximum=1)
     B=&FloatFromCurve(
      FloatCurve=DataInterface() {propertyValues {Curve {keys [
          {arriveTangent=2.7533, interpMode=RCIM_Cubic, leaveTangent=1.8958, tangentMode=RCTM_User, time=0, value=1}
          {arriveTangent=0.8292, interpMode=RCIM_Cubic, leaveTangent=0.8292, tangentMode=RCTM_User, time=0.321, value=1.3453}
          {arriveTangent=0.0529, interpMode=RCIM_Cubic, leaveTangent=0.2021, tangentMode=RCTM_User, time=1, value=1.5}
         ]}}}
      CurveIndex=&Add_Float(A=&Multiply_Float(A=$Particles.Initial.MaterialRandom, B=0.5), B=$Particles.NormalizedAge)
     )
    )
    Index 0 Param 4=&FloatFromCurve(FloatCurve=DataInterface() {propertyValues {Curve {keys [{arriveTangent=-3.2497, interpMode=RCIM_Cubic, leaveTangent=-3.2497, tangentMode=RCTM_User, time=0, value=1}, <at 1 0 RCIM_Cubic>]}}})
    Index 2 Param 1=&FloatFromCurve(FloatCurve=DataInterface() {propertyValues {Curve {keys [
         {interpMode=RCIM_Cubic, leaveTangent=3.0943, tangentMode=RCTM_Break, time=0, value=0}
         {interpMode=RCIM_Cubic, tangentMode=RCTM_Break, time=1, value=1}
        ]}}})
    Index 2 Param 2=&FloatFromCurve(
     FloatCurve=DataInterface() {propertyValues {Curve {keys [
         {interpMode=RCIM_Cubic, leaveTangentWeight=0.397, tangentMode=RCTM_Break, time=0, value=0.7}
         {arriveTangentWeight=0.4342, interpMode=RCIM_Cubic, tangentMode=RCTM_Break, time=1, value=1}
        ]}}}
     Scale Curve=&RandomRangeFloat(Minimum=0, Maximum=1)
    )
    Index 2 Param 3=&FloatFromCurve(
     FloatCurve=DataInterface() {propertyValues {Curve {keys [
         {interpMode=RCIM_Cubic, leaveTangent=-6.5865, tangentMode=RCTM_Break, time=0, value=0.5}
         {interpMode=RCIM_Cubic, leaveTangentWeight=0.3354, tangentMode=RCTM_Break, time=0.2, value=0}
         {arriveTangentWeight=0.3308, interpMode=RCIM_Cubic, tangentMode=RCTM_Break, time=1, value=1}
        ]}}}
     Scale Curve=6
    )
    Index 2 Param 4=&FloatFromCurve(FloatCurve=DataInterface() {propertyValues {Curve {keys [
         {interpMode=RCIM_Cubic, leaveTangentWeight=0.4044, tangentMode=RCTM_Break, time=0, value=0.8}
         {arriveTangent=0.4016, arriveTangentWeight=0.382, interpMode=RCIM_Cubic, tangentMode=RCTM_Break, time=1, value=1}
        ]}}})
    Module.Param0WriteEnabled=true
    Module.Param0WriteEnabled2=true
    Module.Param1WriteEnabled=false
    Module.Param1WriteEnabled2=true
    Module.Param2WriteEnabled=true
    Module.Param2WriteEnabled2=true
    Module.Param3WriteEnabled=true
    Module.Param3WriteEnabled2=true
    Write Parameter Index 2=true
   )
   /SetVariables(Particles.DebugSpeed:NiagaraFloat=$User.Slide_Direction) {isSetParameters=true}
   &SpriteRotationRate(Rotation Rate=&Multiply_Float(
     A=$Particles.InitSpriteRotRate
     B=&FloatFromCurve(
      FloatCurve=DataInterface() {propertyValues {Curve {keys [
          {interpMode=RCIM_Cubic, leaveTangent=-2.6704, tangentMode=RCTM_Break, time=0, value=1}
          {arriveTangent=-0.4702, interpMode=RCIM_Cubic, leaveTangent=-0.4702, tangentMode=RCTM_User, time=0.2, value=0.2}
          {interpMode=RCIM_Cubic, tangentMode=RCTM_Break, time=1, value=0.1}
         ]}}}
      Scale Curve=400
     )
    ))
  ]
  rendererClasses [&NiagaraSpriteRendererProperties]
  renderers [
   {rendererClass=&NiagaraSpriteRendererProperties}
   {rendererClass=&NiagaraSpriteRendererProperties, rendererIndex=1}
  ]
 }
 ExhaustFlame {
  emitterUpdateScript [
   &EmitterState(Life Cycle Mode=::NewEnumerator1, Loop Duration=0.2, UseLoopDelay=false)
   &SpawnRate(
    Spawn Probability=&MakeCustomFloatFromBool(Bool=&Boolean_OrOperation(Bool1=$User.PeelOut_Wheel_RL, Bool2=$User.BoostFX), True Float=1, False Float=0)
    SpawnRate=&MakeFloatFromInt(Int=&MakeIntFromFloat(Float=&MakeCustomFloatFromBool(Bool=$User.IsWaterDeep, True Float=0, False Float=75)))
    Use Spawn Probability=true
   )
   &SpawnRate(SpawnRate=100, Use Spawn Probability=false) {enabled=false, module=SpawnRate001}
  ]
  particleSpawnScript [
   &FN_SetSpawnGroup()
   &InitializeParticle(
    Color=<rgba 3 0.8285 0.1654 1>
    Lifetime Max=0.15
    Lifetime Min=0.07
    Lifetime Mode=::NewEnumerator1
    Mesh Scale=&Lerp_Vector(
     A=<0.7 0.5 0.5>
     B=<1.2 0.8 0.8>
     Alpha=&FloatFromCurve(
      FloatCurve=DataInterface() {propertyValues {Curve {keys [
          {interpMode=RCIM_Cubic, tangentMode=RCTM_Break, time=0, value=0}
          {arriveTangent=1.393, interpMode=RCIM_Cubic, tangentMode=RCTM_Break, time=1, value=1}
         ]}}}
      CurveIndex=&RandomRangeFloat(Minimum=0, Maximum=1)
     )
    )
    Mesh Scale Mode=::NewEnumerator4
    UsePositionOffset=false
    Write Material Random=true
   )
   &SkeletalMeshLocation(
    Bone / Socket Index=$Particles.SpawnGroup
    Interpolate Mesh Sampling=false
    MaskSpawnGroup=false
    Mesh Sampling Type=::NewEnumerator5
    Orientation Sampling=::NewEnumerator1
    Sampled Position Offset=&SelectVectorFromArray(
     Direct Array Index=$Particles.SpawnGroup
     Vector Selection Array=DataInterface() {propertyValues {FloatData [<-25 5 -20>, <-25 5 20>, <-10 18 -42>, <-10 20 45>]}}
    )
    Skeletal Mesh=$User.Vehicle_Selection_Exhaust
    Socket Sampling Mode=::NewEnumerator5
   )
   &InitialMeshOrientation(
    Mesh Orientation Mode=::NewEnumerator2
    Orientation Vector=&SelectVectorFromArray(
     Direct Array Index=$Particles.SpawnGroup
     Vector Selection Array=DataInterface() {propertyValues {FloatData [<0 1 0>, <0 -1 0>, <0 1 0>, <0 -1 0>]}}
    )
    Rotation=&Multiply_VectorByFloat(
     Vector=&Lerp_Vector(
      A=<0 0 0>
      B=<0 0 0.05>
      Alpha=&MakeCustomFloatFromBool(Bool=$User.BoostFX, True Float=1, False Float=0)
     )
     Float=&SelectFloatFromArray(Direct Array Index=$Particles.SpawnGroup, Float Selection Array=DataInterface() {propertyValues {FloatData [-1,1,-1,1]}})
    )
    Use Rotation Vector=true
   )
   &InitialMeshOrientation(
    Mesh Orientation Mode=::NewEnumerator2
    Orientation Vector=&SelectVectorFromArray(
     Direct Array Index=$Particles.SpawnGroup
     Vector Selection Array=DataInterface() {propertyValues {FloatData [<0 1 0>, <0 -1 0>, <0 1 0>, <0 -1 0>]}}
    )
    Rotation=&RandomRangeVector(Minimum=<0 0 0>, Maximum=<0 0 0>)
    Use Rotation Vector=true
   ) {enabled=false, module=InitialMeshOrientation001}
  ]
  particleUpdateScript [
   &ParticleState()
   &ScaleColor(
    Scale Alpha=&FloatFromCurve(FloatCurve=DataInterface() {propertyValues {Curve {keys [
         {interpMode=RCIM_Cubic, leaveTangent=9.8521, tangentMode=RCTM_Break, time=0, value=0.1}
         {interpMode=RCIM_Cubic, tangentMode=RCTM_User, time=0.2, value=1}
         {interpMode=RCIM_Cubic, tangentMode=RCTM_Break, time=1, value=0}
        ]}}})
    Scale RGB=&VectorFromFloat(Value=&RandomRangeFloat(Minimum=1, Maximum=10))
    ScaleA=true
    ScaleRGB=true
   )
   &ScaleMeshSize(Scale Factor=&VectorFromCurve(
     VectorCurve=DataInterface() {propertyValues {
       bOptimizeLUT=false
       XCurve {keys [{arriveTangent=-0.4338, interpMode=RCIM_Cubic, leaveTangent=-0.4338, tangentMode=RCTM_User, time=0, value=0.75}, <at 1 2>]}
       YCurve {keys [{arriveTangent=0.1422, interpMode=RCIM_Cubic, leaveTangent=0.1422, tangentMode=RCTM_User, time=0, value=0.75}, <at 1 2>]}
       ZCurve {keys [
         <at -0.0011 1.0067 RCIM_Cubic>
         {arriveTangent=-0.105, interpMode=RCIM_Cubic, leaveTangent=-0.105, tangentMode=RCTM_User, time=1.0016, value=1}
        ]}
      }}
     Scale Curve=&VectorFromFloat(Value=1.3)
    ))
   /Game/Effects/Niagara/Modules/Materials/MaterialPanner(
    Jitter Pan by SpawnGroups=false
    Offset:NiagaraFloat=&Add_Float(A=$Emitter.Age, B=$Particles.Initial.MaterialRandom)
    Speed=-5
   )
   &ScaleMeshSize(Scale Factor=&MakeVector(
     X=&Lerp_Float(
      A=1
      B=&Lerp_Float(A=0.3, B=0.6, Alpha=&MakeCustomFloatFromBool(Bool=$User.BoostFX, True Float=1, False Float=0))
      Alpha=&FloatFromCurve(
       FloatCurve=DataInterface() {propertyValues {Curve {keys [{interpMode=RCIM_Cubic, tangentMode=RCTM_Break, time=0, value=0}, {interpMode=RCIM_Cubic, tangentMode=RCTM_Break, time=0.5, value=1}]}}}
       CurveIndex=$User.Vehicle_Speed
      )
     )
     Y=1
     Z=1
    )) {module=ScaleMeshSize001}
  ]
  rendererClasses [&NiagaraMeshRendererProperties]
  renderers [{rendererClass=&NiagaraMeshRendererProperties, rendererIndex=0}]
 }
 ExhaustLight {
  emitterUpdateScript [
   &EmitterState(Life Cycle Mode=::NewEnumerator1, Loop Duration=0.1, UseLoopDelay=false)
   &SpawnBurst_Instantaneous(
    Spawn Count=2
    Spawn Probability=&MakeCustomFloatFromBool(Bool=&Boolean_OrOperation(Bool1=$User.PeelOut_Wheel_RL, Bool2=$User.BoostFX), True Float=1, False Float=0)
    Use Spawn Probability=true
    UseLoopCountLimit=false
   )
  ]
  particleSpawnScript [
   &FN_SetSpawnGroup(Number of Spawn Groups=2)
   &InitializeParticle(
    Color=<rgba 3 0.8285 0.1654 1>
    Lifetime=0.1
    Sprite Size Mode=::NewEnumerator3
    Uniform Sprite Size=30
    UsePositionOffset=false
    Write Material Random=true
   )
   &SkeletalMeshLocation(
    Bone / Socket Index=$Particles.SpawnGroup
    Interpolate Mesh Sampling=false
    MaskSpawnGroup=false
    Mesh Sampling Type=::NewEnumerator5
    Orientation Sampling=::NewEnumerator1
    Sampled Position Offset=&SelectVectorFromArray(Direct Array Index=$Particles.SpawnGroup, Vector Selection Array=DataInterface() {propertyValues {FloatData [<-75 25 -30>, <-75 25 30>]}})
    Skeletal Mesh=$User.Vehicle_Selection_Exhaust
    Socket Sampling Mode=::NewEnumerator5
   )
  ]
  particleUpdateScript [
   &ParticleState()
   &FN_Light_Attributes(
    Color=<rgba 1 0.3436 0 1>
    Create Color Attribute=true
    Create Exponent Attribute=true
    Create Position Attribute=true
    Create Radius Attribute=true
    Position=&ConvertPositionToVector(Input Position=$Particles.Position, Convert To Absolute Large World Vector=false)
    Radius=400
    Scale Color=&Lerp_Float(
     A=0
     B=&Multiply_Float(
      A=&Lerp_Float(A=1, B=2, Alpha=&MakeCustomFloatFromBool(Bool=$User.BoostFX, True Float=1, False Float=0))
      B=&Waveform("Waveform [1]"=::NewEnumerator0, "[1] Frequency"=10, "[1] Amplitude Min/Max"=<0.85 1>)
     )
     Alpha=&FloatFromCurve(
      FloatCurve=DataInterface() {propertyValues {Curve {keys [{interpMode=RCIM_Cubic, tangentMode=RCTM_Break, time=0, value=0}, {interpMode=RCIM_Cubic, tangentMode=RCTM_Break, time=0.8, value=1}]}}}
      CurveIndex=$User.Vehicle_Speed
     )
    )
   )
  ]
  rendererClasses [&NiagaraLightRendererProperties, &NiagaraSpriteRendererProperties]
  renderers [
   {rendererClass=&NiagaraLightRendererProperties, rendererIndex=0}
   {rendererClass=&NiagaraSpriteRendererProperties, rendererIndex=1}
  ]
 }
 ExhaustSmoke {
  emitterUpdateScript [
   &EmitterState(Life Cycle Mode=::NewEnumerator1, Loop Duration=1, UseLoopDelay=false)
   &SpawnRate(
    Spawn Probability=&MakeCustomFloatFromBool(
     Bool=$User.VehicleHasDriver
     True Float=&MakeCustomFloatFromBool(
      Bool=&SetBoolByFloatComparison(
       A=&MakeFloatFromLinearColor(LinearColor=$User.DustTireBlurColor, Channel:ENiagaraLinearColor_Channels=::NewEnumerator0)
       B=1
       Comparison Type=::NewEnumerator4
      )
      True Float=1
      False Float=0
     )
     False Float=0
    )
    SpawnRate=&Lerp_Float(
     A=18
     B=&Lerp_Float(A=16, B=22, Alpha=&MakeCustomFloatFromBool(Bool=$User.BoostFX, True Float=1, False Float=0))
     Alpha=&FloatFromCurve(
      FloatCurve=DataInterface() {propertyValues {Curve {keys [{interpMode=RCIM_Cubic, tangentMode=RCTM_Break, time=0, value=0}, {interpMode=RCIM_Cubic, tangentMode=RCTM_Break, time=0.5, value=1}]}}}
      CurveIndex=$User.Vehicle_Speed
     )
    )
    Use Spawn Probability=true
   )
   /SetVariables(Emitter.DepthFadeHigh=400, Emitter.DepthFadeLow=200) {isSetParameters=true, module=SetVariables_#4}
  ]
  particleSpawnScript [
   /SetVariables(
    Particles.ColorBoost:LinearColor=&RandomRangeLinearColor(Minimum=<rgba 0.024 0.0217 0.0212 1>, Maximum=<rgba 0.09 0.0862 0.0824 1>)
    Particles.ColorNormal:LinearColor=&RandomRangeLinearColor(Minimum=<rgba 0.0885 0.0946 0.1042 1>, Maximum=<rgba 0.1385 0.1409 0.1458 1>)
   ) {isSetParameters=true}
   &FN_SetSpawnGroup(Number of Spawn Groups=2)
   /SetVariables(Particles.InitSpriteRotRate=&RandomRangeFloat(Minimum=-1, Maximum=1)) {isSetParameters=true, module=SetVariables_#5}
   &InitializeParticle(
    Color=&LerpLinearColors(
     StartColor=$Particles.ColorNormal
     EndColor=$Particles.ColorBoost
     LerpFactor=&MakeCustomFloatFromBool(Bool=$User.BoostFX, True Float=1, False Float=0)
    )
    Lifetime=&Lerp_Float(
     A=&UniformRangedFloat(Minimum=1.4, Maximum=1.6)
     B=&Multiply_Float(
      A=&UniformRangedFloat(Minimum=0.9, Maximum=1)
      B=&Lerp_Float(A=1, B=1.5, Alpha=&MakeCustomFloatFromBool(Bool=$User.BoostFX, True Float=1, False Float=0))
     )
     Alpha=$User.Vehicle_Speed
    )
    Sprite Size Mode=::NewEnumerator3
    Uniform Sprite Size=&Lerp_Float(
     A=&Lerp_Float(
      A=&UniformRangedFloat(Minimum=300, Maximum=200)
      B=&UniformRangedFloat(Minimum=180, Maximum=250)
      Alpha=$User.Vehicle_Speed
     )
     B=400
     Alpha=&MakeCustomFloatFromBool(Bool=$User.BoostFX, True Float=1, False Float=0)
    )
    UsePositionOffset=false
    Write Material Random=true
   )
   &InitializeParticle(
    Color=&MultiplyLinearColor(
     Linear Color=$Particles.ColorNormal
     Scale Factor=&MakeFloat4FromVectorAndFloat(
      "Vector (RGB)"=&VectorFromFloat(Value=&Lerp_Float(A=1, B=0.1, Alpha=&MakeCustomFloatFromBool(Bool=$User.BoostFX, True Float=1, False Float=0)))
      "Float (Alpha)"=1
     )
    )
    Lifetime=&Lerp_Float(
     A=&UniformRangedFloat(Minimum=1.4, Maximum=2)
     B=&UniformRangedFloat(Minimum=1.6, Maximum=1.4)
     Alpha=$User.Vehicle_Speed
    )
    Sprite Size Mode=::NewEnumerator3
    Uniform Sprite Size=&Lerp_Float(
     A=&Lerp_Float(
      A=&UniformRangedFloat(Minimum=90, Maximum=150)
      B=&UniformRangedFloat(Minimum=350, Maximum=500)
      Alpha=$User.Vehicle_Speed
     )
     B=400
     Alpha=&self:ScratchAbs(New float:NiagaraFloat=$User.Slide_Direction)
    )
    UsePositionOffset=false
    Write Material Random=true
   ) {enabled=false, module=InitializeParticle001}
   &SkeletalMeshLocation(
    Bone / Socket Index=$Particles.SpawnGroup
    Interpolate Mesh Sampling=false
    MaskSpawnGroup=false
    Mesh Sampling Type=::NewEnumerator5
    Orientation Sampling=::NewEnumerator1
    Sampled Position Offset=&SelectVectorFromArray(Direct Array Index=$Particles.SpawnGroup, Vector Selection Array=DataInterface() {propertyValues {FloatData [<40 50 10>, <40 50 10>]}})
    Skeletal Mesh=$User.Vehicle_Selection_Exhaust
    Socket Sampling Mode=::NewEnumerator5
    Velocity Sampling=::NewEnumerator0
    VelocitySpeedLimit=false
   )
   &ShapeLocation(MaskSpawnGroup=false, Override Seed=false, Sphere Radius=5, UseOwnerScale=true)
   &AddVelocity(Velocity=&Multiply_VectorByFloat(
     Vector=&SelectVectorFromArray(Direct Array Index=$Particles.SpawnGroup, Vector Selection Array=DataInterface() {propertyValues {FloatData [<0 500 0>, <0 -500 0>]}})
     Float=&Multiply_Float(
      A=&RandomRangeFloat(Minimum=0.75, Maximum=1)
      B=&Lerp_Float(
       A=1.1
       B=&Lerp_Float(A=0.9, B=2, Alpha=&MakeCustomFloatFromBool(Bool=$User.BoostFX, True Float=1, False Float=0))
       Alpha=&FloatFromCurve(
        FloatCurve=DataInterface() {propertyValues {Curve {keys [{interpMode=RCIM_Cubic, tangentMode=RCTM_Break, time=0, value=0}, {interpMode=RCIM_Cubic, tangentMode=RCTM_Break, time=0.5, value=1}]}}}
        CurveIndex=$User.Vehicle_Speed
       )
      )
     )
    )) {module=AddVelocity001}
   &AddVelocity(Velocity=&Multiply_VectorByFloat(
     Vector=&RandomRangeVector(Minimum=<400 0 25>, Maximum=<600 0 125>)
     Float=&MakeCustomFloatFromBool(Bool=$User.BoostFX, True Float=1, False Float=0)
    )) {module=AddVelocity002}
  ]
  particleUpdateScript [
   &ParticleState()
   &GravityForce(Gravity=&RandomRangeVector(Minimum=<0 0 200>, Maximum=<0 0 25>))
   &Drag(Use Linear Drag=true, Use Rotational Drag=false)
   &ScaleColor(
    Scale Alpha=&FloatFromCurve(
     FloatCurve=DataInterface() {propertyValues {Curve {keys [
         {arriveTangent=-0.0075, interpMode=RCIM_Cubic, leaveTangent=-0.0075, tangentMode=RCTM_User, time=0, value=0}
         {interpMode=RCIM_Cubic, tangentMode=RCTM_User, time=0.3, value=1}
         {interpMode=RCIM_Cubic, tangentMode=RCTM_Break, time=0.9, value=1}
         {interpMode=RCIM_Cubic, tangentMode=RCTM_Break, time=1, value=0}
        ]}}}
     Scale Curve=&RandomRangeFloat(Minimum=1, Maximum=1)
    )
    Scale RGB=&VectorFromFloat(Value=1)
    ScaleA=true
    ScaleRGB=false
   )
   &SolveForcesAndVelocity(Clamp Velocity=false, Limit Acceleration=false)
   &ScaleSpriteSize(Scale Factor=&Vector2DFromFloat(Value=&FloatFromCurve(
      FloatCurve=DataInterface() {propertyValues {Curve {keys [
          {interpMode=RCIM_Cubic, leaveTangentWeight=0.3336, tangentMode=RCTM_User, time=0, value=0.2}
          {arriveTangent=1.5237, interpMode=RCIM_Cubic, leaveTangent=1.5237, tangentMode=RCTM_User, time=0.5, value=0.8}
          <at 1 1>
         ]}}}
      Scale Curve=&Lerp_Float(
       A=1.1
       B=1.6
       Alpha=&FloatFromCurve(
        FloatCurve=DataInterface() {propertyValues {Curve {keys [{interpMode=RCIM_Cubic, tangentMode=RCTM_Break, time=0, value=0}, {interpMode=RCIM_Cubic, tangentMode=RCTM_Break, time=0.5, value=1}]}}}
        CurveIndex=$User.Vehicle_Speed
        Scale Curve=&Lerp_Float(A=1, B=2, Alpha=&MakeCustomFloatFromBool(Bool=$User.BoostFX, True Float=1, False Float=0))
       )
      )
     )))
   &DynamicMaterialParameters(
    Index 0 Param 1=&Add_Float(
     A=&RandomRangeFloat(Minimum=0, Maximum=1)
     B=&FloatFromCurve(
      FloatCurve=DataInterface() {propertyValues {Curve {keys [
          {arriveTangent=2.7533, interpMode=RCIM_Cubic, leaveTangent=1.8958, tangentMode=RCTM_User, time=0, value=1}
          {arriveTangent=0.8292, interpMode=RCIM_Cubic, leaveTangent=0.8292, tangentMode=RCTM_User, time=0.321, value=1.3453}
          {arriveTangent=0.0529, interpMode=RCIM_Cubic, leaveTangent=0.2021, tangentMode=RCTM_User, time=1, value=1.5}
         ]}}}
      CurveIndex=&Add_Float(A=&Multiply_Float(A=$Particles.Initial.MaterialRandom, B=0.5), B=$Particles.NormalizedAge)
     )
    )
    Index 0 Param 4=&FloatFromCurve(FloatCurve=DataInterface() {propertyValues {Curve {keys [{arriveTangent=-3.2497, interpMode=RCIM_Cubic, leaveTangent=-3.2497, tangentMode=RCTM_User, time=0, value=1}, <at 1 0 RCIM_Cubic>]}}})
    Index 2 Param 1=&FloatFromCurve(
     FloatCurve=DataInterface() {propertyValues {Curve {keys [
         {interpMode=RCIM_Cubic, leaveTangent=3.0943, tangentMode=RCTM_Break, time=0, value=0}
         {interpMode=RCIM_Cubic, tangentMode=RCTM_Break, time=1, value=1}
        ]}}}
     Scale Curve=2
    )
    Index 2 Param 2=&FloatFromCurve(
     FloatCurve=DataInterface() {propertyValues {Curve {keys [
         {interpMode=RCIM_Cubic, leaveTangentWeight=0.397, tangentMode=RCTM_Break, time=0, value=0.7}
         {arriveTangentWeight=0.4342, interpMode=RCIM_Cubic, tangentMode=RCTM_Break, time=1, value=1}
        ]}}}
     Scale Curve=&RandomRangeFloat(Minimum=0, Maximum=1)
    )
    Index 2 Param 3=&FloatFromCurve(
     FloatCurve=DataInterface() {propertyValues {Curve {keys [
         {interpMode=RCIM_Cubic, leaveTangent=-6.5865, tangentMode=RCTM_Break, time=0, value=0.5}
         {interpMode=RCIM_Cubic, leaveTangentWeight=0.3354, tangentMode=RCTM_Break, time=0.2, value=0}
         {arriveTangentWeight=0.3308, interpMode=RCIM_Cubic, tangentMode=RCTM_Break, time=1, value=1}
        ]}}}
     Scale Curve=6
    )
    Index 2 Param 4=&FloatFromCurve(
     FloatCurve=DataInterface() {propertyValues {Curve {keys [
         {interpMode=RCIM_Cubic, leaveTangentWeight=0.4044, tangentMode=RCTM_Break, time=0, value=0.8}
         {arriveTangent=0.4016, arriveTangentWeight=0.382, interpMode=RCIM_Cubic, tangentMode=RCTM_Break, time=1, value=1}
        ]}}}
     Scale Curve=1.5
    )
    Module.Param0WriteEnabled=true
    Module.Param0WriteEnabled2=true
    Module.Param1WriteEnabled=false
    Module.Param1WriteEnabled2=true
    Module.Param2WriteEnabled=true
    Module.Param2WriteEnabled2=true
    Module.Param3WriteEnabled=true
    Module.Param3WriteEnabled2=true
    Write Parameter Index 2=true
   )
   &SpriteFacingAndAlignment(Sprite Facing=<1 1 1>, WriteAlignment=false, WriteFacing=true)
   &SpriteRotationRate(Rotation Rate=&Multiply_Float(
     A=$Particles.InitSpriteRotRate
     B=&FloatFromCurve(
      FloatCurve=DataInterface() {propertyValues {Curve {keys [
          {interpMode=RCIM_Cubic, leaveTangent=-2.6704, tangentMode=RCTM_Break, time=0, value=1}
          {arriveTangent=-0.4702, interpMode=RCIM_Cubic, leaveTangent=-0.4702, tangentMode=RCTM_User, time=0.2, value=0.2}
          {interpMode=RCIM_Cubic, tangentMode=RCTM_Break, time=1, value=0.1}
         ]}}}
      Scale Curve=0
     )
    ))
  ]
  rendererClasses [&NiagaraSpriteRendererProperties]
  renderers [
   {rendererClass=&NiagaraSpriteRendererProperties}
   {rendererClass=&NiagaraSpriteRendererProperties, rendererIndex=1}
  ]
 }
 ShiftBlip {
  emitterUpdateScript [
   &EmitterState(Life Cycle Mode=::NewEnumerator1, Loop Duration=0.1, UseLoopDelay=false)
   &SpawnBurst_Instantaneous(
    Spawn Count=&Multiply_Int(
     A=&RandomRangeInt(Minimum=1, Maximum=2)
     B=&MakeCustomIntFromBool(Bool=$User.IsWaterDeep, True INT=0, False INT=1)
    )
    Spawn Probability=&MakeCustomFloatFromBool(
     Bool=&Boolean_AndOperation(Bool1=$User.VehicleHasDriver, Bool2=&Boolean_AndOperation(Bool1=$User.VehicleHasFuel, Bool2=$User.IsShifting))
     True Float=1
     False Float=0
    )
    Use Spawn Probability=true
   )
   &SpawnBurst_Instantaneous(
    Spawn Count=&Multiply_Int(
     A=&RandomRangeInt(Minimum=1, Maximum=2)
     B=&MakeCustomIntFromBool(Bool=$User.IsWaterDeep, True INT=0, False INT=1)
    )
    Spawn Group=1
    Spawn Probability=&MakeCustomFloatFromBool(
     Bool=&Boolean_AndOperation(Bool1=$User.VehicleHasDriver, Bool2=&Boolean_AndOperation(Bool1=$User.VehicleHasFuel, Bool2=$User.IsShifting))
     True Float=1
     False Float=0
    )
    Use Spawn Probability=true
   ) {module=SpawnBurst_Instantaneous001}
  ]
  particleSpawnScript [
   /SetVariables(Particles.RandomFloat=&UniformAOrBFloat(A=-1)) {isSetParameters=true}
   /Niagara/Modules/Spawn/Initialization/InitializeParticle(
    Color=<rgba 0.6062 0.754 1 0.025>
    Lifetime=&MakeCustomFloatFromBool(
     Bool=$User.Downshifting
     True Float=&RandomRangeFloat(Minimum=0.1, Maximum=0.14)
     False Float=&RandomRangeFloat(Minimum=0.225, Maximum=0.375)
    )
    Module.Write Color=true
    Module.Write Lifetime=true
    Module.Write Mass=false
    Module.Write Material Random=true
    Module.Write Position=true
    Module.Write Scale=false
    Module.Write SpriteRotation=true
    Module.Write SpriteSize=true
    Position=$Engine.Owner.Position
    Sprite Rotation=&UniformRangedFloat(Minimum=0, Maximum=360)
    Sprite Size=&Vector2DFromFloat(Value=&MakeCustomFloatFromBool(
      Bool=$User.Downshifting
      True Float=&RandomRangeFloat(Minimum=40, Maximum=60)
      False Float=&RandomRangeFloat(Minimum=55, Maximum=80)
     ))
   )
   &DynamicMaterialParameters(
    Index 0 Param 1=&FloatFromCurve(FloatCurve=DataInterface() {propertyValues {Curve {keys [
         {arriveTangent=8.397, interpMode=RCIM_Cubic, leaveTangent=4.6996, tangentMode=RCTM_User, time=-0.0012, value=-0.0261}
         {arriveTangent=1.3472, interpMode=RCIM_Cubic, leaveTangent=1.3472, tangentMode=RCTM_User, time=0.1801, value=0.7926}
         <at 0.5 1 RCIM_Cubic>
        ]}}})
    Index 0 Param 2=&UniformRangedFloat(Minimum=0.333, Maximum=1)
    Index 0 Param 3=&FloatFromCurve(FloatCurve=DataInterface() {propertyValues {Curve {keys [{arriveTangent=1.4224, interpMode=RCIM_Cubic, time=0.5, value=1}, <at 1 2 RCIM_Cubic>]}}})
    Index 0 Param 4=&UniformRangedFloat(Minimum=1.333, Maximum=2)
    Module.Param0WriteEnabled=true
    Module.Param1WriteEnabled=true
    Module.Param2WriteEnabled=true
    Module.Param3WriteEnabled=true
   ) {module=DynamicMaterialParameters001}
  ]
  particleUpdateScript [
   &ParticleState()
   &SkeletalMeshLocation(
    Bone / Socket Index=&Add_Integer(A=&SpawnGroup(), B=12)
    Inherit Velocity=true
    "Inherit Velocity (Scale)"=-0.05
    Interpolate Mesh Sampling=false
    MaskSpawnGroup=false
    Mesh Sampling Type=::NewEnumerator5
    Sampled Position Offset=&MakeVector(X=0, Y=0, Z=0)
    Sampled Position Offset Coordinate Space:ENiagaraMeshTransforms=::NewEnumerator3
    Skeletal Mesh=$User.Vehicle_Selection_02
    Socket Sampling Mode=::NewEnumerator5
    VelocitySpeedLimit=false
    Write Sampled Mesh Attrributes to Particle Attributes=true
   ) {module=SkeletalMeshLocation001}
   /SetVariables(Particles.Position=&Add_Vector(
     A=$Particles.SkeletalMeshLocation001.SampledPosition
     B=&Multiply_VectorByFloat(
      Vector=&RotateVectorByQuaternion(
       Vector To Rotate=&MakeVector(X=1, Y=0, Z=0)
       Quaternion=$Particles.SkeletalMeshLocation001.SampledRotation
      )
      Float=&Multiply_Float(A=$Particles.NormalizedAge, B=&Lerp_Float(A=90, B=60, Alpha=$User.Vehicle_Speed))
     )
    )) {isSetParameters=true}
   /SetVariables(Particles.Color2=&ColorFromCurve(ColorCurve=DataInterface() {propertyValues {
       AlphaCurve {keys [<at 0.4847 1 RCIM_Cubic>,<at 1 0 RCIM_Cubic>]}
       BlueCurve {keys [<at 0.0194 50 RCIM_Cubic>,<at 0.1503 150>,<at 0.1866 50>,<at 0.3975 0 RCIM_Cubic>]}
       GreenCurve {keys [<at 0.0194 30.8898 RCIM_Cubic>,<at 0.1503 92.6695>,<at 0.1866 0>,<at 0.3975 5.5901 RCIM_Cubic>]}
       RedCurve {keys [<at 0.0194 0.1 RCIM_Cubic>,<at 0.1503 0>,<at 0.1866 10.781>,<at 0.3975 25 RCIM_Cubic>]}
      }})) {isSetParameters=true}
   /SetVariables(Particles.SubUVIndex2:NiagaraFloat=&Add_Float(
     A=&RandomRangeFloat(Minimum=0, Maximum=32)
     B=&Lerp_Float(A=0, B=60, Alpha=$Particles.NormalizedAge)
    )) {isSetParameters=true}
   &ScaleSpriteSize(Scale Factor=&Vector2DFromFloat(Value=&FloatFromCurve(FloatCurve=DataInterface() {propertyValues {Curve {keys [{interpMode=RCIM_Cubic, leaveTangent=2.377, tangentMode=RCTM_Break, time=0, value=0.179}, <at 0.8329 1>]}}})))
   &SpriteRotationRate(Rotation Rate=&Multiply_Float(A=$Particles.Initial.MaterialRandom, B=1))
   &DynamicMaterialParameters(
    Index 0 Param 1=&FloatFromCurve(FloatCurve=DataInterface() {propertyValues {Curve {keys [{arriveTangent=2.3188, interpMode=RCIM_Cubic, time=0, value=0}, <at 0.5 1 RCIM_Cubic>]}}})
    Index 0 Param 3=&FloatFromCurve(FloatCurve=DataInterface() {propertyValues {Curve {keys [{arriveTangent=1.4224, interpMode=RCIM_Cubic, time=0.5, value=1}, <at 1 2 RCIM_Cubic>]}}})
    Module.Param0WriteEnabled=true
    Module.Param1WriteEnabled=false
    Module.Param2WriteEnabled=true
    Module.Param3WriteEnabled=false
   )
  ]
  rendererClasses [&NiagaraSpriteRendererProperties]
  renderers [{rendererClass=&NiagaraSpriteRendererProperties}]
 }
 ShiftBlipLight {
  emitterUpdateScript [
   &EmitterState(Life Cycle Mode=::NewEnumerator1, Loop Duration=0.1, UseLoopDelay=false)
   &SpawnBurst_Instantaneous(
    Spawn Count=&Multiply_Int(A=1, B=&MakeCustomIntFromBool(Bool=$User.IsWaterDeep, True INT=0, False INT=1))
    Spawn Probability=&MakeCustomFloatFromBool(
     Bool=&Boolean_AndOperation(Bool1=$User.VehicleHasDriver, Bool2=&Boolean_AndOperation(Bool1=$User.VehicleHasFuel, Bool2=$User.IsShifting))
     True Float=1
     False Float=0
    )
    Use Spawn Probability=true
   )
   &SpawnBurst_Instantaneous(
    Spawn Count=&Multiply_Int(A=1, B=&MakeCustomIntFromBool(Bool=$User.IsWaterDeep, True INT=0, False INT=1))
    Spawn Group=1
    Spawn Probability=&MakeCustomFloatFromBool(
     Bool=&Boolean_AndOperation(Bool1=$User.VehicleHasDriver, Bool2=&Boolean_AndOperation(Bool1=$User.VehicleHasFuel, Bool2=$User.IsShifting))
     True Float=1
     False Float=0
    )
    Use Spawn Probability=true
   ) {module=SpawnBurst_Instantaneous001}
  ]
  particleSpawnScript [&InitializeParticle(
    Lifetime=&RandomRangeFloat(Minimum=0.085, Maximum=0.125)
    Module.Write Color=true
    Module.Write Lifetime=true
    Module.Write Mass=false
    Module.Write Material Random=true
    Module.Write Position=true
    Module.Write Scale=false
    Module.Write SpriteRotation=true
    Module.Write SpriteSize=true
    Position=$Engine.Owner.Position
   )]
  particleUpdateScript [
   &ParticleState()
   &SkeletalMeshLocation(
    Bone / Socket Index=&Add_Integer(A=&SpawnGroup(), B=12)
    Inherit Velocity=false
    "Inherit Velocity (Scale)"=-0.05
    Interpolate Mesh Sampling=false
    MaskSpawnGroup=false
    Mesh Sampling Type=::NewEnumerator5
    Sampled Position Offset=&MakeVector(X=15, Y=0, Z=0)
    Sampled Position Offset Coordinate Space:ENiagaraMeshTransforms=::NewEnumerator3
    Skeletal Mesh=$User.Vehicle_Selection_02
    Socket Sampling Mode=::NewEnumerator5
    Write Sampled Mesh Attrributes to Particle Attributes=true
   ) {module=SkeletalMeshLocation001}
   &FN_Light_Attributes(
    Color=&ColorFromCurve(ColorCurve=DataInterface() {propertyValues {
       AlphaCurve {keys [<at 0.0118 0>,<at 0.3542 1 RCIM_Cubic>,<at 1 0 RCIM_Cubic>]}
       BlueCurve {keys [<at 0.054 0>,<at 0.4426 0 RCIM_Cubic>]}
       GreenCurve {keys [<at 0.054 4.7963>,<at 0.4426 2.6832 RCIM_Cubic>]}
       RedCurve {keys [<at 0.054 24>,<at 0.4426 12 RCIM_Cubic>]}
      }})
    Create Color Attribute=true
    Create Exponent Attribute=true
    Create Radius Attribute=true
    Exponent=8
    Radius=120
    Scale Color=&MakeCustomFloatFromBool(Bool=$User.Downshifting, True Float=0.15, False Float=1)
   )
  ]
  rendererClasses [&NiagaraLightRendererProperties]
  renderers [{rendererClass=&NiagaraLightRendererProperties, rendererIndex=0}]
 }
 Sparks {
  emitterUpdateScript [
   &EmitterState()
   &SpawnRate(
    Spawn Probability=&MakeCustomFloatFromBool(
     Bool=$User.TireIntactFR
     True Float=0
     False Float=&MakeCustomFloatFromBool(Bool=&SetBoolByFloatComparison(A=$User.Vehicle_Speed, B=0.1), True Float=1, False Float=0)
    )
    SpawnGroup=1
    SpawnRate=125
    Use Spawn Probability=true
   )
   &SpawnRate(
    Spawn Probability=&MakeCustomFloatFromBool(
     Bool=$User.TireIntactFL
     True Float=0
     False Float=&MakeCustomFloatFromBool(Bool=&SetBoolByFloatComparison(A=$User.Vehicle_Speed, B=0.1), True Float=1, False Float=0)
    )
    SpawnRate=125
    Use Spawn Probability=true
   ) {module=SpawnRate001}
  ]
  particleSpawnScript [
   &FN_SetSpawnGroup(Number of Spawn Groups=2)
   /Niagara/Modules/Spawn/Initialization/V2/InitializeParticle(
    Color Maximum=<rgba 1 0.1188 0 1>
    Color Minimum=<rgba 5 0.7736 0 1>
    Color Mode=::NewEnumerator2
    Lifetime Max=1.5
    Lifetime Min=0.4
    Lifetime Mode=::NewEnumerator1
    Sprite Size Max=<6 6>
    Sprite Size Min=<2 2>
    Sprite Size Mode=::NewEnumerator2
    UsePositionOffset=false
    Write Material Random=true
   )
   &SkeletalMeshLocation(
    Bone / Socket Index=$Emitter.SpawnGroup
    Interpolate Mesh Sampling=false
    MaskSpawnGroup=false
    Mesh Sampling Type=::NewEnumerator5
    Orientation Sampling=::NewEnumerator1
    Sampled Position Offset=&SelectVectorFromArray(
     Direct Array Index=$Particles.SpawnGroup
     Vector Selection Array=DataInterface() {propertyValues {FloatData [<-275 -25 -10>, <-275 25 -10>]}}
    )
    Skeletal Mesh=$User.Vehicle_Selection_02
    Socket Sampling Mode=::NewEnumerator5
   )
   &ShapeLocation(Box Size=<550 30 10>, MaskSpawnGroup=false, Override Seed=false, Shape Primitive=::NewEnumerator2, Surface Only=false, UseOwnerScale=true) {module=ShapeLocation001}
   &AddVelocity(Velocity=&Multiply_VectorByFloat(
     Vector=&Multiply_Vector(
      A=&RandomRangeVector(Minimum=<0 100 50>, Maximum=<0 1300 900>)
      B=&MakeVector(
       X=1
       Y=&SelectFloatFromArray(Direct Array Index=$Emitter.SpawnGroup, Float Selection Array=DataInterface() {propertyValues {FloatData [1,-1]}})
       Z=1
      )
     )
     Float=1
    ))
   &AddVelocity(Module.Use Falloff Distance=false, Override Seed=false, Velocity Mode=::NewEnumerator1, Velocity Speed=200) {module=AddVelocity001}
  ]
  particleUpdateScript [
   &ParticleState()
   &AccelerationForce(Acceleration=&RandomRangeVector(Minimum=<0 0 -800>, Maximum=<0 0 -1100>))
   &CurlNoiseForce(
    Noise Strength=&FloatFromCurve(
     FloatCurve=DataInterface() {propertyValues {Curve {keys [{interpMode=RCIM_Cubic, tangentMode=RCTM_Break, time=0, value=1}, {interpMode=RCIM_Cubic, tangentMode=RCTM_Break, time=1, value=0}]}}}
     Scale Curve=2000
    )
    Pan Noise=false
   )
   &Drag(
    Drag=&FloatFromCurve(
     FloatCurve=DataInterface() {propertyValues {Curve {keys [
         {interpMode=RCIM_Cubic, leaveTangentWeight=0.5018, tangentMode=RCTM_Break, time=0, value=1}
         {interpMode=RCIM_Cubic, tangentMode=RCTM_Break, time=0.2, value=0}
        ]}}}
     Scale Curve=25
    )
    Use Linear Drag=true
    Use Rotational Drag=false
   )
   &SolveForcesAndVelocity(Clamp Velocity=false, Limit Acceleration=false)
   &ScaleSpriteSize(
    Uniform Curve Scale=1.2
    Uniform Curve Sprite Scale=DataInterface() {propertyValues {Curve {keys [{interpMode=RCIM_Cubic, tangentMode=RCTM_Break, time=0, value=1}, {interpMode=RCIM_Cubic, tangentMode=RCTM_Break, time=1, value=0.2}]}}}
   )
   &ScaleSpriteSize(
    Scale Sprite Size Mode=::NewEnumerator0
    Uniform Scale Factor=/Game/Effects/Niagara/DynamicInputs/Scalability/FN_ScalabilityFloatModifier(
     Value at DistanceFraction 0:NiagaraFloat=1
     Value at DistanceFraction 1:NiagaraFloat=3
     DistanceRamp:NiagaraDataInterfaceCurve=DataInterface() {propertyValues {Curve {keys [<at 0.1 0>,<at 0.5 1>]}}}
    )
   ) {module=ScaleSpriteSize001}
   &ScaleSpriteSizeBySpeed(Max Scale Factor=<1 5>)
   &ScaleColor(
    Scale Alpha=&FloatFromCurve(FloatCurve=DataInterface() {propertyValues {Curve {keys [
         {interpMode=RCIM_Cubic, tangentMode=RCTM_User, time=0.0048, value=0.438}
         {arriveTangent=0.3021, interpMode=RCIM_Cubic, leaveTangent=0.3021, tangentMode=RCTM_User, time=0.0983, value=1.5996}
         {arriveTangent=-1.0514, interpMode=RCIM_Cubic, leaveTangent=-1.0514, tangentMode=RCTM_User, time=0.3368, value=0.5383}
         {arriveTangent=-0.8117, interpMode=RCIM_Cubic, leaveTangent=-0.8117, tangentMode=RCTM_User, time=0.5811, value=1.092}
         <at 1 0>
        ]}}})
    Scale RGB=&VectorFromCurve(VectorCurve=DataInterface() {propertyValues {
       XCurve {keys [<at 0 16 RCIM_Cubic>,<at 0.8 1 RCIM_Cubic>]}
       YCurve {keys [<at 0 16 RCIM_Cubic>,<at 0.8 0.1249 RCIM_Cubic>]}
       ZCurve {keys [<at 0 16 RCIM_Cubic>,<at 0.8 0.2547 RCIM_Cubic>]}
      }})
    ScaleA=true
    ScaleRGB=true
   )
  ]
  rendererClasses [&NiagaraSpriteRendererProperties]
  renderers [{rendererClass=&NiagaraSpriteRendererProperties}]
 }
 SparksIdle {
  bEnabled=false
  emitterUpdateScript [
   &EmitterState()
   &SpawnRate(
    Spawn Probability=&MakeCustomFloatFromBool(
     Bool=$User.TireIntactFR
     True Float=0
     False Float=&MakeCustomFloatFromBool(
      Bool=&SetBoolByFloatComparison(A=$User.Vehicle_Speed, B=0.1, Comparison Type=::NewEnumerator6)
      True Float=1
      False Float=0
     )
    )
    SpawnGroup=1
    SpawnRate=60
    Use Spawn Probability=true
   )
   &SpawnRate(
    Spawn Probability=&MakeCustomFloatFromBool(
     Bool=$User.TireIntactFL
     True Float=0
     False Float=&MakeCustomFloatFromBool(
      Bool=&SetBoolByFloatComparison(A=$User.Vehicle_Speed, B=0.1, Comparison Type=::NewEnumerator6)
      True Float=1
      False Float=0
     )
    )
    SpawnRate=60
    Use Spawn Probability=true
   ) {module=SpawnRate001}
  ]
  particleSpawnScript [
   &FN_SetSpawnGroup(Number of Spawn Groups=2)
   &InitializeParticle(
    Color Maximum=<rgba 1 0.1188 0 1>
    Color Minimum=<rgba 5 0.7736 0 1>
    Color Mode=::NewEnumerator2
    Lifetime Max=1.5
    Lifetime Min=0.4
    Lifetime Mode=::NewEnumerator1
    Sprite Size Max=<6 6>
    Sprite Size Min=<2 2>
    Sprite Size Mode=::NewEnumerator2
    UsePositionOffset=false
    Write Material Random=true
   )
   &SkeletalMeshLocation(
    Bone / Socket Index=$Emitter.SpawnGroup
    Interpolate Mesh Sampling=false
    MaskSpawnGroup=false
    Mesh Sampling Type=::NewEnumerator5
    Orientation Sampling=::NewEnumerator1
    Sampled Position Offset=&SelectVectorFromArray(
     Direct Array Index=$Particles.SpawnGroup
     Vector Selection Array=DataInterface() {propertyValues {FloatData [<-275 -25 -10>, <-275 25 -10>]}}
    )
    Skeletal Mesh=$User.Vehicle_Selection_02
    Socket Sampling Mode=::NewEnumerator5
   )
   &ShapeLocation(Box Size=<550 30 10>, MaskSpawnGroup=false, Override Seed=false, Shape Primitive=::NewEnumerator2, Surface Only=false, UseOwnerScale=true) {module=ShapeLocation001}
   &AddVelocity(Velocity=&Multiply_VectorByFloat(
     Vector=&Multiply_Vector(
      A=&RandomRangeVector(Minimum=<0 100 50>, Maximum=<0 1300 900>)
      B=&MakeVector(
       X=1
       Y=&SelectFloatFromArray(Direct Array Index=$Emitter.SpawnGroup, Float Selection Array=DataInterface() {propertyValues {FloatData [1,-1]}})
       Z=1
      )
     )
     Float=0.4
    ))
   &AddVelocity(Module.Use Falloff Distance=false, Override Seed=false, Velocity Mode=::NewEnumerator1, Velocity Speed=100) {module=AddVelocity001}
  ]
  particleUpdateScript [
   &ParticleState()
   &AccelerationForce(Acceleration=&RandomRangeVector(Minimum=<0 0 -800>, Maximum=<0 0 -1100>))
   &CurlNoiseForce(
    Noise Strength=&FloatFromCurve(
     FloatCurve=DataInterface() {propertyValues {Curve {keys [{interpMode=RCIM_Cubic, tangentMode=RCTM_Break, time=0, value=1}, {interpMode=RCIM_Cubic, tangentMode=RCTM_Break, time=1, value=0}]}}}
     Scale Curve=2000
    )
    Pan Noise=false
   )
   &Drag(
    Drag=&FloatFromCurve(
     FloatCurve=DataInterface() {propertyValues {Curve {keys [
         {interpMode=RCIM_Cubic, leaveTangentWeight=0.5018, tangentMode=RCTM_Break, time=0, value=1}
         {interpMode=RCIM_Cubic, tangentMode=RCTM_Break, time=0.2, value=0}
        ]}}}
     Scale Curve=25
    )
    Use Linear Drag=true
    Use Rotational Drag=false
   )
   &SolveForcesAndVelocity(Clamp Velocity=false, Limit Acceleration=false)
   &ScaleSpriteSize(
    Uniform Curve Scale=1.2
    Uniform Curve Sprite Scale=DataInterface() {propertyValues {Curve {keys [{interpMode=RCIM_Cubic, tangentMode=RCTM_Break, time=0, value=1}, {interpMode=RCIM_Cubic, tangentMode=RCTM_Break, time=1, value=0.2}]}}}
   )
   &ScaleSpriteSize(
    Scale Sprite Size Mode=::NewEnumerator0
    Uniform Scale Factor=&FN_ScalabilityFloatModifier(
     Value at DistanceFraction 0:NiagaraFloat=1
     Value at DistanceFraction 1:NiagaraFloat=3
     DistanceRamp:NiagaraDataInterfaceCurve=DataInterface() {propertyValues {Curve {keys [<at 0.1 0>,<at 0.5 1>]}}}
    )
   ) {module=ScaleSpriteSize001}
   &ScaleSpriteSizeBySpeed(Max Scale Factor=<1 5>)
   &ScaleColor(
    Scale Alpha=&FloatFromCurve(FloatCurve=DataInterface() {propertyValues {Curve {keys [
         {interpMode=RCIM_Cubic, tangentMode=RCTM_User, time=0.0048, value=0.438}
         {arriveTangent=0.3021, interpMode=RCIM_Cubic, leaveTangent=0.3021, tangentMode=RCTM_User, time=0.0983, value=1.5996}
         {arriveTangent=-1.0514, interpMode=RCIM_Cubic, leaveTangent=-1.0514, tangentMode=RCTM_User, time=0.3368, value=0.5383}
         {arriveTangent=-0.8117, interpMode=RCIM_Cubic, leaveTangent=-0.8117, tangentMode=RCTM_User, time=0.5811, value=1.092}
         <at 1 0>
        ]}}})
    Scale RGB=&VectorFromCurve(VectorCurve=DataInterface() {propertyValues {
       XCurve {keys [<at 0 16 RCIM_Cubic>,<at 0.8 1 RCIM_Cubic>]}
       YCurve {keys [<at 0 16 RCIM_Cubic>,<at 0.8 0.1249 RCIM_Cubic>]}
       ZCurve {keys [<at 0 16 RCIM_Cubic>,<at 0.8 0.2547 RCIM_Cubic>]}
      }})
    ScaleA=true
    ScaleRGB=true
   )
  ]
  rendererClasses [&NiagaraSpriteRendererProperties]
  renderers [{rendererClass=&NiagaraSpriteRendererProperties}]
 }
 SparksRear {
  emitterUpdateScript [
   &EmitterState()
   &SpawnRate(
    Spawn Probability=&MakeCustomFloatFromBool(
     Bool=$User.TireIntactFR
     True Float=0
     False Float=&MakeCustomFloatFromBool(Bool=&SetBoolByFloatComparison(A=$User.Vehicle_Speed, B=0.1), True Float=1, False Float=0)
    )
    SpawnGroup=1
    SpawnRate=60
    Use Spawn Probability=true
   )
   &SpawnRate(
    Spawn Probability=&MakeCustomFloatFromBool(
     Bool=$User.TireIntactFL
     True Float=0
     False Float=&MakeCustomFloatFromBool(Bool=&SetBoolByFloatComparison(A=$User.Vehicle_Speed, B=0.1), True Float=1, False Float=0)
    )
    SpawnRate=60
    Use Spawn Probability=true
   ) {module=SpawnRate001}
  ]
  particleSpawnScript [
   &FN_SetSpawnGroup(Number of Spawn Groups=2)
   &InitializeParticle(
    Color Maximum=<rgba 1 0.1188 0 1>
    Color Minimum=<rgba 5 0.7736 0 1>
    Color Mode=::NewEnumerator2
    Lifetime Max=1.5
    Lifetime Min=0.4
    Lifetime Mode=::NewEnumerator1
    Sprite Size Max=<6 6>
    Sprite Size Min=<2 2>
    Sprite Size Mode=::NewEnumerator2
    UsePositionOffset=false
    Write Material Random=true
   )
   &SkeletalMeshLocation(
    Bone / Socket Index=$Emitter.SpawnGroup
    Interpolate Mesh Sampling=false
    MaskSpawnGroup=false
    Mesh Sampling Type=::NewEnumerator5
    Orientation Sampling=::NewEnumerator1
    Sampled Position Offset=&SelectVectorFromArray(Direct Array Index=$Particles.SpawnGroup, Vector Selection Array=DataInterface() {propertyValues {FloatData [<-25 -25 -10>, <-25 25 -10>]}})
    Skeletal Mesh=$User.Vehicle_Selection_Rear
    Socket Sampling Mode=::NewEnumerator5
   )
   &ShapeLocation(Box Size=<50 30 10>, MaskSpawnGroup=false, Override Seed=false, Shape Primitive=::NewEnumerator2, Surface Only=false, UseOwnerScale=true) {module=ShapeLocation001}
   &AddVelocity(Velocity=&Multiply_VectorByFloat(
     Vector=&Multiply_Vector(
      A=&RandomRangeVector(Minimum=<100 10 200>, Maximum=<600 20 1000>)
      B=&MakeVector(
       X=1
       Y=&SelectFloatFromArray(Direct Array Index=$Emitter.SpawnGroup, Float Selection Array=DataInterface() {propertyValues {FloatData [1,-1]}})
       Z=1
      )
     )
     Float=1
    ))
  ]
  particleUpdateScript [
   &ParticleState()
   &AccelerationForce(Acceleration=&RandomRangeVector(Minimum=<0 0 -800>, Maximum=<0 0 -1100>))
   &CurlNoiseForce(
    Noise Strength=&FloatFromCurve(
     FloatCurve=DataInterface() {propertyValues {Curve {keys [{interpMode=RCIM_Cubic, tangentMode=RCTM_Break, time=0, value=1}, {interpMode=RCIM_Cubic, tangentMode=RCTM_Break, time=1, value=0}]}}}
     Scale Curve=2000
    )
    Pan Noise=false
   )
   &Drag(
    Drag=&FloatFromCurve(
     FloatCurve=DataInterface() {propertyValues {Curve {keys [
         {interpMode=RCIM_Cubic, leaveTangentWeight=0.5018, tangentMode=RCTM_Break, time=0, value=1}
         {interpMode=RCIM_Cubic, tangentMode=RCTM_Break, time=0.2, value=0}
        ]}}}
     Scale Curve=25
    )
    Use Linear Drag=true
    Use Rotational Drag=false
   )
   &SolveForcesAndVelocity(Clamp Velocity=false, Limit Acceleration=false)
   &ScaleSpriteSize(
    Uniform Curve Scale=1.2
    Uniform Curve Sprite Scale=DataInterface() {propertyValues {Curve {keys [{interpMode=RCIM_Cubic, tangentMode=RCTM_Break, time=0, value=1}, {interpMode=RCIM_Cubic, tangentMode=RCTM_Break, time=1, value=0.2}]}}}
   )
   &ScaleSpriteSize(
    Scale Sprite Size Mode=::NewEnumerator0
    Uniform Scale Factor=&FN_ScalabilityFloatModifier(
     Value at DistanceFraction 0:NiagaraFloat=1
     Value at DistanceFraction 1:NiagaraFloat=3
     DistanceRamp:NiagaraDataInterfaceCurve=DataInterface() {propertyValues {Curve {keys [<at 0.1 0>,<at 0.5 1>]}}}
    )
   ) {module=ScaleSpriteSize001}
   &ScaleSpriteSizeBySpeed(Max Scale Factor=<1 5>)
   &ScaleColor(
    Scale Alpha=&FloatFromCurve(FloatCurve=DataInterface() {propertyValues {Curve {keys [
         {interpMode=RCIM_Cubic, tangentMode=RCTM_User, time=0.0048, value=0.438}
         {arriveTangent=0.3021, interpMode=RCIM_Cubic, leaveTangent=0.3021, tangentMode=RCTM_User, time=0.0983, value=1.5996}
         {arriveTangent=-1.0514, interpMode=RCIM_Cubic, leaveTangent=-1.0514, tangentMode=RCTM_User, time=0.3368, value=0.5383}
         {arriveTangent=-0.8117, interpMode=RCIM_Cubic, leaveTangent=-0.8117, tangentMode=RCTM_User, time=0.5811, value=1.092}
         <at 1 0>
        ]}}})
    Scale RGB=&VectorFromCurve(VectorCurve=DataInterface() {propertyValues {
       XCurve {keys [<at 0 16 RCIM_Cubic>,<at 0.8 1 RCIM_Cubic>]}
       YCurve {keys [<at 0 16 RCIM_Cubic>,<at 0.8 0.1249 RCIM_Cubic>]}
       ZCurve {keys [<at 0 16 RCIM_Cubic>,<at 0.8 0.2547 RCIM_Cubic>]}
      }})
    ScaleA=true
    ScaleRGB=true
   )
  ]
  rendererClasses [&NiagaraSpriteRendererProperties]
  renderers [{rendererClass=&NiagaraSpriteRendererProperties}]
 }
 TreadStreaks {
  emitterUpdateScript [
   &EmitterState(Life Cycle Mode=::NewEnumerator1, Loop Duration=1, UseLoopDelay=false)
   &SpawnRate(
    Spawn Probability=&MakeCustomFloatFromBool(
     Bool=$User.TireIntactFR
     True Float=0
     False Float=&MakeCustomFloatFromBool(Bool=&SetBoolByFloatComparison(A=$User.Vehicle_Speed, B=0.1), True Float=1, False Float=0)
    )
    SpawnGroup=1
    SpawnRate=25
    Use Spawn Probability=true
   )
   &SpawnRate(
    Spawn Probability=&MakeCustomFloatFromBool(
     Bool=$User.TireIntactFL
     True Float=0
     False Float=&MakeCustomFloatFromBool(Bool=&SetBoolByFloatComparison(A=$User.Vehicle_Speed, B=0.1), True Float=1, False Float=0)
    )
    SpawnRate=25
    Use Spawn Probability=true
   ) {module=SpawnRate001}
   /SetVariables(Emitter.DepthFade=150) {isSetParameters=true, module=SetVariables_#3}
  ]
  particleSpawnScript [
   &FN_SetSpawnGroup(Number of Spawn Groups=2)
   /SetVariables(Particles.SizeMax=450, Particles.SizeMin=150) {isSetParameters=true, module=SetVariables_#1}
   /SetVariables(Particles.SizeCurve:NiagaraFloat=&Lerp_Float(
     A=$Particles.SizeMin
     B=$Particles.SizeMax
     Alpha=&FloatFromCurve(
      FloatCurve=DataInterface() {propertyValues {Curve {keys [
          {interpMode=RCIM_Cubic, tangentMode=RCTM_Break, time=0, value=0}
          {arriveTangent=2.305, interpMode=RCIM_Cubic, tangentMode=RCTM_Break, time=1, value=1}
         ]}}}
      CurveIndex=&RandomRangeFloat(Minimum=0, Maximum=1)
     )
    )) {isSetParameters=true, module=SetVariables_#2}
   &InitializeParticle(
    Color Maximum=<rgba 1 0.2016 0 1>
    Color Minimum=<rgba 0.9219 0.009 0 1>
    Color Mode=::NewEnumerator2
    Lifetime Max=0.5
    Lifetime Min=0.25
    Lifetime Mode=::NewEnumerator1
    Sprite Rotation Mode=::NewEnumerator2
    Sprite Size=&Multiply_Vector2D_ByFloat(Float=&RandomRangeFloat(Minimum=0.3, Maximum=0.4), Vector2D=<75 100>)
    Sprite Size Mode=::NewEnumerator4
    UsePositionOffset=false
    Write Material Random=true
   )
   &SkeletalMeshLocation(
    Bone / Socket Index=$Emitter.SpawnGroup
    Interpolate Mesh Sampling=false
    MaskSpawnGroup=false
    Mesh Sampling Type=::NewEnumerator5
    Orientation Sampling=::NewEnumerator1
    Sampled Position Offset=&SelectVectorFromArray(
     Direct Array Index=$Particles.SpawnGroup
     Vector Selection Array=DataInterface() {propertyValues {FloatData [<-275 -25 -10>, <-275 25 -10>]}}
    )
    Skeletal Mesh=$User.Vehicle_Selection_02
    Socket Sampling Mode=::NewEnumerator5
   )
   &ShapeLocation(Box Size=<550 30 10>, MaskSpawnGroup=false, Override Seed=false, Shape Primitive=::NewEnumerator2, Surface Only=false, UseOwnerScale=true) {module=ShapeLocation001}
   &AddVelocity(Velocity=&Multiply_VectorByFloat(
     Vector=&SelectVectorFromArray(Direct Array Index=$Emitter.SpawnGroup, Vector Selection Array=DataInterface() {propertyValues {FloatData [<75 100 25>, <75 -100 25>]}})
     Float=&RandomRangeFloat(Minimum=0.25, Maximum=1)
    )) {module=AddVelocity001}
   &SubUVAnimation(End Frame=3, SubUV Animation Mode:ENiagara_SubUVLookupMode=::NewEnumerator2)
  ]
  particleUpdateScript [
   &ParticleState()
   &SolveForcesAndVelocity(Clamp Velocity=false, Limit Acceleration=false)
   &ScaleSpriteSize(Scale Factor=&Vector2DFromCurve(
     Vector2Curve=DataInterface() {propertyValues {XCurve {keys [<at 0.0036 1.6897>,<at 0.9967 0.0582>]}, YCurve {keys [<at 0 1>,<at 1.002 3.7227>]}}}
     ScaleCurve=<1 1>
    ))
   &ScaleColor(
    Scale Alpha=&FloatFromCurve(
     FloatCurve=DataInterface() {propertyValues {Curve {keys [
         {interpMode=RCIM_Cubic, leaveTangentWeight=0.0962, tangentMode=RCTM_Break, time=0, value=0}
         {interpMode=RCIM_Cubic, tangentMode=RCTM_Break, time=0.2, value=1}
         {interpMode=RCIM_Cubic, tangentMode=RCTM_Break, time=1, value=0}
        ]}}}
     Scale Curve=&RandomRangeFloat(Minimum=0.25, Maximum=0.6)
    )
    Scale RGB=&VectorFromFloat(Value=&Lerp_Float(
      A=1
      B=5
      Alpha=&FloatFromCurve(
       FloatCurve=DataInterface() {propertyValues {Curve {keys [
           {interpMode=RCIM_Cubic, tangentMode=RCTM_Break, time=0.4, value=0}
           {arriveTangent=4.1312, interpMode=RCIM_Cubic, tangentMode=RCTM_Break, time=1, value=1}
          ]}}}
       CurveIndex=&RandomRangeFloat(Minimum=0, Maximum=1)
      )
     ))
    ScaleA=true
    ScaleRGB=true
   )
   &DynamicMaterialParameters(
    Index 0 Param 2=1
    Index 0 Param 3=1
    Index 0 Param 4=&FloatFromCurve(
     FloatCurve=DataInterface() {propertyValues {Curve {keys [{interpMode=RCIM_Cubic, tangentMode=RCTM_Break, time=0.3, value=1}, {interpMode=RCIM_Cubic, tangentMode=RCTM_Break, time=1, value=0.25}]}}}
     Scale Curve=&RandomRangeFloat(Minimum=0.75, Maximum=1)
    )
    Module.Param0WriteEnabled=true
    Module.Param1WriteEnabled=true
    Module.Param2WriteEnabled=true
    Module.Param3WriteEnabled=true
   )
   &ScaleColor(
    Scale Alpha=&FNTimeOfDayScalar(Night Scale=0.3, Morning Scale=0.1, Evening Scale=0.2)
    ScaleA=true
    ScaleRGB=false
   ) {module=ScaleColor001}
  ]
  rendererClasses [&NiagaraSpriteRendererProperties]
  renderers [{rendererClass=&NiagaraSpriteRendererProperties}]
 }
}
summary {
 systemName=NS_Tank_Tread_Interaction
 emitters [
  {emitterName=ShiftBlip, rendererClasses [&NiagaraSpriteRendererProperties]}
  {emitterName=ShiftBlipLight, rendererClasses [&NiagaraLightRendererProperties]}
  {emitterName=Dust_Tread, rendererClasses [&NiagaraSpriteRendererProperties]}
  {emitterName=DirtClumps, rendererClasses [&NiagaraSpriteRendererProperties]}
  {emitterName=ExhaustSmoke, rendererClasses [&NiagaraSpriteRendererProperties]}
  {emitterName=Airflow, rendererClasses [&NiagaraMeshRendererProperties]}
  {emitterName=ExhaustFlame, rendererClasses [&NiagaraMeshRendererProperties]}
  {emitterName=DirtStreaks, rendererClasses [&NiagaraSpriteRendererProperties]}
  {emitterName=Sparks, rendererClasses [&NiagaraSpriteRendererProperties]}
  {bEnabled=false, emitterName=SparksIdle, rendererClasses [&NiagaraSpriteRendererProperties]}
  {emitterName=TreadStreaks, rendererClasses [&NiagaraSpriteRendererProperties]}
  {emitterName=SparksRear, rendererClasses [&NiagaraSpriteRendererProperties]}
  {emitterName=ExhaustLight, rendererClasses [&NiagaraLightRendererProperties, &NiagaraSpriteRendererProperties]}
 ]
 userVariables [
  User.BoostFX {type=NiagaraBool}
  User.DamageAmount {type=NiagaraFloat, defaultValue=0, description=""}
  User.Downshifting {type=NiagaraBool}
  User.Dust_AlphaCurve {
   type=NiagaraDataInterfaceCurve
   description=""
   defaultValue {
    struct=NiagaraExt_VariableValue_DataInterface
    value {dataInterface=&self:NiagaraDataInterfaceCurve_0, dataInterfaceClass=&NiagaraDataInterfaceCurve}
   }
  }
  User.Dust_Particles_Per_Wheel {type=NiagaraInt32, defaultValue=18}
  User.Dust_SpawnCurve {
   type=NiagaraDataInterfaceCurve
   description=""
   defaultValue {
    struct=NiagaraExt_VariableValue_DataInterface
    value {dataInterface=&self:NiagaraDataInterfaceCurve_1, dataInterfaceClass=&NiagaraDataInterfaceCurve}
   }
  }
  User.Dust_Wheel_FL {type=NiagaraBool}
  User.Dust_Wheel_FR {type=NiagaraBool}
  User.Dust_Wheel_ML {type=NiagaraBool}
  User.Dust_Wheel_MR {type=NiagaraBool}
  User.Dust_Wheel_RL {type=NiagaraBool}
  User.Dust_Wheel_RR {type=NiagaraBool}
  User.DustAlphaMultiplier {type=NiagaraFloat, defaultValue=0.25, description=""}
  User.DustColorLerp {type=NiagaraFloat, defaultValue=0, description=""}
  User.DustSpawnPerUnit {type=NiagaraFloat, defaultValue=150, description=""}
  User.DustTireBlurColor {type=LinearColor, description="", defaultValue {struct=LinearColor, value=<rgba 1 1 1 1>}}
  User.ExhaustColor {type=LinearColor, description="", defaultValue {struct=LinearColor, value=<rgba 0.6062 0.754 1 0.025>}}
  User.ExhaustScale {type=NiagaraFloat, defaultValue=1, description=""}
  User.FuelAmount {type=NiagaraFloat, defaultValue=0, description=""}
  User.IsOnDirt {type=NiagaraBool}
  User.IsShifting {type=NiagaraBool}
  User.IsWaterDeep {type=NiagaraBool}
  User.Mod_OffroadTires {type=NiagaraBool}
  User.ParticleCountPerWheel {type=NiagaraInt32, defaultValue=150}
  User.ParticleScale {type=NiagaraFloat, defaultValue=0, description=""}
  User.PeelOut_Wheel_RL {type=NiagaraBool}
  User.PeelOut_Wheel_RR {type=NiagaraBool}
  User.Skid_Wheels {type=NiagaraBool}
  User.Slide_Direction {type=NiagaraFloat, defaultValue=0, description=""}
  User.Sparks_Wheel_FL {type=NiagaraBool}
  User.Sparks_Wheel_FR {type=NiagaraBool}
  User.Sparks_Wheel_ML {type=NiagaraBool}
  User.Sparks_Wheel_MR {type=NiagaraBool}
  User.Sparks_Wheel_RL {type=NiagaraBool}
  User.Sparks_Wheel_RR {type=NiagaraBool}
  User.SparkSPU_MaxSpeed {type=NiagaraFloat, defaultValue=3, description=""}
  User.SparkSPU_MinSpeed {type=NiagaraFloat, defaultValue=4, description=""}
  User.TankBone {type=NiagaraInt32}
  User.Tire_Scale_Back {type=Vector3f, description="", defaultValue {struct=Vector3f, value=<0.39 0.4325 0.39>}}
  User.Tire_Scale_Front {type=Vector3f, description="", defaultValue {struct=Vector3f, value=<0.3 0.4 0.3>}}
  User.TireIntactFL {type=NiagaraBool, defaultValue=true}
  User.TireIntactFR {type=NiagaraBool, defaultValue=true}
  User.TireIntactRL {type=NiagaraBool, defaultValue=true}
  User.TireIntactRR {type=NiagaraBool, defaultValue=true}
  User.Vehicle_Selection_02 {
   type=NiagaraDataInterfaceSkeletalMesh
   defaultValue {struct=NiagaraExt_VariableValue_DataInterface, value {dataInterface=&self:NiagaraDataInterfaceSkeletalMesh_1}}
  }
  User.Vehicle_Selection_Exhaust {
   type=NiagaraDataInterfaceSkeletalMesh
   defaultValue {struct=NiagaraExt_VariableValue_DataInterface, value {dataInterface=&self:NiagaraDataInterfaceSkeletalMesh_0}}
  }
  User.Vehicle_Selection_Rear {
   type=NiagaraDataInterfaceSkeletalMesh
   defaultValue {struct=NiagaraExt_VariableValue_DataInterface, value {dataInterface=&self:NiagaraDataInterfaceSkeletalMesh_2}}
  }
  User.Vehicle_Speed {type=NiagaraFloat, defaultValue=0, description=""}
  User.VehicleHasDriver {type=NiagaraBool}
  User.VehicleHasFuel {type=NiagaraBool}
  User.Wheel_Scale_Back {type=Vector3f, description="", defaultValue {struct=Vector3f, value=<0.3 0.3 0.3>}}
  User.Wheel_Scale_Front {type=Vector3f, description="", defaultValue {struct=Vector3f, value=<0.3 0.3 0.3>}}
  User.Wheel_Width {type=NiagaraFloat, defaultValue=16, description=""}
  User.Wheel_Width_Back {type=NiagaraFloat, defaultValue=18, description=""}
  User.WheelsTouchingArray {
   type=NiagaraDataInterfaceArrayFloat
   defaultValue {
    struct=NiagaraExt_VariableValue_DataInterface
    value {
     dataInterface=&self:NiagaraDataInterfaceArrayFloat_0
     dataInterfaceClass=&NiagaraDataInterfaceArrayFloat
    }
   }
  }
 ]
}
