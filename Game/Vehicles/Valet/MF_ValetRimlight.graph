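# Material function for the Valet vehicle that adds stylized rim lighting to the Emissive Color of incoming material attributes.
# Two MF_Rimlight_SingleLight calls with parameter directions use desaturated base color and squared Gmap, then get tinted by time-of-day colors and intensity.
# A contrast-adjusted Fresnel mask favors upward-facing normals.
# Named reroutes carry this mask, the rim colors and a height gradient from vertex-interpolated pre-skinned position.
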MaterialFunction /Game/Vehicles/Valet/MF_ValetRimlight
graph {nodes [
  n1 = GetMaterialAttributes(None=FunctionInput() {inputType=FunctionInput_MaterialAttributes}) {attributeGetTypes [#1, #2]}
  n2 = Multiply(A=NamedRerouteUsage(), B=NamedRerouteUsage())
  FunctionOutput(None=SetMaterialAttributes!(MaterialAttributes=@n1.MaterialAttributes, EmissiveColor=Add(A=@n1.EmissiveColor, B=@n2)) {attributeSetTypes [#1]})
  n3 = PreSkinnedPosition()
  n4 = Divide(
   A=FunctionInput(Preview=VertexInterpolator(VS=@n3).PS) {bUsePreviewValueAsDefault=true, inputName=Pre Skinned Local Position}
   B=Constant() {r=170}
  )
  NamedRerouteDeclaration(None=Saturate(None=ComponentMask(None=@n4) {b=true})) {name=Z-Gradient, nodeColor=<rgba 0 1 0.7891 1>}
  n5 = FunctionInput() {bUsePreviewValueAsDefault=true, inputName=Gmap, inputType=FunctionInput_Scalar}
  n6 = Saturate(None=Multiply(A=@n5, B=@n5))
  n7 = VertexNormalWS()
  n8 = Multiply(
   A=Fresnel(ExponentIn=ScalarParameter() {defaultValue=2, group=RimLight-Close, parameterName=RimFresnelExp}) {baseReflectFraction=0}
   B=Multiply(A=@n6, B=MaterialFunctionCall(Float3=@n7) {materialFunction=&BreakOutFloat3Components}.B)
  )
  n9 = ScalarParameter() {defaultValue=0.03, group=RimLight, parameterName=RimContrast}
  NamedRerouteDeclaration(None=Saturate(None=MaterialFunctionCall(In=@n8, Contrast=@n9) {materialFunction=&CheapContrast}.Result)) {name=RimLightFresnel, nodeColor=<rgba 1 0.375 0 1>}
  n10 = NamedRerouteUsage()
  n11 = VectorParameter() {group=RimLight, parameterName=RimMorning, defaultValue=<rgba 0.875 0.8626 0.7474 1>}
  n12 = VectorParameter() {group=RimLight, parameterName=RimDay, defaultValue=<rgba 1.2 0.9092 0.5938 1>}
  n13 = VectorParameter() {group=RimLight, parameterName=RimEvening, defaultValue=<rgba 0.9635 0.5427 0.2158 1>}
  n14 = VectorParameter() {group=RimLight, parameterName=RimNight, defaultValue=<rgba 0.3927 0.3462 0.651 1>}
  n15 = Desaturation(None=Clamp(None=@n1.BaseColor) {maxDefault=0.7, minDefault=0.03}, Fraction=Constant() {r=0.6})
  n16 = MaterialFunctionCall!(
   BaseColor=@n15
   Gmap=@n6
   LightVector=VectorParameter() {group=RimLight, parameterName=RimLightDirection01, defaultValue=<rgba -2 1 2 1>}.RGB
  ) {materialFunction=/Game/Packages/Fortress_SharedMaterials/Base_Material_Functions/MF_Rimlight_SingleLight}
  n17 = MaterialFunctionCall(
   BaseColor=@n15
   Gmap=@n6
   LightVector=VectorParameter() {group=RimLight, parameterName=RimLightDirection02, defaultValue=<rgba 0.6 0.5 1.5 1>}.RGB
  ) {materialFunction=&MF_Rimlight_SingleLight}
  n18 = Multiply(
   A=Multiply(
    A=MaterialFunctionCall(Color_Morning=@n11.RGB, Color_Day=@n12.RGB, Color_Evening=@n13.RGB, Color_Night=@n14.RGB) {materialFunction=&Color_Time_of_Day}.Result
    B=Add(A=@n16.Result, B=@n17.Result)
   )
   B=ScalarParameter() {defaultValue=6, desc="artemis = 2.2 - Asteria 6(pending)", group=RimLight, parameterName=RimLightIntensity}
  )
  NamedRerouteDeclaration(None=Clamp(None=Multiply(A=@n10, B=@n18)) {maxDefault=500}) {name=RimLightColors, nodeColor=<rgba 0 1 0.1094 1>}
  PreSkinnedLocalBounds()
  Transform()
 ]}
