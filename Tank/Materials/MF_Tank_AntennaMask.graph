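# Material function for the tank that can exempt the antenna region from opacity masking, toggled by the Prevent Antenna Masking static switch.
# It normalizes pre-skinned position within antenna bounding-box parameters and thresholds each axis into an inside/outside mask.
# The mask multiplies X by itself and by Z, so Y is ignored.
# Evaluated per vertex and interpolated, it lerps incoming Opacity Mask toward opaque; with the switch off, attributes pass through unchanged.
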MaterialFunction /Tank/Materials/MF_Tank_AntennaMask
graph {nodes [
  n1 = FunctionInput() {inputType=FunctionInput_MaterialAttributes}
  n2 = GetMaterialAttributes(None=@n1) {attributeGetTypes [#1]}
  n3 = VectorParameter() {parameterName=AntennaBBoxMin, defaultValue=<rgba -347.1604 -57.6865 300 1>}
  n4 = Subtract(A=PreSkinnedPosition(), B=@n3.RGB)
  n5 = Subtract(A=VectorParameter() {parameterName=AntennaBBoxMax, defaultValue=<rgba -328.7761 -45.4515 185417584 1>}.RGB, B=@n3.RGB)
  n6 = Constant3Vector()
  n7 = MaterialFunctionCall(Color=If(
    A=Divide(A=@n4, B=@n5)
    B=Constant3Vector() {constant=<rgba 0.5 0.5 0.5 0>}
    "A > B"=@n6
    "A == B"=Constant3Vector() {constant=<rgba 1 1 1 0>}
    "A < B"=@n6
   ) {equalsThreshold=0.5}) {materialFunction=&SplitComponents}
  n8 = Multiply(A=@n7.R, B=@n7.R)
  n9 = LinearInterpolate(A=@n2.OpacityMask, Alpha=VertexInterpolator(VS=Multiply(A=@n8, B=@n7.B)).PS)
  FunctionOutput(None=StaticSwitchParameter(
    True=SetMaterialAttributes!(MaterialAttributes=@n2.MaterialAttributes, OpacityMask=@n9) {attributeSetTypes [#1]}
    False=@n1
   ) {parameterName=Prevent Antenna Masking})
 ]}
