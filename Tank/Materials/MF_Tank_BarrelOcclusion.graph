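# Material function that hides the front section of a tank barrel via opacity mask, driven by the BarrelOcclusionDistance parameter.
# Interpolated pre-skinned X is compared to the muzzle X, hard-clipping pixels forward of a cutoff that moves back as the distance grows.
# A hardcoded pre-skinned Z band confines the cut to barrel height, and a zero distance leaves it fully opaque.
# The Barrel Occlusion static switch either overrides OpacityMask or passes input attributes through.
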MaterialFunction /Tank/Materials/MF_Tank_BarrelOcclusion
graph {nodes [
  n1 = FunctionInput() {inputType=FunctionInput_MaterialAttributes}
  n2 = VertexInterpolator(VS=PreSkinnedPosition())
  n3 = ComponentMask(None=@n2.PS) {r=true}
  n4 = ScalarParameter() {defaultValue=570.8214, group=Barrel Occlusion, parameterName=Muzzle Location X}
  n5 = ScalarParameter() {group=Barrel Occlusion, parameterName=BarrelOcclusionDistance}
  n6 = Sign(None=Add(A=Subtract(A=@n3, B=@n4), B=Add(A=@n5) {constB=-5}))
  n7 = ComponentMask(None=@n2.PS) {b=true}
  n8 = Constant() {r=271.852}
  n9 = Constant()
  n10 = Constant() {r=1}
  n11 = LinearInterpolate(
   B=OneMinus(None=Saturate(None=@n6))
   Alpha=Multiply(A=Sign(None=@n5), B=If(A=@n7, B=@n8, "A > B"=@n9, "A == B"=@n10, "A < B"=@n9) {equalsThreshold=31.8})
  ) {constA=1}
  FunctionOutput(None=StaticSwitchParameter(True=SetMaterialAttributes!(MaterialAttributes=@n1, OpacityMask=@n11) {attributeSetTypes [#1]}, False=@n1) {group=TankFX, parameterName=Barrel Occlusion})
 ]}
