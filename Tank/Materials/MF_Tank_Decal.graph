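# Material function for tank decals that layers a masked color and surface change onto incoming material attributes.
# Base color lerps toward a multiply blend, roughness toward a scaled value, and metallic and specular toward input values weighted by influence inputs.
# The mask is exposed both raw and range-remapped through named reroutes.
# A selectable blend-mode network (overlay, lighten, darken and others) is disconnected from the output, its selector a placeholder constant.
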MaterialFunction /Tank/Materials/MF_Tank_Decal
graph {nodes [
  n1 = GetMaterialAttributes(None=FunctionInput() {inputName=Material, inputType=FunctionInput_MaterialAttributes}) {attributeGetTypes [
    #1
    #2
    #3
    #4
   ]}
  n2 = Multiply(A=NamedRerouteUsage(), B=NamedRerouteUsage())
  n3 = LinearInterpolate(A=@n1.BaseColor, B=@n2, Alpha=NamedRerouteUsage())
  n4 = FunctionInput() {bUsePreviewValueAsDefault=true, inputName=Roughness Multiplier, inputType=FunctionInput_Scalar, sortPriority=50, previewValue=<1 0 0 1>}
  n5 = NamedRerouteUsage()
  n6 = FunctionInput() {bUsePreviewValueAsDefault=true, inputName=Specular Influence, inputType=FunctionInput_Scalar, sortPriority=505, previewValue=<1 0 0 1>}
  n7 = NamedRerouteUsage()
  n8 = FunctionInput() {bUsePreviewValueAsDefault=true, inputName=Metallic Influence, inputType=FunctionInput_Scalar, sortPriority=105, previewValue=<1 0 0 1>}
  n9 = SetMaterialAttributes!(
   MaterialAttributes=@n1.MaterialAttributes
   BaseColor=@n3
   Metallic=LinearInterpolate(
    A=@n1.Metallic
    B=FunctionInput() {bUsePreviewValueAsDefault=true, inputName=Metallic, inputType=FunctionInput_Scalar, sortPriority=100}
    Alpha=Multiply(A=@n7, B=@n8)
   )
   Roughness=LinearInterpolate(A=@n1.Roughness, B=Multiply(A=@n1.Roughness, B=@n4), Alpha=NamedRerouteUsage())
   Specular=LinearInterpolate(
    A=@n1.Specular
    B=FunctionInput() {bUsePreviewValueAsDefault=true, inputName=Specular, inputType=FunctionInput_Scalar, sortPriority=500}
    Alpha=Multiply(A=@n5, B=@n6)
   )
  ) {attributeSetTypes [
    #1
    #2
    #3
    #4
   ]}
  FunctionOutput(None=@n9)
  n10 = Round(None=Constant() {desc=Replace with function input, r=4})
  n11 = Subtract(A=@n10) {constB=4}
  MaterialFunctionCall(
   1=@n2
   2=MaterialFunctionCall!(Blend=NamedRerouteUsage(), Base=NamedRerouteUsage()) {materialFunction=/Engine/Functions/Engine_MaterialFunctions03/Blends/Blend_Screen}.Result
   3=MaterialFunctionCall!(Blend=NamedRerouteUsage(), Base=NamedRerouteUsage()) {materialFunction=/Engine/Functions/Engine_MaterialFunctions03/Blends/Blend_Darken}.Result
   4=MaterialFunctionCall(Base=NamedRerouteUsage(), Blend=NamedRerouteUsage()) {materialFunction=&Blend_Overlay}.Result
   A=Saturate(None=Divide(A=@n11) {constB=3})
  ) {materialFunction=&Lerp_Multiple_Float3}
  n12 = FunctionInput() {bUsePreviewValueAsDefault=true, inputName=Mask, inputType=FunctionInput_Scalar, sortPriority=10}
  n13 = FunctionInput() {bUsePreviewValueAsDefault=true, inputName=Mask Min, inputType=FunctionInput_Scalar, sortPriority=11}
  n14 = FunctionInput() {bUsePreviewValueAsDefault=true, inputName=Mask Max, inputType=FunctionInput_Scalar, sortPriority=12, previewValue=<1 0 0 1>}
  NamedRerouteDeclaration(None=Saturate(None=MaterialFunctionCall(Input=@n12, Input Low=@n13, Input High=@n14) {materialFunction=&RemapValueRange}.Result)) {name=Mask Remapped, nodeColor=<rgba 0.7109 1 0 1>}
  NamedRerouteDeclaration(None=@n12) {name=Mask Original, nodeColor=<rgba 1 0.0234 0 1>}
  n15 = NamedRerouteDeclaration(None=FunctionInput() {bUsePreviewValueAsDefault=true, inputName=Color, sortPriority=8}) {name=Color, nodeColor=<rgba 1 0 0.3984 1>}
  n16 = MaterialFunctionCall(
   1=MaterialFunctionCall(Base=@n1.BaseColor, Blend=@n15) {materialFunction=&Blend_Overlay}.Result
   2=@n15
   3=MaterialFunctionCall!(Blend=@n15, Base=@n1.BaseColor) {materialFunction=/Engine/Functions/Engine_MaterialFunctions03/Blends/Blend_Lighten}.Result
   4=MaterialFunctionCall(Blend=@n15, Base=@n1.BaseColor) {materialFunction=&Blend_Darken}.Result
   A=Saturate(None=Divide(A=@n10) {constB=3})
  ) {materialFunction=&Lerp_Multiple_Float3}
  n17 = NamedRerouteUsage()
  n18 = Constant() {r=4}
  n19 = Constant() {r=1}
  n20 = Constant()
  Reroute(None=LinearInterpolate(
    A=LinearInterpolate(A=@n1.BaseColor, B=@n16.Lerp 4 Inputs, Alpha=@n17)
    B=@n3
    Alpha=If(A=@n10, B=@n18, "A > B"=@n19, "A == B"=@n19, "A < B"=@n20)
   ) {constAlpha=1}) {desc=connect to base color to re-enable blends}
 ]}
