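# Material function that adds a boost-driven exhaust glow to the emissive of a tank material's attributes.
# A powered sphere mask along UV V blends two glow colors and shapes intensity, limited to areas selected by vertex color red times inverted blue.
# A saturated Boost On scalar from custom primitive data scales the glow per primitive.
# The Exhaust Glow static switch passes the input through untouched when off.
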MaterialFunction /Tank/Materials/MF_Tank_ExhaustGlow
graph {nodes [
  n1 = FunctionInput() {inputName=Material, inputType=FunctionInput_MaterialAttributes}
  n2 = GetMaterialAttributes(None=@n1) {attributeGetTypes [#1]}
  n3 = VertexColor()
  n4 = ComponentMask(None=TextureCoordinate()) {g=true}
  n5 = ScalarParameter() {defaultValue=0.128, parameterName=Glow Position}
  n6 = ScalarParameter() {defaultValue=0.06, parameterName=Glow Radius}
  n7 = Constant()
  n8 = Power(Base=Saturate(None=SphereMask(A=@n4, B=@n5, Radius=@n6, Hardness=@n7)))
  n9 = Multiply(
   A=Multiply(A=@n3.R, B=OneMinus(None=@n3.B))
   B=LinearInterpolate(
    A=VectorParameter() {group=Exhaust Glow, parameterName=Exhaust Glow Color, defaultValue=<rgba 0.5 0.0356 0 0>}.RGB
    B=VectorParameter() {group=Exhaust Glow, parameterName=Exhaust Glow Color 2, defaultValue=<rgba 0.5 0.1962 0 0>}.RGB
    Alpha=@n8
   )
  )
  n10 = Saturate(None=ScalarParameter() {bUseCustomPrimitiveData=true, group=TankFX, parameterName=Boost On, primitiveDataIndex=12})
  n11 = Add(A=@n2.EmissiveColor, B=Multiply(A=Multiply(A=@n9, B=@n10), B=@n8))
  FunctionOutput(None=StaticSwitchParameter(
    True=SetMaterialAttributes!(MaterialAttributes=@n2.MaterialAttributes, EmissiveColor=@n11) {attributeSetTypes [#1]}
    False=@n1
   ) {group=TankFX, parameterName=Exhaust Glow})
 ]}
