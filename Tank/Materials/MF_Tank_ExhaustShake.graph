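# Material function that adds an exhaust shake to tank materials through World Position Offset, toggled by a static switch.
# A fast time-driven sine, masked by vertex color red times inverted blue, rotates pre-skinned positions about a fixed pivot.
# The sign of local Y mirrors pivot and axis, so one setup serves both sides.
# Custom primitive data supplies speed, amplitude and a Tank On multiplier that gates the motion.
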MaterialFunction /Tank/Materials/MF_Tank_ExhaustShake
graph {nodes [
  n1 = VectorParameter() {defaultValue=<rgba 1 1 0 1>}
  n2 = Constant() {r=1}
  n3 = PreSkinnedPosition()
  n4 = ShaderStageSwitch(PixelShader=VertexInterpolator(VS=@n3).PS, VertexShader=@n3)
  n5 = MaterialFunctionCall(X=@n2, Y=Sign(None=ComponentMask(None=@n4) {g=true}), Z=@n2) {materialFunction=&MakeFloat3}
  n6 = ScalarParameter() {bUseCustomPrimitiveData=true, defaultValue=1, group=Exhaust Shake, parameterName=Exhaust Shake Speed, primitiveDataIndex=13}
  n7 = Multiply(A=Time(), B=@n6)
  n8 = VertexColor()
  n9 = Multiply(A=Multiply(A=Sine(None=@n7) {period=0.1}) {constB=0.0035}, B=Multiply(A=@n8.R, B=OneMinus(None=@n8.B)))
  n10 = ScalarParameter() {bUseCustomPrimitiveData=true, defaultValue=1, group=Exhaust Shake, parameterName=Exhaust Shake Amp, primitiveDataIndex=14}
  n11 = RotateAboutAxis(
   NormalizedRotationAxis=Multiply(A=Normalize(VectorInput=@n1.RGB), B=@n5.Result)
   RotationAngle=Multiply(A=Multiply(A=@n9, B=@n10), B=ScalarParameter() {bUseCustomPrimitiveData=true, group=TankFX, parameterName=Tank On, primitiveDataIndex=8})
   PivotPoint=Multiply(A=Constant3Vector() {constant=<rgba -336 112 88 0>}, B=@n5.Result)
   Position=@n4
  )
  n12 = FunctionInput() {inputName=Material, inputType=FunctionInput_MaterialAttributes}
  FunctionOutput(None=StaticSwitchParameter(
    True=MaterialFunctionCall!(Material=@n12, WorldPositionOffset=@n11) {materialFunction=/Engine/Functions/MaterialLayerFunctions/MatLayerBlend_AddWorldPositionOffset}.Blended Material
    False=@n12
   ) {group=TankFX, parameterName=Exhaust Shake})
 ]}
