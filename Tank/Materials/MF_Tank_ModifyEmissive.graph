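# Material function that adds animated tank emissive to incoming material attributes for headlights, engine and hatch, toggled by the Modify Emissive static switch.
# Pre-skinned local X position selects each part's region of the masked emissive texture.
# Custom primitive data drives headlight on/off, an engine glow that pulses at full health and dims when damaged, and the hatch display.
# The hatch lights as a health bar along local Y, hue-shifting and pulsing while healing.
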MaterialFunction /Tank/Materials/MF_Tank_ModifyEmissive
graph {nodes [
  n1 = FunctionInput() {inputType=FunctionInput_MaterialAttributes}
  n2 = GetMaterialAttributes(None=@n1) {attributeGetTypes [#1]}
  n3 = Multiply(
   A=TextureSampleParameter2D() {parameterName=Tank Emissive, samplerType=SAMPLERTYPE_Masks, texture=&T_BlackMask}.RGB
   B=VectorParameter() {parameterName=Emissive Multiplier, defaultValue=<rgba 5 5 5 0>}.RGB
  )
  n4 = LinearInterpolate(
   A=@n3
   B=Multiply(A=@n3, B=ScalarParameter() {bUseCustomPrimitiveData=true, group=TankFX, parameterName=Tank On, primitiveDataIndex=8})
   Alpha=NamedRerouteUsage()
  )
  n5 = Saturate(None=ScalarParameter() {bUseCustomPrimitiveData=true, defaultValue=1, group=TankFX, parameterName=Engine Health, primitiveDataIndex=9})
  n6 = Constant() {r=1}
  n7 = Sine(None=Time()) {period=0.4}
  n8 = Add(A=Multiply(A=@n5) {constB=0.7}, B=Multiply(A=ConstantBiasScale(None=@n7)) {constB=2})
  n9 = LinearInterpolate(A=@n4, B=Multiply(A=@n4, B=If(A=@n5, B=@n6, "A > B"=@n8, "A == B"=@n8, "A < B"=@n5)), Alpha=NamedRerouteUsage())
  n10 = ScalarParameter() {bUseCustomPrimitiveData=true, parameterName=Hatch Healing, primitiveDataIndex=10}
  n11 = MaterialFunctionCall(Hue Shift Percentage=LinearInterpolate(Alpha=@n10) {constB=-0.335}, Texture=@n9) {materialFunction=&HueShift}
  n12 = Sine(None=Time())
  n13 = ScalarParameter() {bUseCustomPrimitiveData=true, defaultValue=1, parameterName=Hatch Health, primitiveDataIndex=11}
  n14 = NamedRerouteUsage()
  n15 = Multiply(
   A=LinearInterpolate(B=ConstantBiasScale(None=@n12) {bias=1.5}, Alpha=@n10) {constA=1}
   B=If(A=OneMinus(None=@n13), B=OneMinus(None=@n14), "A > B"=Constant() {r=1}, "A < B"=Constant())
  )
  n16 = Add(A=@n2.EmissiveColor, B=LinearInterpolate(A=@n9, B=Multiply(A=@n11.Result, B=@n15), Alpha=NamedRerouteUsage()))
  FunctionOutput(None=StaticSwitchParameter(
    True=SetMaterialAttributes!(MaterialAttributes=@n2.MaterialAttributes, EmissiveColor=@n16) {attributeSetTypes [#1]}
    False=@n1
   ) {group=TankFX, parameterName=Modify Emissive})
  n17 = Constant()
  n18 = VertexInterpolator(VS=PreSkinnedPosition())
  n19 = ComponentMask(None=@n18.PS) {r=true}
  NamedRerouteDeclaration(None=If(A=@n19, B=Constant() {r=290}, "A > B"=@n17, "A == B"=Constant() {r=1}, "A < B"=@n17) {desc=Head light mask, equalsThreshold=20}) {name=Headlight Mask, nodeColor=<rgba 0.4297 1 0 1>}
  n20 = Constant()
  NamedRerouteDeclaration(None=If(A=@n19, B=Constant() {r=-400}, "A > B"=@n20, "A == B"=Constant() {r=1}, "A < B"=@n20) {desc=Engine Mask, equalsThreshold=10}) {name=Engine Mask, nodeColor=<rgba 1 0 0.1172 1>}
  n21 = Constant()
  NamedRerouteDeclaration(None=If(A=@n19, B=Constant() {r=260}, "A > B"=@n21, "A == B"=Constant() {r=1}, "A < B"=@n21) {desc=Hatch Mask, equalsThreshold=10}) {name=Hatch Mask, nodeColor=<rgba 1 0.5625 0 1>}
  n22 = Constant() {r=-40}
  n23 = Constant() {r=40}
  n24 = ComponentMask(None=@n18.PS) {g=true}
  NamedRerouteDeclaration(None=Saturate(None=InverseLinearInterpolate(A=@n22, B=@n23, Value=@n24))) {name=Hatch Health Gradient, nodeColor=<rgba 0 0.9062 1 1>}
  OneMinus()
 ]}
