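# Material function that adds animated tank treads, with damage and repair effects, to input material attributes behind a static switch.
# Pre-skinned Y picks left or right values, mostly from custom primitive data, scrolling UVs by position and blending to blurred textures by speed.
# Damage collapses vertices to world origin through WPO.
# Repair time drives a UV wipe of opacity mask, Fresnel-tinted glow and normal offset, optionally preserving the incoming mask.
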MaterialFunction /Tank/Materials/MF_Tank_Treads
graph {nodes [
  n1 = TextureSampleParameter2D() {
   group=Texture Input
   parameterName=Normal_Blurred
   samplerType=SAMPLERTYPE_Normal
   texture=&FlatNormalBlue
  }
  n2 = FunctionInput() {inputName=Material, inputType=FunctionInput_MaterialAttributes}
  n3 = BlendMaterialAttributes(
   A=@n2
   B=SetMaterialAttributes!(
    MaterialAttributes=@n2
    BaseColor=TextureSampleParameter2D() {group=Texture Input, parameterName=Diffuse_Blurred, texture=&T_BlacksRGB}.RGB
    Normal=@n1.RGB
   ) {attributeSetTypes [#1, #2]}
   Alpha=NamedRerouteUsage()
  )
  n4 = NamedRerouteUsage()
  n5 = NamedRerouteUsage()
  n6 = NamedRerouteUsage()
  n7 = MaterialFunctionCall!(
   Material=MaterialFunctionCall!(
    Material=SetMaterialAttributes!(MaterialAttributes=@n3, CustomizedUV0=@n4) {attributeSetTypes [#3]}
    Mask
    WorldPositionOffset=Add(A=@n5, B=@n6)
   ) {materialFunction=/Engine/Functions/MaterialLayerFunctions/MatLayerBlend_OverrideWorldPositionOffset}.Blended Material
   Emissive=NamedRerouteUsage()
  ) {materialFunction=/Engine/Functions/MaterialLayerFunctions/MatLayerBlend_Emissive}
  n8 = GetMaterialAttributes(None=@n7.Blended Material) {attributeGetTypes [#4]}
  n9 = NamedRerouteUsage()
  n10 = SetMaterialAttributes!(MaterialAttributes=@n8.MaterialAttributes, OpacityMask=Multiply(A=@n8.OpacityMask, B=@n9)) {attributeSetTypes [#4]}
  n11 = MaterialFunctionCall!(Material=@n7.Blended Material, NewOpacityMask=@n9, Mask) {materialFunction=/Engine/Functions/MaterialLayerFunctions/MatLayerBlend_OverrideOpacityMask}
  FunctionOutput(None=StaticSwitchParameter(True=StaticSwitchParameter(True=@n10, False=@n11.Blended Material) {parameterName=Treads Camera Mask}, False=@n2) {group=TankFX, parameterName=Treads})
  n12 = Saturate(None=ComponentMask(None=PreSkinnedPosition()) {g=true})
  n13 = ShaderStageSwitch(PixelShader=VertexInterpolator(VS=@n12).PS, VertexShader=@n12)
  NamedRerouteDeclaration(None=@n13) {name=Treads Mask, desc="A Left, B Right", nodeColor=<rgba 1 0.0703 0 1>}
  n14 = Round(None=LinearInterpolate(
    A=ScalarParameter() {bUseCustomPrimitiveData=true, group=Tank Treads, parameterName=Tread Damage L, primitiveDataIndex=5}
    B=ScalarParameter() {bUseCustomPrimitiveData=true, group=Tank Treads, parameterName=Tread Damage R, primitiveDataIndex=4}
    Alpha=NamedRerouteUsage()
   ))
  n15 = WorldPosition()
  NamedRerouteDeclaration(None=Multiply(A=Saturate(None=@n14), B=Multiply(A=@n15.XYZ) {constB=-1})) {name=Tread Damage, nodeColor=<rgba 1 0.3047 0 1>}
  n16 = VectorParameter() {parameterName=Repair Glow 1, defaultValue=<rgba 0 0.132 1 1>}
  n17 = VectorParameter() {parameterName=Repair Glow 2, defaultValue=<rgba 0 0.5542 1 1>}
  n18 = Fresnel(Normal=PixelNormalWS())
  n19 = ScalarParameter() {defaultValue=0.1, parameterName=Glow Falloff}
  n20 = ScalarParameter() {defaultValue=0.026, parameterName=WPO Falloff}
  n21 = ScalarParameter() {bUseCustomPrimitiveData=true, defaultValue=-100, parameterName=Tread Repair Time L, primitiveDataIndex=7}
  n22 = ScalarParameter() {bUseCustomPrimitiveData=true, defaultValue=-100, parameterName=Tread Repair Time R, primitiveDataIndex=6}
  n23 = NamedRerouteUsage()
  n24 = Divide(
   A=Subtract(A=Time(), B=LinearInterpolate(A=@n21, B=@n22, Alpha=@n23))
   B=ScalarParameter() {defaultValue=1.2, parameterName=Tread Repair Duration}
  ) {constB=1}
  n25 = LinearInterpolate(A=NamedRerouteUsage(), B=NamedRerouteUsage(), Alpha=NamedRerouteUsage())
  n26 = InverseLinearInterpolate(A=Add(A=@n25, B=Constant() {r=-2}), B=Add(A=@n25, B=Constant() {r=3}), Value=ComponentMask(None=TextureCoordinate()) {g=true})
  n27 = SmoothStep(Min=Subtract(A=@n24, B=ShaderStageSwitch(PixelShader=@n19, VertexShader=@n20)) {constB=0.25}, Max=@n24, Value=@n26)
  NamedRerouteDeclaration(None=Multiply(A=LinearInterpolate(A=@n16.RGB, B=@n17.RGB, Alpha=@n18), B=@n27)) {name=Repair Emissive, nodeColor=<rgba 0.6172 0 1 1>}
  NamedRerouteDeclaration(None=OneMinus(None=SmoothStep(Min=@n24, Max=@n24, Value=@n26))) {name=Repair Mask, nodeColor=<rgba 0.3594 0 1 1>}
  n28 = VertexNormalWS()
  NamedRerouteDeclaration(None=Multiply(A=Multiply(A=@n27, B=@n28), B=ScalarParameter() {defaultValue=-2.5, parameterName=Repair Vertex Offset})) {name=Repair WPO Offset, nodeColor=<rgba 0.6875 1 0 1>}
  n29 = VectorParameter() {
   desc="R Right Pos G Right Vel, B Left Pos A Left Vel"
   group=Tank Treads
   parameterName=TreadsPosition
   channelNames {a=Tread Left Blur, b=Tread Left Position, g=Tread Right Blur, r=Tread Right Position}
  }
  n30 = VectorParameter() {
   bUseCustomPrimitiveData=true
   desc="R Right Pos G Right Vel, B Left Pos A Left Vel"
   group=Tank Treads
   parameterName=TreadsPosition
   channelNames {a=Tread Left Blur, b=Tread Left Position, g=Tread Right Blur, r=Tread Right Position}
  }
  n31 = Add(A=@n29.Tread Left Position, B=@n30.Tread Left Position)
  n32 = Divide(A=Constant() {desc=Tread Length, r=1595.05}, B=Constant() {desc=Tread Lenght UV Space, r=-5})
  n33 = MaterialFunctionCall(Y=NamedRerouteDeclaration(None=Divide(A=@n31, B=@n32)) {name=UV Position L, nodeColor=<rgba 0 1 0.0391 1>}) {materialFunction=&MakeFloat2}
  n34 = TextureCoordinate()
  n35 = Add(A=@n29.Tread Right Position, B=@n30.Tread Right Position)
  n36 = MaterialFunctionCall(Y=NamedRerouteDeclaration(None=Divide(A=@n35, B=@n32)) {name=UV Position R, nodeColor=<rgba 1 0 0.2812 1>}) {materialFunction=&MakeFloat2}
  NamedRerouteDeclaration(None=LinearInterpolate(A=Add(A=@n33.Result, B=@n34), B=Add(A=@n36.Result, B=@n34), Alpha=@n13) {desc="A Left, B Right tread"}) {name=Treads, nodeColor=<rgba 1 0 0.6094 1>}
  n37 = Abs(None=LinearInterpolate(A=Max(A=@n29.Tread Left Blur, B=@n30.Tread Left Blur), B=Max(A=@n29.Tread Right Blur, B=@n30.Tread Right Blur), Alpha=@n13) {desc="A Left, B Right tread"})
  NamedRerouteDeclaration(None=Saturate(None=Divide(A=@n37) {constB=800})) {name=Treads Blur, nodeColor=<rgba 1 0.6797 0 1>}
 ]}
