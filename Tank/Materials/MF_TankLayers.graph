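# Material function that layers tank-specific effects and two optional tinted decals onto incoming material attributes.
# It chains camera dither, exhaust shake and glow, treads, barrel occlusion, emissive tweak and antenna mask functions.
# Switchable MF_Tank_Decal calls follow, masked by selectable channels of one UV1 texture, with color, roughness, metallic and specular controls.
# "Use Tank Functions" off bypasses everything; both Blend Mode parameters are unconnected.
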MaterialFunction /Tank/Materials/MF_TankLayers
graph {nodes [
  n1 = TextureSampleParameter2D(UVs=TextureCoordinate() {coordinateIndex=1}) {group=Tank Decal, parameterName=Decal Mask Texture, samplerType=SAMPLERTYPE_Masks, texture=&T_BlackMask}
  n2 = ChannelMaskParameter(None=@n1.RGBA) {group=Tank Decal, maskChannel=Alpha, parameterName=Decal 2 Mask Channel, defaultValue=<rgba 0 0 0 1>}
  n3 = ScalarParameter() {defaultValue=1, desc=(0 Overlay, 1 Replace, 2, Lighten, 3 Darken), group=Tank Decal, parameterName=Decal 2 Metallic Influence, sliderMax=1}
  n4 = ScalarParameter() {defaultValue=1, desc=(0 Overlay, 1 Replace, 2, Lighten, 3 Darken), group=Tank Decal, parameterName=Decal 2 Specular Influence, sliderMax=1}
  n5 = ScalarParameter() {defaultValue=1, desc=(0 Overlay, 1 Replace, 2, Lighten, 3 Darken), group=Tank Decal, parameterName=Decal 1 Metallic Influence, sliderMax=1}
  n6 = ScalarParameter() {defaultValue=1, desc=(0 Overlay, 1 Replace, 2, Lighten, 3 Darken), group=Tank Decal, parameterName=Decal 1 Specular Influence, sliderMax=1}
  n7 = FunctionInput() {inputType=FunctionInput_MaterialAttributes}
  n8 = MaterialFunctionCall!(In=@n7) {materialFunction=/Game/Packages/Fortress_SharedMaterials/Base_Material_Functions/BaseCustomizationFunctions/MF_FN_Vehicle_CameraDither}
  n9 = MaterialFunctionCall!(Material=MaterialFunctionCall!(Material=MaterialFunctionCall!(Material=@n8.Result) {materialFunction=&/MF_Tank_ExhaustShake}.Result) {materialFunction=&/MF_Tank_ExhaustGlow}.Result) {materialFunction=&/MF_Tank_Treads}
  n10 = MaterialFunctionCall!(In=MaterialFunctionCall!(In=MaterialFunctionCall!(In=@n9.Result) {materialFunction=&/MF_Tank_BarrelOcclusion}.Result) {materialFunction=&/MF_Tank_ModifyEmissive}.Result) {materialFunction=&/MF_Tank_AntennaMask}
  n11 = StaticSwitchParameter(
   True=MaterialFunctionCall!(
    Material=@n10.Result
    Color=VectorParameter() {group=Tank Decal, parameterName=Decal 1 Color, defaultValue=<rgba 0.1044 0.2019 0.401 1>}.RGB
    Mask=ChannelMaskParameter(None=@n1.RGBA) {group=Tank Decal, parameterName=Decal 1 Mask Channel}.R
    Mask Min=ScalarParameter() {group=Tank Decal, parameterName=Decal 1 Mask Remap Min, sliderMax=1}
    Mask Max=ScalarParameter() {defaultValue=0.45, group=Tank Decal, parameterName=Decal 1 Mask Remap Max, sliderMax=1}
    Roughness Multiplier=ScalarParameter() {defaultValue=1.6, group=Tank Decal, parameterName=Decal 1 Roughness Multiplier}
    Metallic=ScalarParameter() {defaultValue=0.08, desc=(0 Overlay, 1 Replace, 2, Lighten, 3 Darken), group=Tank Decal, parameterName=Decal 1 Metallic}
    Metallic Influence=@n5
    Specular=ScalarParameter() {defaultValue=0.5, desc=(0 Overlay, 1 Replace, 2, Lighten, 3 Darken), group=Tank Decal, parameterName=Decal 1 Specular}
    Specular Influence=@n6
   ) {materialFunction=&/MF_Tank_Decal}.Result
   False=@n10.Result
  ) {group=Tank Decal, parameterName=Enable Decal 1, sortPriority=2}
  n12 = MaterialFunctionCall(
   Material=@n11
   Color=VectorParameter() {group=Tank Decal, parameterName=Decal 2 Color, defaultValue=<rgba 0.5156 0.099 0.051 1>}.RGB
   Mask=@n2.R
   Mask Min=ScalarParameter() {group=Tank Decal, parameterName=Decal 2 Mask Remap Min, sliderMax=1}
   Mask Max=ScalarParameter() {defaultValue=1, group=Tank Decal, parameterName=Decal 2 Mask Remap Max, sliderMax=1}
   Roughness Multiplier=ScalarParameter() {defaultValue=0.3, group=Tank Decal, parameterName=Decal 2 Roughness Multiplier}
   Metallic=ScalarParameter() {defaultValue=0.06, desc=(0 Overlay, 1 Replace, 2, Lighten, 3 Darken), group=Tank Decal, parameterName=Decal 2 Metallic}
   Metallic Influence=@n3
   Specular=ScalarParameter() {defaultValue=0.5, desc=(0 Overlay, 1 Replace, 2, Lighten, 3 Darken), group=Tank Decal, parameterName=Decal 2 Specular}
   Specular Influence=@n4
  ) {materialFunction=&MF_Tank_Decal}
  FunctionOutput(None=StaticSwitchParameter(True=StaticSwitchParameter(True=@n12.Result, False=@n11) {group=Tank Decal, parameterName=Enable Decal 2, sortPriority=2}, False=@n7) {parameterName=Use Tank Functions})
  ScalarParameter() {
   desc="(Method 1: 0 Overlay, 1 Replace, 2, Lighten, 3 Darken)\r\n(Method 2: 4 Multiply, 5 Screen, 6, SoftLight, 7 Overlay)"
   group=Tank Decal
   parameterName="Decal 2 Blend Mode (0 to 7 read tooltip)"
  }
  ScalarParameter() {
   desc="(Method 1: 0 Overlay, 1 Replace, 2, Lighten, 3 Darken)\r\n(Method 2: 4 Multiply, 5 Screen, 6, SoftLight, 7 Overlay)"
   group=Tank Decal
   parameterName="Decal 1 Blend Mode (0 to 7 read tooltip)"
  }
 ]}
