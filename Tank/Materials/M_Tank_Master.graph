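# Master Material Attributes material for the Tank vehicle, built from shared Valet vehicle functions.
# Base functionality feeds headlights, turn signals, decal, reverse lights, damage, mud, party lights, tank layers and occlusion prevention in sequence, then an optional material-layer stack.
# Pre-skinned position within local bounds drives light masks and damage, though the divisor subtracts Max from itself, yielding zero.
# A rimlight branch off damage is left unconnected.
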Material /Tank/Materials/M_Tank_Master
graph {
 properties {
  bUseMaterialAttributes=true
  bUsedWithInstancedStaticMeshes=true
  bUsedWithMeshDeformer=true
  bUsedWithSkeletalMesh=true
  bUsedWithSplineMeshes=true
  ditherOpacityMask=true
  pixelDepthOffsetMode=PDOM_Legacy
  usedShadingModels=Default Lit
 }
 nodes [
  n1 = PreSkinnedPosition()
  n2 = VertexInterpolator(VS=@n1)
  n3 = PreSkinnedLocalBounds()
  n4 = Divide(A=Subtract(A=@n2.PS, B=@n3.Min), B=Subtract(A=@n3.Max, B=@n3.Max))
  n5 = Round(None=ComponentMask(None=@n4) {r=true})
  n6 = MaterialFunctionCall() {materialFunction=/Valet/Materials/MF_Valet_BaseFunctionality}
  n7 = MaterialFunctionCall!(
   In=MaterialFunctionCall!(Mask=Multiply(A=@n5, B=@n6.Lights), MA=@n6.Result) {materialFunction=/Valet/Materials/MF_Valet_Headlights}.Result
   CenterLine=StaticSwitchParameter(True=ComponentMask(None=@n4) {g=true}, False=Constant()) {defaultValue=true, parameterName=UseCenterLine}
   Alpha=Multiply(A=@n6.Lights, B=OneMinus(None=@n5))
  ) {materialFunction=/Game/Weapons/WeaponSkins/MaterialFunction/TurnSignals}
  n8 = MaterialFunctionCall!(
   In=MaterialFunctionCall!(MA=MaterialFunctionCall!(Material=@n7.Result) {materialFunction=/Valet/Materials/MF_Valet_Decal}.Result) {materialFunction=/Valet/Materials/MF_Valet_ReverseLights}.Result
   Metal Tint=@n6.Scratch Color
   Local Bounds Split=@n4
  ) {materialFunction=/Game/Vehicles/Valet/MF_Valet_Damage}
  MaterialFunctionCall!(Gmap=@n6.Gmap, Pre Skinned Local Position=@n2.PS, In=@n8.Result) {materialFunction=/Game/Vehicles/Valet/MF_ValetRimlight}
  n9 = MaterialFunctionCall!(Material=@n8.Result, PreskinnedPos=@n1) {materialFunction=/Valet/Materials/MF_ValetMud}
  n10 = MaterialFunctionCall!(In=MaterialFunctionCall!(In=MaterialFunctionCall!(In=@n9.Result) {materialFunction=/Game/Packages/Fortress_SharedMaterials/PartyBus_TempLocation/MF_ArmoredBattleBus_PartyLights}.Result) {materialFunction=&/MF_TankLayers}.Result) {materialFunction=/Valet/Materials/MF_OcclusionPrevention}
  StaticSwitchParameter(True=MaterialAttributeLayers!(None=@n10.Result), False=@n10.Result) {parameterName=Use Material Layers}
 ]
}
