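# Physics asset for the tank vehicle skeletal mesh's turret, defining a jointed antenna that can flex under physics.
# Three capsule bodies form the antenna chain, rooted on a tiny capsule on the turret bone that serves as the attachment point rather than the turret's collision shape.
# Constraints link each segment to its parent, with swing limits widening toward the tip and twist held nearly rigid, so the antenna bends like a whip.
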PhysicsAsset /Tank/Meshes/SK_Tank_Vehicle_Physics_Turret
bodies {
 antenna_01 [{shapeType=Capsule, length=30, radius=1, shapeName=antenna_01_capsule, center=<15 0 0>, rotation=<rot 90 0 0>}]
 antenna_02 [{shapeType=Capsule, length=30, radius=1, shapeName=antenna_02_capsule, center=<15 0 0>, rotation=<rot 90 0 0>}]
 antenna_03 [{shapeType=Capsule, length=30, radius=1, shapeName=antenna_03_capsule, center=<15 0 0>, rotation=<rot 90 0 0>}]
 turret [{shapeType=Capsule, length=1, radius=1, shapeName=turret_capsule}]
}
constraints [
 {bone1Name=antenna_02, bone2Name=antenna_01, swing1LimitDegrees=7, swing2LimitDegrees=7, twistLimitDegrees=1}
 {bone1Name=antenna_03, bone2Name=antenna_02, swing1LimitDegrees=10, swing2LimitDegrees=10, twistLimitDegrees=1}
 {bone1Name=antenna_01, bone2Name=turret, swing1LimitDegrees=5, swing2LimitDegrees=5, twistLimitDegrees=1}
]
