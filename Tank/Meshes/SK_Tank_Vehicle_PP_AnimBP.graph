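# Post-process style AnimBlueprint for the tank vehicle skeleton that adds secondary physics motion to an incoming pose.
# A RigidBody node simulates the turret physics asset in turret space, then a chain of spring-driven Anim Dynamics nodes swings turret and body handles.
# Each update, a deprecated gravity override function sets the gravity vector used by the RigidBody node.
# An Anim Dynamics chain for three antenna bones exists but is not connected to the output pose.
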AnimBlueprint /Tank/Meshes/SK_Tank_Vehicle_PP_AnimBP
target_skeleton=&/SK_Tank_Vehicle_Skeleton
graphs [
 AnimGraph {nodes [
   n1 = Anim Dynamics - Bone: antenna_01(Alpha=1, ExternalForce="0,0,0") {
    angularDampingOverride=1
    angularSpringConstant=50000
    bAngularSpring=true
    bGravityOverrideInSimSpace=true
    bOverrideAngularDamping=true
    bUseGravityOverride=true
    gravityScale=0.35
    simulationSpace=BoneRelative
    alphaBoolBlend {blendOption=Linear}
    boundBone {boneName=antenna_01}
    componentAppliedLinearAccClamp=<5000 5000 5000>
    componentLinearAccScale=<0.3 0.3 0.3>
    componentLinearVelScale=<0.5 0.5 0.5>
    gravityOverride {z=-980}
    physicsBodyDefinitions [{
      boundBone {boneName=antenna_01}
      boxExtents=<20 3 3>
      constraintSetup {angularLimitsMax=<2 0 2>, angularLimitsMin=<-2 0 -2>, angularTarget=<0 0 0>}
      localJointOffset=<10 0 0>
     }]
    relativeSpaceBone {boneName=turret}
   }
   n2 = Anim Dynamics - Bone: antenna_02(ComponentPose=@n1, Alpha=1, ExternalForce="0,0,0") {
    angularDampingOverride=1
    angularSpringConstant=50000
    bAngularSpring=true
    bGravityOverrideInSimSpace=true
    bOverrideAngularDamping=true
    bUseGravityOverride=true
    gravityScale=0.35
    simulationSpace=BoneRelative
    alphaBoolBlend {blendOption=Linear}
    boundBone {boneName=antenna_02}
    componentAppliedLinearAccClamp=<6000 6000 6000>
    componentLinearAccScale=<0.3 0.3 0.3>
    componentLinearVelScale=<0.6 0.6 0.6>
    gravityOverride {z=-980}
    physicsBodyDefinitions [{
      boundBone {boneName=antenna_02}
      boxExtents=<20 3 3>
      constraintSetup {angularLimitsMax=<5 0 5>, angularLimitsMin=<-5 0 -5>, angularTarget=<0 0 0>}
      localJointOffset=<10 0 0>
     }]
    relativeSpaceBone {boneName=antenna_01}
   }
   Anim Dynamics - Bone: antenna_03(ComponentPose=@n2, Alpha=1, ExternalForce="0,0,0") {
    angularDampingOverride=1
    angularSpringConstant=50000
    bAngularSpring=true
    bGravityOverrideInSimSpace=true
    bOverrideAngularDamping=true
    bUseGravityOverride=true
    gravityScale=0.35
    simulationSpace=BoneRelative
    alphaBoolBlend {blendOption=Linear}
    boundBone {boneName=antenna_03}
    componentAppliedLinearAccClamp=<6000 6000 6000>
    componentLinearAccScale=<0.3 0.3 0.3>
    componentLinearVelScale=<0.6 0.6 0.6>
    gravityOverride {z=-980}
    physicsBodyDefinitions [{
      boundBone {boneName=antenna_03}
      boxExtents=<20 3 3>
      constraintSetup {angularLimitsMax=<8 0 8>, angularLimitsMin=<-8 0 -8>, angularTarget=<0 0 0>}
      localJointOffset=<10 0 0>
     }]
    relativeSpaceBone {boneName=antenna_02}
   }
   n3 = RigidBody!(
    ComponentPose=Local To Component(LocalPose=Input Pose())
    bAlphaBoolEnabled:bool
    OverrideWorldGravity:Vector=Get Gravity Override()
    ExternalForce
   ) -> (Pose) {
    alphaInputType=Bool
    bClampLinearTranslationLimitToRefPose=true
    bOverrideWorldGravity=true
    overridePhysicsAsset=&/SK_Tank_Vehicle_Physics_Turret
    simulationSpace=BaseBoneSpace
    alphaBoolBlend {blendOption=Linear}
    baseBoneRef {boneName=turret}
    componentAppliedLinearAccClamp=<4000 4000 4000>
    componentLinearAccScale=<0.3 0.3 0.3>
    componentLinearVelScale=<3 0.3 0.3>
    overrideWorldGravity {z=-980}
   }
   n4 = Anim Dynamics - Bone: handle_back_l(ComponentPose=@n3, Alpha=1, ExternalForce="0,0,0") {
    angularDampingOverride=1
    angularSpringConstant=40000
    bAngularSpring=true
    bGravityOverrideInSimSpace=true
    bOverrideAngularDamping=true
    bUseGravityOverride=true
    gravityScale=0.35
    simulationSpace=BoneRelative
    alphaBoolBlend {blendOption=Linear}
    boundBone {boneName=handle_back_l}
    componentAppliedLinearAccClamp=<6000 6000 6000>
    componentLinearAccScale=<0.4 0.4 0.4>
    componentLinearVelScale=<0.4 0.4 0.4>
    gravityOverride {z=-980}
    physicsBodyDefinitions [{
      boundBone {boneName=handle_back_l}
      boxExtents=<80 20 60>
      constraintSetup {angularLimitsMax=<0 0 5>, angularLimitsMin=<0 0 -45>, angularTarget=<0 0 0>}
      localJointOffset=<40 0 0>
     }]
    relativeSpaceBone {boneName=turret}
   }
   n5 = Anim Dynamics - Bone: handle_back_r(ComponentPose=@n4, Alpha=1, ExternalForce="0,0,0") {
    angularDampingOverride=1
    angularSpringConstant=40000
    bAngularSpring=true
    bGravityOverrideInSimSpace=true
    bOverrideAngularDamping=true
    bUseGravityOverride=true
    gravityScale=0.35
    simulationSpace=BoneRelative
    alphaBoolBlend {blendOption=Linear}
    boundBone {boneName=handle_back_r}
    componentAppliedLinearAccClamp=<6000 6000 6000>
    componentLinearAccScale=<0.4 0.4 0.4>
    componentLinearVelScale=<0.4 0.4 0.4>
    gravityOverride {z=-980}
    physicsBodyDefinitions [{
      boundBone {boneName=handle_back_r}
      boxExtents=<60 20 60>
      constraintSetup {angularLimitsMax=<0 0 3>, angularLimitsMin=<0 0 -40>, angularTarget=<0 0 0>}
      localJointOffset=<30 0 0>
     }]
    relativeSpaceBone {boneName=turret}
   }
   n6 = Anim Dynamics - Bone: handle_01_r(ComponentPose=@n5, Alpha=1, ExternalForce="0,0,0") {
    angularDampingOverride=1
    angularSpringConstant=20000
    bAngularSpring=true
    bGravityOverrideInSimSpace=true
    bOverrideAngularDamping=true
    bUseGravityOverride=true
    gravityScale=0.35
    simulationSpace=BoneRelative
    alphaBoolBlend {blendOption=Linear}
    boundBone {boneName=handle_01_r}
    componentAppliedLinearAccClamp=<8000 8000 8000>
    componentLinearAccScale=<0.4 0.4 0.4>
    componentLinearVelScale=<0.4 0.4 0.4>
    gravityOverride {z=-980}
    physicsBodyDefinitions [{
      boundBone {boneName=handle_01_r}
      boxExtents=<40 10 30>
      constraintSetup {angularLimitsMax=<0 0 10>, angularLimitsMin=<0 0 -45>, angularTarget=<0 0 0>}
      localJointOffset=<20 0 0>
     }]
    relativeSpaceBone {boneName=body}
   }
   n7 = Anim Dynamics - Bone: handle_02_r(ComponentPose=@n6, Alpha=1, ExternalForce="0,0,0") {
    angularDampingOverride=1
    angularSpringConstant=20000
    bAngularSpring=true
    bGravityOverrideInSimSpace=true
    bOverrideAngularDamping=true
    bUseGravityOverride=true
    gravityScale=0.35
    simulationSpace=BoneRelative
    alphaBoolBlend {blendOption=Linear}
    boundBone {boneName=handle_02_r}
    componentAppliedLinearAccClamp=<8000 8000 8000>
    componentLinearAccScale=<0.4 0.4 0.4>
    componentLinearVelScale=<0.4 0.4 0.4>
    gravityOverride {z=-980}
    physicsBodyDefinitions [{
      boundBone {boneName=handle_02_r}
      boxExtents=<40 10 30>
      constraintSetup {angularLimitsMax=<0 0 10>, angularLimitsMin=<0 0 -45>, angularTarget=<0 0 0>}
      localJointOffset=<20 0 0>
     }]
    relativeSpaceBone {boneName=body}
   }
   n8 = Anim Dynamics - Bone: handle_01_l(ComponentPose=@n7, Alpha=1, ExternalForce="0,0,0") {
    angularDampingOverride=1
    angularSpringConstant=20000
    bAngularSpring=true
    bGravityOverrideInSimSpace=true
    bOverrideAngularDamping=true
    bUseGravityOverride=true
    gravityScale=0.35
    simulationSpace=BoneRelative
    alphaBoolBlend {blendOption=Linear}
    boundBone {boneName=handle_01_l}
    componentAppliedLinearAccClamp=<8000 8000 8000>
    componentLinearAccScale=<0.4 0.4 0.4>
    componentLinearVelScale=<0.4 0.4 0.4>
    gravityOverride {z=-980}
    physicsBodyDefinitions [{
      boundBone {boneName=handle_01_l}
      boxExtents=<40 10 30>
      constraintSetup {angularLimitsMax=<0 0 10>, angularLimitsMin=<0 0 -45>, angularTarget=<0 0 0>}
      localJointOffset=<20 0 0>
     }]
    relativeSpaceBone {boneName=body}
   }
   n9 = Anim Dynamics - Bone: handle_02_l(ComponentPose=@n8, Alpha=1, ExternalForce="0,0,0") {
    angularDampingOverride=1
    angularSpringConstant=20000
    bAngularSpring=true
    bGravityOverrideInSimSpace=true
    bOverrideAngularDamping=true
    bUseGravityOverride=true
    gravityScale=0.35
    simulationSpace=BoneRelative
    alphaBoolBlend {blendOption=Linear}
    boundBone {boneName=handle_02_l}
    componentAppliedLinearAccClamp=<8000 8000 8000>
    componentLinearAccScale=<0.4 0.4 0.4>
    componentLinearVelScale=<0.4 0.4 0.4>
    gravityOverride {z=-980}
    physicsBodyDefinitions [{
      boundBone {boneName=handle_02_l}
      boxExtents=<40 10 30>
      constraintSetup {angularLimitsMax=<0 0 10>, angularLimitsMin=<0 0 -45>, angularTarget=<0 0 0>}
      localJointOffset=<20 0 0>
     }]
    relativeSpaceBone {boneName=body}
   }
   Output Pose(Result=Component To Local(ComponentPose=@n9)) {layerGroup=None}
  ]}
 EventGraph {nodes [
   Event BlueprintUpdateAnimation()
   > n1 = Gravity Override State Function DEPRECATED(CharacterPartAnimInstance=Self-Reference(), GameTimeInSeconds=GetGameTimeInSeconds(), GravityOverrideParams=Get Gravity Override Params())
   > Set Gravity Override(Gravity Override=@n1)
   TryGetPawnOwner()
  ]}
]
variables [
 Gravity Override {type=Vector}
 Gravity Override Params {type=GravityOverrideParamsStruct}
]
