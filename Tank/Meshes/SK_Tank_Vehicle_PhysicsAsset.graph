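# Physics asset for the tank vehicle skeletal mesh, giving it a collision hull on a single root body with no constraints.
# Overlapping boxes, several of them pitched, approximate the hull, with named shapes for the front, front left and right body, engine and driver cover.
# Three capsules named as passenger collision sit higher toward the rear: one central, two offset behind it to either side.
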PhysicsAsset /Tank/Meshes/SK_Tank_Vehicle_PhysicsAsset
bodies {root [
  {shapeType=Capsule, length=115, radius=60, shapeName=PassengerCollision, center=<-138.9455 0.0002 260>}
  {shapeType=Capsule, length=115, radius=60, shapeName=PassengerBLCollision, center=<-282.7771 -125.4474 260>}
  {shapeType=Capsule, length=115, radius=60, shapeName=PassengerBRCollision, center=<-285.1627 121.2916 266.0811>}
  {
   shapeType=Box
   extentX=159.125
   extentY=89.875
   extentZ=131.4375
   shapeName=DriverCover
   center=<165.3101 -0.0006 131.889>
   rotation=<rot -2.8125 0 -0.0003>
  }
  {shapeType=Box, extentX=60, extentY=191.375, extentZ=146, shapeName=Engine, center=<-395.3615 -0.0001 133.8832>}
  {shapeType=Box, extentX=60, extentY=185.375, extentZ=137.5, center=<-387.7465 -0.0001 133.8846>}
  {shapeType=Box, extentX=77.3125, extentY=91.5, extentZ=26.8125, center=<266.2794 -178.3947 155.8057>, rotation=<rot -14.375 0.0002 -0.0006>}
  {shapeType=Box, extentX=77.3125, extentY=91.5, extentZ=26.8125, center=<266.3369 177.5122 155.7517>, rotation=<rot -14.375 0.0002 -0.0006>}
  {
   shapeType=Box
   extentX=63
   extentY=256.6875
   extentZ=104.25
   shapeName=Front
   center=<251.384 0.0001 90.3867>
   rotation=<rot -19.6875 0.0005 -0.0014>
  }
  {
   shapeType=Box
   extentX=186.3125
   extentY=200.8125
   extentZ=122.1875
   shapeName=FrontLeftBody
   center=<146.1572 -140.014 118.3806>
   rotation=<rot -14.0625 0 0>
  }
  {
   shapeType=Box
   extentX=186.3125
   extentY=198
   extentZ=122.1875
   shapeName=FrontRightBody
   center=<146.3324 141.2338 118.2316>
   rotation=<rot -14.0625 0 0>
  }
 ]}
constraints []
